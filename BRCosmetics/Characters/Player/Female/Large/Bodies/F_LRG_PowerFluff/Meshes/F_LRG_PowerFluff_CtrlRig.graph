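# Control Rig for the F_LRG_PowerFluff cosmetic body that runs a shared DeformRig function and procedurally adjusts the belt and its attachments.
# DeltaRotation converts right-thigh swing/twist against a reference null into clamped side and up weights.
# These weights offset belt aim targets, which follow body bones through weighted parent constraints and steer the belt bones through aim constraints.
# Grenade and rabbit bones follow through parent constraints, and IsDBNO triggers extra transform offsets.
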ControlRigBlueprint /BRCosmetics/Characters/Player/Female/Large/Bodies/F_LRG_PowerFluff/Meshes/F_LRG_PowerFluff_CtrlRig
graphs [
 New Function_ContainedGraph {nodes [
   n1 = RigVMFunctionEntryNode()
   RigVMFunction_MathBoolToInteger = &RigVMFunction_MathBoolToInteger(Value=@n1.onlyTwist)
   RigVMFunction_MathQuaternionSwingTwist = &RigVMFunction_MathQuaternionSwingTwist(
    twistAxis=@n1.twistAxis
    Input=&RigUnit_GetRelativeTransformForItem(
     bChildInitial=False
     bParentInitial=False
     child=(Type=Bone,Name="None")
     parent=(Type=Bone,Name="None")
     child.Name=@n1.child
     parent.Name=@n1.parent
     parent.Type=@n1.parentType
    ).RelativeTransform.Rotation
   )
   RigVMFunction_MathQuaternionToEuler = &RigVMFunction_MathQuaternionToEuler(
    Value=RigVMDispatchNode(
     Values=((),())
     Index=@RigVMFunction_MathBoolToInteger.Result
     Values.0=@RigVMFunction_MathQuaternionSwingTwist.Swing
     Values.1=@RigVMFunction_MathQuaternionSwingTwist.Twist
    ) {template="DISPATCH_RigVMDispatch_SelectInt32(in Index,in Values,out Result)"}.Result
    RotationOrder=@n1.Rotation Order
   )
   > DISPATCH_RigVMDispatch_Print = RigVMDispatchNode(ScreenColor=(R=1,G=1,B=1,A=1), ScreenDuration=1.5, Enabled=@n1.debugRotation, Value=@RigVMFunction_MathQuaternionToEuler.Result) {template="DISPATCH_RigVMDispatch_Print(in Prefix,in Value,in Enabled,in ScreenDuration,in ScreenColor)"}
   Clamp = &RigVMFunction_MathIntClamp(Value=@n1.primAxis)
   DISPATCH_RigVMDispatch_SelectInt32 = RigVMDispatchNode(
    Values=(0,0,0)
    Index=@Clamp.Result
    Values.0=@RigVMFunction_MathQuaternionToEuler.Result.X
    Values.1=@RigVMFunction_MathQuaternionToEuler.Result.Y
    Values.2=@RigVMFunction_MathQuaternionToEuler.Result.Z
   ) {template="DISPATCH_RigVMDispatch_SelectInt32(in Index,in Values,out Result)"}
   Remap = &RigVMFunction_MathDoubleRemap(
    TargetMaximum=1
    TargetMinimum=0
    bClamp=False
    Value=@DISPATCH_RigVMDispatch_SelectInt32.Result
    SourceMaximum=@n1.max
    SourceMinimum=@n1.min
   )
   Multiply = &RigVMFunction_MathDoubleMul(A=&RigVMFunction_MathDoubleClamp(Value=@Remap.Result).Result, B=@n1.premult)
   > RigVMFunctionReturnNode(outputWeight=@Multiply.Result)
  ]}
 RigVMFunctionLibrary {nodes [DeltaRotation = RigVMCollapseNode(ExecuteContext=(), Rotation Order=ZYX, premult=1, twistAxis=(X=1,Y=0,Z=0))]}
 RigVMFunction_Sequence_1_ContainedGraph {nodes [
   n1 = RigVMFunctionEntryNode()
   RigVMFunction_Sequence = &RigVMFunction_Sequence(ExecuteContext=@n1.ExecuteContext)
   n2 = &RigVMFunction_Sequence(ExecuteContext=@RigVMFunction_Sequence.B)
   n3 = &RigVMFunction_Sequence(ExecuteContext=@n2.B)
   n4 = &RigVMFunction_Sequence(ExecuteContext=@n3.B)
   n5 = &RigVMFunction_Sequence(ExecuteContext=@n4.B)
   n6 = &RigVMFunction_Sequence(ExecuteContext=@n5.B)
   RigVMFunctionReturnNode(A=@RigVMFunction_Sequence.A, B=@n2.A, C=@n3.A, D=@n4.A, E=@n5.A, F=@n6.A, G=@n6.B)
  ]}
 RigVMModel {nodes [
   RigUnit_BeginExecution = &RigUnit_BeginExecution()
   RigVMFunction_Sequence = RigVMAggregateNode(ExecuteContext=@RigUnit_BeginExecution.ExecutePin)
   FN_DeformRig_FNC = RigVMFunctionReferenceNode(
    DeformRig_ Curve="(deformRig_l_upperarm_up_alpha_72_#1=False,deformRig_r_upperarm_up_alpha_73_#2=False,deformRig_l_upperarm_fr_alpha_74_#3=False,deformRig_r_upperarm_fr_alpha_75_CB51B8 ...(truncated)"
    DeformRig_Bool="(upperarm_1_#4=True,elbow_3_#5=True,wrist_7_#6=True,groin_8_#7=True,glute_17_#8=True,knee_9_40C408324CA261 ...(truncated)"
    DeformRig_Debug=(upperarm_debug_18_#4=False,wrist_debug_20_#6=False,lat_debug_26_#9=False,groinGlute_debug_27_#7=False)
    DeformRig_Value="(upperarm_up_18_#10=(Rotation=(X=0,Y=0,Z=0,W=1),Translation=(X=0,Y=0,Z=8),Scale3D=(X=0.7,Y=1,Z=1)),upperarm_fwd_fr_19_#11=(Rotation=(X=0,Y=0,Z=0,W=1),Translation=(X=0,Y=10,Z=0),Scale3D=(X=0.5,Y=1,Z= ...(truncated)"
    ExecuteContext=@RigVMFunction_Sequence.B
   )
   DeltaRotation_1 = RigVMFunctionReferenceNode(
    Rotation Order=ZYX
    child=thigh_r
    debugRotation=false
    max=-65
    onlyTwist=false
    parent=thigh_r_null
    parentType=Null
    premult=1
    primAxis=1
    twistAxis=(X=1,Y=0,Z=0)
    ExecuteContext=@RigVMFunction_Sequence.C
   )
   > RigVMVariableNode(Value=@DeltaRotation_1.outputWeight) {variable=delta_thigh_R_side}
   > &RigUnit_ModifyTransforms(
    ItemToModify="((Item=(Type=Bone,Name=\"aim_belt_center_transform_01\"),Transform=(Rotation=(X=0,Y=0,Z=0,W=-1),Translation=(X=4.4919,Y=0,Z=0),Scale3D=(X=1,Y=1,Z=1))),(Item=(Type=Bone,Name=\"aim_belt_fr_transform_02\"),Transform=(Rotation=(X=0,Y=0,Z=0,W=-1),Translation=(X=0.6 ...(truncated)"
    WeightMinimum=0
    Weight=@DeltaRotation_1.outputWeight
   )
   &RigUnit_ParentConstraint(
    AdvancedSettings=(RotationOrderForFilter=XZY)
    Child=(Type=Bone,Name="aim_belt_fr_02")
    Parents="((Item=(Type=Bone,Name=\"spine_04\"),Weight=0.0029),(Item=(Type=Bone,Name=\"spine_01\"),Weight=0.0081),(Item=(Type=Bone,Name=\"thigh_twist_01_l\"),Weight=0.0111),(Item=(Type=Bone,Name=\"spine_02\"),Weight=0.0141),(Item=(Type=Bone,Name=\"thigh_r\"),Weight=0.0436),(It ...(truncated)"
    Weight=1
    ExecutePin=@RigVMFunction_Sequence.D
   )
   > &RigUnit_ParentConstraint(
    AdvancedSettings=(RotationOrderForFilter=XZY)
    Child=(Type=Bone,Name="aim_belt_fr_03")
    Parents="((Item=(Type=Bone,Name=\"thigh_l\"),Weight=0.0003),(Item=(Type=Bone,Name=\"thigh_twist_01_l\"),Weight=0.0012),(Item=(Type=Bone,Name=\"spine_01\"),Weight=0.0019),(Item=(Type=Bone,Name=\"spine_02\"),Weight=0.0031),(Item=(Type=Bone,Name=\"thigh_r\"),Weight=0.1412),(Ite ...(truncated)"
    Weight=1
   )
   > DeltaRotation_0 = RigVMFunctionReferenceNode(
    Rotation Order=ZYX
    child=thigh_r
    debugRotation=false
    max=-100
    parent=thigh_r_null
    parentType=Null
    premult=1
    primAxis=2
    twistAxis=(X=1,Y=0,Z=0)
   )
   > RigVMVariableNode(Value=@DeltaRotation_0.outputWeight) {variable=delta_thigh_R_up}
   RigVMFunction_AnimEvalRichCurve = &RigVMFunction_AnimEvalRichCurve(
    Curve="(EditorCurveData=(Keys=((InterpMode=RCIM_Cubic,TangentMode=RCTM_User,ArriveTangent=0.3757,LeaveTangent=0.3757),(InterpMode=RCIM_Cubic,TangentMode=RCTM_User,Time=0.7766,Value=0.5537,ArriveTangent=1,LeaveTangent=1),(InterpMode=RCIM_Cubic,TangentMode=RCTM_Use ...(truncated)"
    SourceMinimum=0
    TargetMinimum=0
    Value=@DeltaRotation_0.outputWeight
   )
   > &RigUnit_ModifyTransforms(
    ItemToModify="((Item=(Type=Bone,Name=\"aim_belt_fr_transform_02\"),Transform=(Rotation=(X=0,Y=0,Z=0,W=-1),Translation=(X=-8.0561,Y=2.7537,Z=-5.6274),Scale3D=(X=1,Y=1,Z=1))),(Item=(Type=Bone,Name=\"aim_belt_fr_transform_03\"),Transform=(Rotation=(X=0,Y=0,Z=0,W=1),Translation ...(truncated)"
    WeightMinimum=0
    Weight=@RigVMFunction_AnimEvalRichCurve.Result
   )
   > &RigUnit_AimConstraintLocalSpaceOffset(
    Child=(Type=Bone,Name="base_belt_fr_01")
    Filter=(bX=True,bY=True,bZ=True)
    Parents=((Item=(Type=Bone,Name="aim_belt_fr_transform_02"),Weight=1))
    Weight=1
    WorldUp=(Target=(X=0,Y=0,Z=0),Kind=Direction,Space=(Type=None,Name="None"))
   )
   > &RigUnit_AimConstraintLocalSpaceOffset(
    Child=(Type=Bone,Name="base_belt_fr_02")
    Filter=(bX=True,bY=True,bZ=True)
    Parents=((Item=(Type=Bone,Name="aim_belt_fr_transform_03"),Weight=1))
    Weight=1
    WorldUp=(Target=(X=0,Y=0,Z=0),Kind=Direction,Space=(Type=None,Name="None"))
   )
   > &RigUnit_AimConstraintLocalSpaceOffset(
    Child=(Type=Bone,Name="base_belt_fr_03")
    Filter=(bX=True,bY=True,bZ=True)
    Parents=((Item=(Type=Bone,Name="aim_belt_center_transform_01"),Weight=1))
    Weight=1
    WorldUp=(Target=(X=0,Y=0,Z=0),Kind=Direction,Space=(Type=None,Name="None"))
   )
   > &RigUnit_ModifyTransforms(
    ItemToModify="((Item=(Type=Bone,Name=\"dyn_belt_fr_01\"),Transform=(Rotation=(X=0,Y=0,Z=0,W=1),Translation=(X=0,Y=0,Z=0),Scale3D=(X=1,Y=1,Z=1))),(Item=(Type=Bone,Name=\"dyn_belt_fr_02\"),Transform=(Rotation=(X=0.3635,Y=0,Z=0,W=-0.9316),Translation=(X=0,Y=0,Z=0),Scale3D=(X=1 ...(truncated)"
    WeightMinimum=0
    Weight=@RigVMFunction_AnimEvalRichCurve.Result
   )
   > &RigUnit_ModifyTransforms(
    ItemToModify=((Item=(Type=Bone,Name="dyn_rabit_front_r"),Transform=(Rotation=(X=0.0998,Y=0,Z=0,W=-0.995),Translation=(X=0,Y=0,Z=0),Scale3D=(X=1,Y=1,Z=1))))
    Weight=1
    WeightMinimum=0
   )
   &RigUnit_ParentConstraint(
    AdvancedSettings=(RotationOrderForFilter=XZY)
    Child=(Type=Bone,Name="aim_belt_bck_02")
    Parents="((Item=(Type=Bone,Name=\"spine_03\"),Weight=0.0022),(Item=(Type=Bone,Name=\"thigh_l\"),Weight=0.0141),(Item=(Type=Bone,Name=\"spine_02\"),Weight=0.0202),(Item=(Type=Bone,Name=\"spine_01\"),Weight=0.0223),(Item=(Type=Bone,Name=\"deform_glute_l\"),Weight=0.0229),(Item ...(truncated)"
    Weight=1
    ExecutePin=@RigVMFunction_Sequence.E
   )
   > &RigUnit_ParentConstraint(
    AdvancedSettings=(RotationOrderForFilter=XZY)
    Child=(Type=Bone,Name="aim_belt_bck_03")
    Parents="((Item=(Type=Bone,Name=\"thigh_l\"),Weight=0.0049),(Item=(Type=Bone,Name=\"spine_02\"),Weight=0.0067),(Item=(Type=Bone,Name=\"spine_01\"),Weight=0.0068),(Item=(Type=Bone,Name=\"thigh_twist_01_l\"),Weight=0.0146),(Item=(Type=Bone,Name=\"thigh_r\"),Weight=0.0778),(Ite ...(truncated)"
    Weight=1
   )
   > &RigUnit_ModifyTransforms(
    ItemToModify="((Item=(Type=Bone,Name=\"aim_belt_bck_transform_02\"),Transform=(Rotation=(X=0,Y=0,Z=0,W=1),Translation=(X=-0.5519,Y=-1.774,Z=-2.1896),Scale3D=(X=1,Y=1,Z=1))),(Item=(Type=Bone,Name=\"aim_belt_bck_transform_03\"),Transform=(Rotation=(X=0,Y=0,Z=0,W=1),Translatio ...(truncated)"
    WeightMinimum=0
    Weight=RigVMVariableNode() {variable=delta_thigh_R_up}.Value
   )
   > &RigUnit_AimConstraintLocalSpaceOffset(
    Child=(Type=Bone,Name="base_belt_bck_01")
    Filter=(bX=True,bY=True,bZ=True)
    Parents=((Item=(Type=Bone,Name="aim_belt_bck_transform_02"),Weight=1))
    Weight=1
    WorldUp=(Target=(X=0,Y=0,Z=0),Kind=Direction,Space=(Type=None,Name="None"))
   )
   > &RigUnit_AimConstraintLocalSpaceOffset(
    Child=(Type=Bone,Name="base_belt_bck_02")
    Filter=(bX=True,bY=True,bZ=True)
    Parents=((Item=(Type=Bone,Name="aim_belt_bck_transform_03"),Weight=1))
    Weight=1
    WorldUp=(Target=(X=0,Y=0,Z=0),Kind=Direction,Space=(Type=None,Name="None"))
   )
   > &RigUnit_ModifyTransforms(
    ItemToModify="((Item=(Type=Bone,Name=\"dyn_belt_bck_02\"),Transform=(Rotation=(X=0.1634,Y=0,Z=0,W=0.9866),Translation=(X=0,Y=0,Z=0),Scale3D=(X=1,Y=1,Z=1))),(Item=(Type=Bone,Name=\"dyn_belt_bck_03\"),Transform=(Rotation=(X=-0.4337,Y=-0.141,Z=-0.0285,W=-0.8895),Translation=(X ...(truncated)"
    WeightMinimum=0
    Weight=RigVMVariableNode() {variable=delta_thigh_R_up}.Value
   )
   > &RigUnit_ParentConstraint(
    AdvancedSettings=(RotationOrderForFilter=XZY)
    Child=(Type=Bone,Name="aim_belt_center_transform")
    Parents="((Item=(Type=Bone,Name=\"thigh_l\"),Weight=0.0002),(Item=(Type=Bone,Name=\"spine_01\"),Weight=0.0002),(Item=(Type=Bone,Name=\"spine_02\"),Weight=0.0003),(Item=(Type=Bone,Name=\"thigh_twist_01_l\"),Weight=0.0006),(Item=(Type=Bone,Name=\"deform_groin_r\"),Weight=0.000 ...(truncated)"
    Weight=1
   )
   > &RigUnit_ModifyTransforms(
    ItemToModify=((Item=(Type=Bone,Name="aim_belt_center_transform_01"),Transform=(Rotation=(X=0,Y=0,Z=0,W=-1),Translation=(X=1.2573,Y=-3.2972,Z=8.4808),Scale3D=(X=1,Y=1,Z=1))))
    WeightMinimum=0
    Weight=RigVMVariableNode() {variable=delta_thigh_R_up}.Value
   )
   > &RigUnit_AimConstraintLocalSpaceOffset(
    Child=(Type=Bone,Name="base_belt_bck_03")
    Filter=(bX=True,bY=True,bZ=True)
    Parents=((Item=(Type=Bone,Name="aim_belt_center_transform_01"),Weight=1))
    Weight=1
    WorldUp=(Target=(X=0,Y=0,Z=1),Kind=Direction,Space=(Type=Bone,Name="aim_belt_center_transform_01"))
   )
   > &RigUnit_AimConstraintLocalSpaceOffset(
    Child=(Type=Bone,Name="base_belt_fr_03")
    Filter=(bX=True,bY=True,bZ=True)
    Parents=((Item=(Type=Bone,Name="aim_belt_center_transform_01"),Weight=1))
    Weight=1
    WorldUp=(Target=(X=0,Y=0,Z=1),Kind=Direction,Space=(Type=Bone,Name="aim_belt_center_transform_01"))
   )
   &RigUnit_ParentConstraint(
    AdvancedSettings=(RotationOrderForFilter=XZY)
    Child=(Type=Bone,Name="base_grenade_back_r")
    Parents=((Item=(Type=Bone,Name="dyn_belt_bck_02"),Weight=0.0034),(Item=(Type=Bone,Name="dyn_belt_fr_03"),Weight=0.0348),(Item=(Type=Bone,Name="dyn_belt_bck_03"),Weight=0.9619))
    Weight=1
    ExecutePin=@RigVMFunction_Sequence.F
   )
   > &RigUnit_ParentConstraint(
    AdvancedSettings=(RotationOrderForFilter=XZY)
    Child=(Type=Bone,Name="base_grenade_center_r")
    Parents=((Item=(Type=Bone,Name="dyn_belt_bck_03"),Weight=0.2025),(Item=(Type=Bone,Name="dyn_belt_fr_03"),Weight=0.7975))
    Weight=1
   )
   > &RigUnit_ParentConstraint(
    AdvancedSettings=(RotationOrderForFilter=XZY)
    Child=(Type=Bone,Name="base_grenade_front_r")
    Parents=((Item=(Type=Bone,Name="pelvis"),Weight=0.0002),(Item=(Type=Bone,Name="dyn_belt_bck_03"),Weight=0.0144),(Item=(Type=Bone,Name="dyn_belt_fr_02"),Weight=0.0724),(Item=(Type=Bone,Name="dyn_belt_fr_03"),Weight=0.9131))
    Weight=1
   )
   > &RigUnit_ParentConstraint(
    AdvancedSettings=(RotationOrderForFilter=XZY)
    Child=(Type=Bone,Name="base_rabit_front_r")
    Parents="((Item=(Type=Bone,Name=\"deform_groin_l\"),Weight=0.0001),(Item=(Type=Bone,Name=\"spine_01\"),Weight=0.0001),(Item=(Type=Bone,Name=\"spine_02\"),Weight=0.0002),(Item=(Type=Bone,Name=\"thigh_twist_01_l\"),Weight=0.0003),(Item=(Type=Bone,Name=\"dyn_belt_fr_01\"),Weigh ...(truncated)"
    Weight=1
   )
   > &RigUnit_ParentConstraint(
    AdvancedSettings=(RotationOrderForFilter=XZY)
    Child=(Type=Bone,Name="base_rabit_back_r")
    Parents="((Item=(Type=Bone,Name=\"pelvis\"),Weight=0.0001),(Item=(Type=Bone,Name=\"dyn_belt_bck_01\"),Weight=0.0017),(Item=(Type=Bone,Name=\"dyn_belt_fr_03\"),Weight=0.0167),(Item=(Type=Bone,Name=\"dyn_belt_bck_02\"),Weight=0.2136),(Item=(Type=Bone,Name=\"dyn_belt_bck_03\"), ...(truncated)"
    Weight=1
   )
   RigVMFunction_ControlFlowBranch = &RigVMFunction_ControlFlowBranch(ExecuteContext=@RigVMFunction_Sequence.G, Condition=RigVMVariableNode() {variable=IsDBNO}.Value)
   &RigUnit_ModifyTransforms(
    ItemToModify="((Item=(Type=Bone,Name=\"None\"),Transform=(Rotation=(X=0,Y=0,Z=0,W=1),Translation=(X=0,Y=0,Z=0),Scale3D=(X=1,Y=1,Z=1))),(Item=(Type=Bone,Name=\"dyn_belt_bck_03\"),Transform=(Rotation=(X=0,Y=0,Z=0,W=1),Translation=(X=1.569,Y=-0.662,Z=0),Scale3D=(X=1,Y=1,Z=1))) ...(truncated)"
    Weight=1
    WeightMinimum=0
    ExecutePin=@RigVMFunction_ControlFlowBranch.True
   )
   &RigUnit_ModifyTransforms(
    ExecutePin=()
    ItemToModify=((Item=(Type=Bone,Name="dyn_belt_bck_03"),Transform=(Rotation=(X=0,Y=0,Z=0,W=1),Translation=(X=0,Y=0,Z=0),Scale3D=(X=1,Y=1,Z=1))))
    WeightMinimum=0
    Weight=RigVMVariableNode() {variable=delta_thigh_R_side}.Value
   )
   RigUnit_GetTransform = &RigUnit_GetTransform(Item=(Type=Control,Name="thigh_r_ctrl"), bInitial=False)
   RigUnit_SetTransform = &RigUnit_SetTransform(ExecutePin=(), Item=(Type=Bone,Name="thigh_r"), Weight=1, bInitial=False, Value=@RigUnit_GetTransform.Transform)
  ]}
]
hierarchy [
 root {loc=<0 0 0>, rot=<0 0 0 1>}
 pelvis {parent=root, loc=<0 -0.648 94.7699>, rot=<0.0127 0.707 -0.0127 -0.707>}
 spine_01 {parent=pelvis, loc=<5.0975 -0.6388 0>, rot=<0 0 0.034 -0.9994>}
 spine_02 {parent=spine_01, loc=<7.252 -0.035 0>, rot=<0 0 0.0238 -0.9997>}
 spine_03 {parent=spine_02, loc=<6.0106 0 0>, rot=<0 0 0.0149 -0.9999>}
 spine_04 {parent=spine_03, loc=<8.1023 0 0>, rot=<0 0 -0.0132 -0.9999>}
 spine_05 {parent=spine_04, loc=<10.0142 0 0>, rot=<0 0 -0.0015 -1>}
 neck_01 {parent=spine_05, loc=<11.3775 -0.6396 0>, rot=<0 0 -0.2003 -0.9797>}
 neck_02 {parent=neck_01, loc=<4.5627 0 0>, rot=<0 0 0.0504 -0.9987>}
 head {parent=neck_02, loc=<5.3914 0 0>, rot=<0 0 0.115 -0.9934>}
 clavicle_l {parent=spine_05, loc=<4.1651 -0.3023 -5.0942>, rot=<0.0465 -0.7289 -0.009 -0.6829>}
 upperarm_l {parent=clavicle_l, loc=<11.2117 -1.1486 -0.0332>, rot=<0.0179 -0.2982 -0.028 -0.9539>}
 lowerarm_l {parent=upperarm_l, loc=<26.5013 -0.0835 0.2605>, rot=<0.0036 0.0161 -0.2298 -0.9731>}
 hand_l {parent=lowerarm_l, loc=<23.1356 -1.1606 -0.2612>, rot=<-0.6704 -0.0028 0.0567 -0.7398>}
 thumb_01_l {parent=hand_l, loc=<1.3118 -0.3191 -2.459>, rot=<0.6546 -0.4042 0.0664 -0.6354>}
 thumb_02_l {parent=thumb_01_l, loc=<5.9336 -0.0448 0.0401>, rot=<-0.0037 0.0213 0.298 -0.9543>}
 thumb_03_l {parent=thumb_02_l, loc=<2.2223 0.3682 -0.0907>, rot=<0.0118 -0.013 0.0066 -0.9998>}
 index_metacarpal_l {parent=hand_l, loc=<1.6535 0.0183 -2.96>, rot=<0.0158 -0.0124 -0.0173 -0.9996>}
 index_01_l {parent=index_metacarpal_l, loc=<8.3008 0.0357 -0.0524>, rot=<-0.0044 0.0485 0.0924 -0.9945>}
 index_02_l {parent=index_01_l, loc=<3.3587 -0.3004 0.0452>, rot=<-0.002 0.0027 0.3225 -0.9466>}
 index_03_l {parent=index_02_l, loc=<2.5578 -0.123 -0.0252>, rot=<-0.002 0.0067 0.2558 -0.9667>}
 middle_metacarpal_l {parent=hand_l, loc=<2.1856 -0.1725 -0.5984>, rot=<0.006 0.0168 -0.0329 -0.9993>}
 middle_01_l {parent=middle_metacarpal_l, loc=<7.3275 -0.2447 0.0129>, rot=<-0.0208 0.0042 0.0858 -0.9961>}
 middle_02_l {parent=middle_01_l, loc=<4.0123 0.1401 0.1175>, rot=<-0.0218 0.0258 0.3512 -0.9357>}
 middle_03_l {parent=middle_02_l, loc=<2.6077 0.0534 0.1373>, rot=<0.0082 0.0188 0.1708 -0.9851>}
 ring_metacarpal_l {parent=hand_l, loc=<2.2074 -0.3625 1.3638>, rot=<-0.0284 0.0722 -0.0458 -0.9959>}
 ring_01_l {parent=ring_metacarpal_l, loc=<7.0295 -0.6197 -0.1446>, rot=<-0.0489 -0.0252 0.1892 -0.9804>}
 ring_02_l {parent=ring_01_l, loc=<3.6174 0.4647 -0.1027>, rot=<0 0 0.2806 -0.9598>}
 ring_03_l {parent=ring_02_l, loc=<2.4152 0.114 0.0199>, rot=<0 0 0.1189 -0.9929>}
 pinky_metacarpal_l {parent=hand_l, loc=<1.7616 -0.9753 3.3814>, rot=<-0.0397 0.0814 -0.069 -0.9935>}
 pinky_01_l {parent=pinky_metacarpal_l, loc=<7.1709 -0.4831 -0.1094>, rot=<-0.0654 -0.0035 0.2242 -0.9723>}
 pinky_02_l {parent=pinky_01_l, loc=<2.8274 0.0893 -0.1043>, rot=<0.0037 0.0011 0.2854 -0.9584>}
 pinky_03_l {parent=pinky_02_l, loc=<2.2046 0.0307 -0.0078>, rot=<-0.0002 0.0098 0.1445 -0.9895>}
 deform_wrist_dn_base_l {parent=hand_l, loc=<-0.7813 -1.7011 0.9082>, rot=<0 0 0 -1>}
 deform_wrist_dn_l {parent=deform_wrist_dn_base_l, loc=<0 0 0>, rot=<0 0 0 1>}
 deform_wrist_up_base_l {parent=hand_l, loc=<-0.1406 2.5627 0.0698>, rot=<0 0 0 -1>}
 deform_wrist_up_l {parent=deform_wrist_up_base_l, loc=<0 0 0>, rot=<0 0 0 1>}
 lowerarm_twist_02_l {parent=lowerarm_l, loc=<11.5353 0 0>, rot=<0 0 0 1>}
 lowerarm_twist_01_l {parent=lowerarm_l, loc=<16.7573 0 0>, rot=<0 0 0 1>}
 deform_elbow_in_base_l {parent=lowerarm_l, loc=<0 4.5494 0>, rot=<0 0 0 1>}
 deform_elbow_in_l {parent=deform_elbow_in_base_l, loc=<0 0 0>, rot=<0 0 0 1>}
 deform_elbow_out_base_l {parent=lowerarm_l, loc=<-2.6147 -4.6123 0>, rot=<0 0 0 1>}
 deform_elbow_out_l {parent=deform_elbow_out_base_l, loc=<0 0 0>, rot=<0 0 0 1>}
 upperarm_twist_01_l {parent=upperarm_l, loc=<7.2135 -0.1134 0>, rot=<0 0 0 1>}
 deform_upperarm_fr_base_l {parent=upperarm_twist_01_l, loc=<-5.3904 6.218 -0.859>, rot=<0 0 0 -1>}
 deform_upperarm_fr_l {parent=deform_upperarm_fr_base_l, loc=<0 0 0>, rot=<0 0 0 1>}
 deform_upperarm_up_base_l {parent=upperarm_twist_01_l, loc=<-8.9945 -0.5206 4.5756>, rot=<0 0 0 -1>}
 deform_upperarm_up_l {parent=deform_upperarm_up_base_l, loc=<0 0 0>, rot=<0 0 0 1>}
 deform_upperarm_bk_base_l {parent=upperarm_twist_01_l, loc=<-5.5712 -7.2751 -1.2275>, rot=<0 0 0 -1>}
 deform_upperarm_bk_l {parent=deform_upperarm_bk_base_l, loc=<0 0 0>, rot=<0 0 0 1>}
 upperarm_twist_02_l {parent=upperarm_l, loc=<19.7998 -0.1134 0>, rot=<0 0 0 1>}
 deform_bicep_base_l {parent=upperarm_twist_02_l, loc=<-1.8082 4.197 0>, rot=<0 0 0 1>}
 deform_bicep_l {parent=deform_bicep_base_l, loc=<0 0 0>, rot=<0 0 0 1>}
 clavicle_r {parent=spine_05, loc=<4.1651 -0.3023 5.0942>, rot=<0.7289 0.0465 -0.6829 0.009>}
 upperarm_r {parent=clavicle_r, loc=<-11.2117 1.1486 0.0332>, rot=<0.0179 -0.2982 -0.028 -0.9539>}
 lowerarm_r {parent=upperarm_r, loc=<-26.5013 0.0835 -0.2604>, rot=<0.0036 0.0161 -0.2298 -0.9731>}
 hand_r {parent=lowerarm_r, loc=<-23.1356 1.1606 0.2612>, rot=<-0.6704 -0.0028 0.0567 -0.7398>}
 thumb_01_r {parent=hand_r, loc=<-1.3118 0.3191 2.459>, rot=<0.6546 -0.4042 0.0664 -0.6354>}
 thumb_02_r {parent=thumb_01_r, loc=<-5.9336 0.0448 -0.0401>, rot=<-0.0037 0.0213 0.298 -0.9543>}
 thumb_03_r {parent=thumb_02_r, loc=<-2.2223 -0.3682 0.0907>, rot=<0.0118 -0.013 0.0066 -0.9998>}
 index_metacarpal_r {parent=hand_r, loc=<-1.6535 -0.0183 2.96>, rot=<0.0158 -0.0124 -0.0173 -0.9996>}
 index_01_r {parent=index_metacarpal_r, loc=<-8.3008 -0.0357 0.0524>, rot=<-0.0044 0.0485 0.0924 -0.9945>}
 index_02_r {parent=index_01_r, loc=<-3.3587 0.3004 -0.0452>, rot=<-0.002 0.0027 0.3225 -0.9466>}
 index_03_r {parent=index_02_r, loc=<-2.5578 0.123 0.0252>, rot=<-0.002 0.0067 0.2558 -0.9667>}
 middle_metacarpal_r {parent=hand_r, loc=<-2.1856 0.1725 0.5984>, rot=<0.006 0.0168 -0.0329 -0.9993>}
 middle_01_r {parent=middle_metacarpal_r, loc=<-7.3275 0.2447 -0.0129>, rot=<-0.0208 0.0042 0.0858 -0.9961>}
 middle_02_r {parent=middle_01_r, loc=<-4.0123 -0.1401 -0.1175>, rot=<-0.0218 0.0258 0.3512 -0.9357>}
 middle_03_r {parent=middle_02_r, loc=<-2.6077 -0.0534 -0.1373>, rot=<0.0082 0.0188 0.1708 -0.9851>}
 ring_metacarpal_r {parent=hand_r, loc=<-2.2074 0.3625 -1.3638>, rot=<-0.0284 0.0722 -0.0458 -0.9959>}
 ring_01_r {parent=ring_metacarpal_r, loc=<-7.0295 0.6197 0.1446>, rot=<-0.0489 -0.0252 0.1892 -0.9804>}
 ring_02_r {parent=ring_01_r, loc=<-3.6174 -0.4647 0.1027>, rot=<0 0 0.2806 -0.9598>}
 ring_03_r {parent=ring_02_r, loc=<-2.4152 -0.114 -0.0199>, rot=<0 0 0.1189 -0.9929>}
 pinky_metacarpal_r {parent=hand_r, loc=<-1.7616 0.9753 -3.3814>, rot=<-0.0397 0.0814 -0.069 -0.9935>}
 pinky_01_r {parent=pinky_metacarpal_r, loc=<-7.1709 0.4831 0.1094>, rot=<-0.0654 -0.0035 0.2242 -0.9723>}
 pinky_02_r {parent=pinky_01_r, loc=<-2.8274 -0.0893 0.1043>, rot=<0.0037 0.0011 0.2854 -0.9584>}
 pinky_03_r {parent=pinky_02_r, loc=<-2.2046 -0.0307 0.0078>, rot=<-0.0002 0.0098 0.1445 -0.9895>}
 base_carrot_acc_01_r {parent=hand_r, loc=<3.0193 -3.1372 0.9089>, rot=<0.3418 -0.7658 -0.4874 0.2435>}
 dyn_carrot_acc_01_r {parent=base_carrot_acc_01_r, loc=<0 0 0>, rot=<0 0 0 1>}
 deform_wrist_dn_base_r {parent=hand_r, loc=<0.781 1.7015 -0.9083>, rot=<0 0 0 -1>}
 deform_wrist_dn_r {parent=deform_wrist_dn_base_r, loc=<0.0003 -0.0003 0.0001>, rot=<0 0 0 -1>}
 deform_wrist_up_base_r {parent=hand_r, loc=<0.1405 -2.5626 -0.0698>, rot=<0 0 0 -1>}
 deform_wrist_up_r {parent=deform_wrist_up_base_r, loc=<0.0001 -0.0001 0>, rot=<0 0 0 -1>}
 lowerarm_twist_02_r {parent=lowerarm_r, loc=<-11.5353 0 0>, rot=<0 0 0 1>}
 lowerarm_twist_01_r {parent=lowerarm_r, loc=<-16.7573 0 0>, rot=<0 0 0 1>}
 deform_elbow_in_base_r {parent=lowerarm_r, loc=<0 -4.5494 -0.0001>, rot=<0 0 0 -1>}
 deform_elbow_in_r {parent=deform_elbow_in_base_r, loc=<0 0 0>, rot=<0 0 0 1>}
 deform_elbow_out_base_r {parent=lowerarm_r, loc=<2.6148 4.6123 0>, rot=<0 0 0 -1>}
 deform_elbow_out_r {parent=deform_elbow_out_base_r, loc=<0 0 0>, rot=<0 0 0 1>}
 upperarm_twist_01_r {parent=upperarm_r, loc=<-7.2135 0.1134 0>, rot=<0 0 0 -1>}
 deform_upperarm_fr_base_r {parent=upperarm_twist_01_r, loc=<5.3903 -6.218 0.8592>, rot=<0 0 0 -1>}
 deform_upperarm_fr_r {parent=deform_upperarm_fr_base_r, loc=<0.0001 0 -0.0001>, rot=<0 0 0 -1>}
 deform_upperarm_up_base_r {parent=upperarm_twist_01_r, loc=<8.9943 0.5206 -4.5754>, rot=<0 0 0 -1>}
 deform_upperarm_up_r {parent=deform_upperarm_up_base_r, loc=<0.0002 0 -0.0002>, rot=<0 0 0 -1>}
 deform_upperarm_bk_base_r {parent=upperarm_twist_01_r, loc=<5.5714 7.2751 1.2273>, rot=<0 0 0 -1>}
 deform_upperarm_bk_r {parent=deform_upperarm_bk_base_r, loc=<-0.0001 0 0.0002>, rot=<0 0 0 -1>}
 upperarm_twist_02_r {parent=upperarm_r, loc=<-19.7998 0.1134 0>, rot=<0 0 0 -1>}
 deform_bicep_base_r {parent=upperarm_twist_02_r, loc=<1.8083 -4.197 -0.0001>, rot=<0 0 0 1>}
 deform_bicep_r {parent=deform_bicep_base_r, loc=<0 0 0>, rot=<0 0 0 1>}
 attach_backpack {parent=spine_05, loc=<-0.1209 -13.4727 -0.0002>, rot=<-0.0093 -0.707 0.0093 -0.707>}
 Backpack_BR {parent=attach_backpack, type=Null, loc=<0 0 0>, rot=<0 0 0 1>}
 attach_cape {parent=spine_05, loc=<9.0194 -11.7138 -0.0003>, rot=<0.0616 -0.7044 -0.0616 -0.7044>}
 cape {parent=attach_cape, type=Null, loc=<0 0 0>, rot=<0 0 0 1>}
 deform_pec_aim_base_l {parent=spine_04, loc=<8.0082 12.8872 -2.9192>, rot=<0.0422 -0.4258 0.4306 -0.7947>}
 deform_pec_aim_l {parent=deform_pec_aim_base_l, loc=<0 0 0>, rot=<0 0 0 1>}
 deform_pec_base_l {parent=deform_pec_aim_l, loc=<16.3279 0 0>, rot=<0 0 0 1>}
 deform_pec_l {parent=deform_pec_base_l, loc=<0 0 0>, rot=<0 0 0 1>}
 deform_pec_aim_base_r {parent=spine_04, loc=<8.0082 12.8872 2.9192>, rot=<-0.4258 -0.0422 0.7947 0.4306>}
 deform_pec_aim_r {parent=deform_pec_aim_base_r, loc=<0 0 0>, rot=<0 0 0 -1>}
 deform_pec_base_r {parent=deform_pec_aim_r, loc=<-16.3279 0.0001 0.0001>, rot=<0 0 0 -1>}
 deform_pec_r {parent=deform_pec_base_r, loc=<0 0 0>, rot=<0 0 0 1>}
 thigh_l {parent=pelvis, loc=<-2.5624 -0.0642 -10.3834>, rot=<0.0004 0.0055 0.0155 -0.9999>}
 calf_l {parent=thigh_l, loc=<-39.2091 -0.1238 -0.6795>, rot=<0.0006 -0.0161 -0.0391 -0.9991>}
 foot_l {parent=calf_l, loc=<-43.624 -2.818 -0.713>, rot=<0.044 0.0061 0.0367 -0.9983>}
 ball_l {parent=foot_l, loc=<-4.6464 10.3094 -0.3137>, rot=<0.0001 -0.0001 -0.7063 -0.7079>}
 calf_twist_02_l {parent=calf_l, loc=<-17.5927 -1.7943 -0.2638>, rot=<0.0237 0.0095 -0.0035 -0.9997>}
 calf_twist_01_l {parent=calf_l, loc=<-34.2864 -1.9197 -0.5794>, rot=<0.0237 0.0095 -0.0035 -0.9997>}
 base_socks_l {parent=calf_twist_01_l, loc=<-1.5616 0.0103 0.03>, rot=<0 0 0 -1>}
 base_socks_08_l {parent=base_socks_l, loc=<-4.9398 5.5945 5.1227>, rot=<0.4545 0.4363 0.7497 0.2026>}
 dyn_socks_08_l {parent=base_socks_08_l, loc=<0 0 0>, rot=<0 0 0 1>}
 base_socks_07_l {parent=base_socks_l, loc=<-6.8946 2.7966 -7.6842>, rot=<-0.0968 -0.8829 0.0104 -0.4593>}
 dyn_socks_07_l {parent=base_socks_07_l, loc=<0 0 0>, rot=<0 0 0 1>}
 base_socks_06_l {parent=base_socks_l, loc=<-0.9605 -7.0703 -6.8587>, rot=<0.1103 -0.8364 0.3745 -0.3847>}
 dyn_socks_06_l {parent=base_socks_06_l, loc=<0 0 0>, rot=<0 0 0 1>}
 base_socks_05_l {parent=base_socks_l, loc=<-8.2333 -9.0868 -2.0598>, rot=<0.1733 -0.9675 -0.1452 -0.113>}
 dyn_socks_05_l {parent=base_socks_05_l, loc=<0 0 0>, rot=<0 0 0 1>}
 base_socks_04_l {parent=base_socks_l, loc=<-7.8042 -6.1372 -6.1681>, rot=<0.1248 -0.8371 0.4153 -0.3334>}
 dyn_socks_04_l {parent=base_socks_04_l, loc=<0 0 0>, rot=<0 0 0 1>}
 base_socks_03_l {parent=base_socks_l, loc=<5.5858 -4.7551 7.7457>, rot=<-0.1132 -0.9486 0.0139 0.2952>}
 dyn_socks_03_l {parent=base_socks_03_l, loc=<0 0 0>, rot=<0 0 0 1>}
 base_socks_02_l {parent=base_socks_l, loc=<0.969 -9.9452 0.7264>, rot=<0.3148 -0.9306 0.1777 0.0577>}
 dyn_socks_02_l {parent=base_socks_02_l, loc=<0 0 0>, rot=<0 0 0 1>}
 base_socks_01_l {parent=base_socks_l, loc=<-6.2417 -6.1638 6.0928>, rot=<-0.0298 -0.9345 -0.2276 0.2719>}
 dyn_socks_01_l {parent=base_socks_01_l, loc=<0 0 0>, rot=<0 0 0 1>}
 base_socks_09_l {parent=base_socks_l, loc=<-4.4698 6.8987 -2.179>, rot=<-0.4033 -0.6668 -0.45 -0.4362>}
 dyn_socks_09_l {parent=base_socks_09_l, loc=<0 0 0>, rot=<0 0 0 1>}
 base_socks_010_l {parent=base_socks_l, loc=<0.9799 2.2328 7.2771>, rot=<0.4812 0.113 0.8674 0.0568>}
 dyn_socks_010_l {parent=base_socks_010_l, loc=<0 0 0>, rot=<0 0 0 1>}
 base_socks_011_l {parent=base_socks_l, loc=<1.0768 1.5016 -8.1649>, rot=<-0.1109 -0.8873 -0.0358 -0.4463>}
 dyn_socks_011_l {parent=base_socks_011_l, loc=<0 0 0>, rot=<0 0 0 1>}
 base_socks_012_l {parent=base_socks_l, loc=<8.5708 1.109 -8.3299>, rot=<-0.0011 -0.9138 -0.0192 -0.4056>}
 dyn_socks_012_l {parent=base_socks_012_l, loc=<0 0 0>, rot=<0 0 0 1>}
 base_socks_013_l {parent=base_socks_l, loc=<7.8176 5.9829 -3.2614>, rot=<-0.3404 -0.7168 -0.5262 -0.3056>}
 dyn_socks_013_l {parent=base_socks_013_l, loc=<0 0 0>, rot=<0 0 0 1>}
 deform_knee_in_base_l {parent=calf_l, loc=<2.8092 -9.2971 0>, rot=<0 0 0 -1>}
 deform_knee_in_l {parent=deform_knee_in_base_l, loc=<0 0 0>, rot=<0 0 0 1>}
 deform_knee_out_base_l {parent=calf_l, loc=<-0.2964 5.0241 0>, rot=<0 0 0 -1>}
 deform_knee_out_l {parent=deform_knee_out_base_l, loc=<0 0 0>, rot=<0 0 0 1>}
 thigh_twist_01_l {parent=thigh_l, loc=<-23.0764 0 0>, rot=<0.0013 0 0 -1>}
 base_grenade_fr_l {parent=thigh_twist_01_l, loc=<3.7495 8.5296 -9.5775>, rot=<-0.3009 -0.667 -0.2773 -0.6226>}
 dyn_grenade_fr_l {parent=base_grenade_fr_l, loc=<0 0 0>, rot=<0 0 0 1>}
 base_grenade_cn_l {parent=thigh_twist_01_l, loc=<3.5636 1.4809 -15.5267>, rot=<0.1979 -0.6717 0.212 -0.6817>}
 dyn_grenade_cn_l {parent=base_grenade_cn_l, loc=<0 0 0>, rot=<0 0 0 1>}
 base_grenade_br_l {parent=thigh_twist_01_l, loc=<2.3164 -7.721 -12.1146>, rot=<0.2468 -0.6889 0.2328 -0.6406>}
 dyn_grenade_br_l {parent=base_grenade_br_l, loc=<0 0 0>, rot=<0 0 0 1>}
 deform_groin_base_l {parent=thigh_l, loc=<0 8.4477 0>, rot=<0 0 0 -1>}
 deform_groin_l {parent=deform_groin_base_l, loc=<0 0 0>, rot=<0 0 0 1>}
 thigh_r {parent=pelvis, loc=<-2.5624 -0.0642 10.3834>, rot=<0.0055 -0.0004 0.9999 0.0155>}
 calf_r {parent=thigh_r, loc=<39.2091 0.1238 0.6795>, rot=<0.0006 -0.0161 -0.0391 -0.9991>}
 foot_r {parent=calf_r, loc=<43.624 2.818 0.713>, rot=<0.044 0.0061 0.0367 -0.9983>}
 ball_r {parent=foot_r, loc=<4.6464 -10.3094 0.3137>, rot=<0.0001 -0.0001 -0.7063 -0.7079>}
 calf_twist_02_r {parent=calf_r, loc=<17.5927 1.7943 0.2638>, rot=<0.0237 0.0095 -0.0035 -0.9997>}
 calf_twist_01_r {parent=calf_r, loc=<34.2864 1.9197 0.5794>, rot=<0.0237 0.0095 -0.0035 -0.9997>}
 base_socks_r {parent=calf_twist_01_r, loc=<1.5616 -0.0103 -0.03>, rot=<0 0 0 -1>}
 base_socks_01_r {parent=base_socks_r, loc=<6.2417 6.1638 -6.0928>, rot=<0.8344 0.0295 -0.5166 -0.1898>}
 dyn_socks_01_r {parent=base_socks_01_r, loc=<0 0 0>, rot=<0 0 0 1>}
 base_socks_02_r {parent=base_socks_r, loc=<-0.9864 9.8854 -1.0377>, rot=<0.7365 0.1537 -0.2275 -0.6183>}
 dyn_socks_02_r {parent=base_socks_02_r, loc=<0 0 0>, rot=<0 0 0 1>}
 base_socks_03_r {parent=base_socks_r, loc=<-5.5858 4.7551 -7.7457>, rot=<0.9496 -0.1422 -0.2567 -0.1105>}
 dyn_socks_03_r {parent=base_socks_03_r, loc=<0 0 0>, rot=<0 0 0 1>}
 base_socks_04_r {parent=base_socks_r, loc=<7.8121 6.1103 6.2831>, rot=<0.3625 0.3554 -0.1623 -0.8462>}
 dyn_socks_04_r {parent=base_socks_04_r, loc=<0 0 0>, rot=<0 0 0 1>}
 base_socks_05_r {parent=base_socks_r, loc=<8.2333 9.0868 2.0598>, rot=<0.9422 0.1537 0.1138 -0.275>}
 dyn_socks_05_r {parent=base_socks_05_r, loc=<0 0 0>, rot=<0 0 0 1>}
 base_socks_06_r {parent=base_socks_r, loc=<0.9111 6.8847 7.0113>, rot=<0.3195 0.3518 -0.1049 -0.8736>}
 dyn_socks_06_r {parent=base_socks_06_r, loc=<0 0 0>, rot=<0 0 0 1>}
 base_socks_07_r {parent=base_socks_r, loc=<6.8946 -2.7966 7.6842>, rot=<0.8679 -0.0461 0.4908 -0.061>}
 dyn_socks_07_r {parent=base_socks_07_r, loc=<0 0 0>, rot=<0 0 0 1>}
 base_socks_08_r {parent=base_socks_r, loc=<4.7312 -5.7005 -5.2341>, rot=<0.8413 -0.1922 -0.3465 0.3678>}
 dyn_socks_08_r {parent=base_socks_08_r, loc=<0 0 0>, rot=<0 0 0 1>}
 base_socks_09_r {parent=base_socks_r, loc=<4.4591 -6.857 2.4437>, rot=<-0.5444 0.4403 0.3549 -0.6195>}
 dyn_socks_09_r {parent=base_socks_09_r, loc=<0 0 0>, rot=<0 0 0 1>}
 base_socks_010_r {parent=base_socks_r, loc=<-0.4199 -2.3055 -7.8117>, rot=<0.9138 -0.0577 -0.3981 0.0558>}
 dyn_socks_010_r {parent=base_socks_010_r, loc=<0 0 0>, rot=<0 0 0 1>}
 base_socks_011_r {parent=base_socks_r, loc=<-1.4732 -1.4282 8.4412>, rot=<-0.147 0.316 0.0743 -0.9344>}
 dyn_socks_011_r {parent=base_socks_011_r, loc=<0 0 0>, rot=<0 0 0 1>}
 base_socks_012_r {parent=base_socks_r, loc=<-8.5708 -1.109 8.3299>, rot=<0.9045 -0.0174 0.4191 -0.0768>}
 dyn_socks_012_r {parent=base_socks_012_r, loc=<0 0 0>, rot=<0 0 0 1>}
 base_socks_013_r {parent=base_socks_r, loc=<-7.8176 -5.9829 3.2614>, rot=<0.0833 0.0406 0.3994 -0.9121>}
 dyn_socks_013_r {parent=base_socks_013_r, loc=<0 0 0>, rot=<0 0 0 1>}
 deform_knee_in_base_r {parent=calf_r, loc=<-2.8091 9.2971 0>, rot=<0 0 0 -1>}
 deform_knee_in_r {parent=deform_knee_in_base_r, loc=<0 0 0>, rot=<0 0 0 1>}
 deform_knee_out_base_r {parent=calf_r, loc=<0.2964 -5.0241 0>, rot=<0 0 0 -1>}
 deform_knee_out_r {parent=deform_knee_out_base_r, loc=<0 0 0>, rot=<0 0 0 1>}
 thigh_twist_01_r {parent=thigh_r, loc=<23.0764 0 0>, rot=<0.0013 0 0 -1>}
 deform_groin_base_r {parent=thigh_r, loc=<0 -8.4477 0>, rot=<0 0 0 -1>}
 deform_groin_r {parent=deform_groin_base_r, loc=<0 0 0>, rot=<0 0 0 1>}
 base_belt {parent=pelvis, loc=<3.2893 0 0>, rot=<0 0 0 1>}
 base_rabit_back_r {parent=base_belt, loc=<-2.1173 -16.394 14.6509>, rot=<0.5641 -0.4263 0.5791 -0.4058>}
 dyn_rabit_back_r {parent=base_rabit_back_r, loc=<0 0 0>, rot=<0 0 0 1>}
 base_rabit_front_r {parent=base_belt, loc=<-1.8499 6.7349 18.7686>, rot=<0.6039 0.3678 0.5903 0.3892>}
 dyn_rabit_front_r {parent=base_rabit_front_r, loc=<0 0 0>, rot=<0 0 0 1>}
 base_grenade_front_r {parent=base_belt, loc=<-3.9257 2.7002 23.6942>, rot=<0.6769 0.2046 0.6691 0.2287>}
 dyn_grenade_front_r {parent=base_grenade_front_r, loc=<0 0 0>, rot=<0 0 0 1>}
 base_grenade_center_r {parent=base_belt, loc=<-4.9482 -4.7084 24.3738>, rot=<0.7045 -0.0605 0.7062 -0.0352>}
 dyn_grenade_center_r {parent=base_grenade_center_r, loc=<0 0 0>, rot=<0 0 0 1>}
 base_grenade_back_r {parent=base_belt, loc=<-4.0471 -12.2168 21.0134>, rot=<0.6627 -0.2467 0.6711 -0.2228>}
 dyn_grenade_back_r {parent=base_grenade_back_r, loc=<0 0 0>, rot=<0 0 0 1>}
 base_belt_fr_00 {parent=base_belt, loc=<6.1474 11.6072 -7.6803>, rot=<0.3461 -0.7263 0.3829 0.454>}
 base_belt_fr_01 {parent=base_belt_fr_00, loc=<0 0 0>, rot=<0 0 0 1>}
 dyn_belt_fr_01 {parent=base_belt_fr_01, loc=<0 0 0>, rot=<0 0 0 1>}
 base_belt_fr_02 {parent=base_belt_fr_01, loc=<14.2782 2.1011 -3.7182>, rot=<-0.0206 0.1777 0.1475 -0.9728>}
 dyn_belt_fr_02 {parent=base_belt_fr_02, loc=<0 0 0>, rot=<0 0 0 1>}
 base_belt_fr_03 {parent=base_belt_fr_02, loc=<14.7872 1.3241 -1.476>, rot=<0 0 0 1>}
 dyn_belt_fr_03 {parent=base_belt_fr_03, loc=<0 0 0>, rot=<0 0 0 1>}
 NewBone {parent=base_belt_fr_03, loc=<2 0 0>, rot=<0 0 0 1>}
 base_belt_bck_00 {parent=base_belt, loc=<4.5475 -16.3443 -7.2379>, rot=<-0.1305 0.8187 -0.093 -0.5513>}
 base_belt_bck_01 {parent=base_belt_bck_00, loc=<0 0 0>, rot=<0 0 0 1>}
 dyn_belt_bck_01 {parent=base_belt_bck_01, loc=<0 0 0>, rot=<0 0 0 1>}
 base_belt_bck_02 {parent=base_belt_bck_01, loc=<11.9028 0.2037 0.0886>, rot=<-0.0124 -0.1052 -0.0708 -0.9918>}
 dyn_belt_bck_02 {parent=base_belt_bck_02, loc=<0 0 0>, rot=<0 0 0 1>}
 base_belt_bck_03 {parent=base_belt_bck_02, loc=<9.7749 0.7293 1.5035>, rot=<-0.0271 0.0072 -0.2572 -0.9659>}
 dyn_belt_bck_03 {parent=base_belt_bck_03, loc=<0 0 0>, rot=<0 0 0 1>}
 deform_glute_base_l {parent=pelvis, loc=<-13.1199 -13.5027 -12.4424>, rot=<-0.011 0.0565 -0.4248 -0.9035>}
 deform_glute_l {parent=deform_glute_base_l, loc=<-0.0103 -0.0257 0>, rot=<0 0 0 1>}
 deform_glute_base_r {parent=pelvis, loc=<-13.1198 -13.5027 12.4424>, rot=<0.0565 0.011 0.9035 -0.4248>}
 deform_glute_r {parent=deform_glute_base_r, loc=<0 0 0>, rot=<0 0 0 1>}
 thigh_r_null {parent=pelvis, type=Null, loc=<-1.0855 -0.2109 8.7053>, rot=<0.0355 -0.0009 0.9993 0.0155>}
 thigh_r_ctrl {controlType=EULER_TRANSFORM, type=Control, loc=<0 0 0>, rot=<0 0 0 1>}
 aim_belt_fr_02 {loc=<-6.1066 13.1022 98.4231>, rot=<0.19 -0.7608 0.5654 -0.2557>}
 aim_belt_fr_transform_02 {parent=aim_belt_fr_02, loc=<0 0 0>, rot=<0 0 0 1>}
 aim_belt_fr_03 {loc=<-18.7833 6.059 94.9179>, rot=<0.19 -0.7608 0.5654 -0.2557>}
 aim_belt_fr_transform_03 {parent=aim_belt_fr_03, loc=<0 0 0>, rot=<0 0 0 1>}
 aim_belt_fr_transform_01 {loc=<7.6803 11.2899 103.7847>, rot=<0.0226 0.8252 -0.5303 0.193>}
 aim_belt_bck_transform_01 {loc=<7.2379 -16.701 103.1872>, rot=<0.0276 0.8698 -0.455 0.1889>}
 aim_belt_bck_02 {loc=<-3.7623 -18.0139 98.8285>, rot=<-0.0851 -0.8867 0.4478 -0.0771>}
 aim_belt_bck_transform_02 {parent=aim_belt_bck_02, loc=<0 0 0>, rot=<0 0 0 1>}
 aim_belt_bck_03 {loc=<-12.9123 -17.0687 95.1237>, rot=<-0.0851 -0.8867 0.4478 -0.0771>}
 aim_belt_bck_transform_03 {parent=aim_belt_bck_03, loc=<0 0 0>, rot=<0 0 0 1>}
 aim_belt_center_transform {loc=<-22.3664 -5.139 93.2808>, rot=<0 0 -0.9977 0.0677>}
 aim_belt_center_transform_01 {parent=aim_belt_center_transform, loc=<0 0 0>, rot=<0 0 0 1>}
]
variables [
 IsDBNO {type=bool, default=False}
 delta_thigh_R_side {type=double, default=0}
 delta_thigh_R_up {type=double, default=0}
]
