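# Animation Blueprint for the PowerFluff large female body cosmetic, adding secondary motion to accessories like the carrot, rabbits, grenades and socks.
# It copies the attached parent mesh's pose into a Control Rig fed bIsDBNO, then runs cached-pose stages per accessory, ending with the right socks.
# Weighted multi-bone constraints make most accessory bones follow nearby limb bones, before spring-driven, angle-limited Anim Dynamics nodes add jiggle.
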AnimBlueprint /BRCosmetics/Characters/Player/Female/Large/Bodies/F_LRG_PowerFluff/Meshes/F_LRG_PowerFluff_AnimBP
target_skeleton=&/F_LRG_PowerFluff_Skeleton
graphs [
 AnimGraph {nodes [
   "Transform (Modify) Bone - Bone: dyn_grenade_fr_l"(Alpha=1, Translation="0,0,0", Rotation="30,20,10", Scale="1,1,1") {rotationMode=BMM_Additive, alphaBoolBlend {blendOption=Linear}, boneToModify {boneName=dyn_grenade_fr_l}, rotation=<rot 30 20 10>}
   n1 = "Constraint - base_socks_01_r (PPPP)"!(
    ComponentPose=Local To Component(LocalPose=Use cached pose 'socket L cache'())
    Alpha=1
    ConstraintWeights_0=0.585
    ConstraintWeights_1:float=0.3903
    ConstraintWeights_2:float=0.0169
    ConstraintWeights_3:float=0.0078
   ) -> (Pose) {
    alphaBoolBlend {blendOption=Linear}
    boneToModify {boneName=base_socks_01_r}
    constraintSetup [
     {targetBone {boneName=foot_r}}
     {targetBone {boneName=calf_twist_01_r}}
     {targetBone {boneName=calf_twist_02_r}}
     {targetBone {boneName=calf_r}}
    ]
    constraintWeights [1,1,1,1]
   }
   n2 = "Constraint - base_socks_02_r (PPPP)"!(
    ComponentPose=@n1
    Alpha=1
    ConstraintWeights_0=0.4975
    ConstraintWeights_1:float=0.3749
    ConstraintWeights_2:float=0.0959
    ConstraintWeights_3:float=0.0316
   ) -> (Pose) {
    alphaBoolBlend {blendOption=Linear}
    boneToModify {boneName=base_socks_02_r}
    constraintSetup [
     {targetBone {boneName=calf_twist_01_r}}
     {targetBone {boneName=calf_twist_02_r}}
     {targetBone {boneName=calf_r}}
     {targetBone {boneName=foot_r}}
    ]
    constraintWeights [1,1,1,1]
   }
   n3 = "Constraint - base_socks_03_r (PPP)"!(ComponentPose=@n2, Alpha=1, ConstraintWeights_0=0.7187, ConstraintWeights_1:float=0.1747, ConstraintWeights_2:float=0.1066) -> (Pose) {
    alphaBoolBlend {blendOption=Linear}
    boneToModify {boneName=base_socks_03_r}
    constraintSetup [{targetBone {boneName=calf_twist_02_r}}, {targetBone {boneName=calf_twist_01_r}}, {targetBone {boneName=calf_r}}]
    constraintWeights [1,1,1]
   }
   n4 = "Constraint - base_socks_04_r (PPPP)"!(
    ComponentPose=@n3
    Alpha=1
    ConstraintWeights_0=0.7699
    ConstraintWeights_1:float=0.2044
    ConstraintWeights_2:float=0.0176
    ConstraintWeights_3:float=0.008
   ) -> (Pose) {
    alphaBoolBlend {blendOption=Linear}
    boneToModify {boneName=base_socks_04_r}
    constraintSetup [
     {targetBone {boneName=foot_r}}
     {targetBone {boneName=calf_twist_01_r}}
     {targetBone {boneName=calf_twist_02_r}}
     {targetBone {boneName=calf_r}}
    ]
    constraintWeights [1,1,1,1]
   }
   n5 = "Constraint - base_socks_05_r (PPPP)"!(
    ComponentPose=@n4
    Alpha=1
    ConstraintWeights_0=0.8234
    ConstraintWeights_1:float=0.1608
    ConstraintWeights_2:float=0.0118
    ConstraintWeights_3:float=0.004
   ) -> (Pose) {
    alphaBoolBlend {blendOption=Linear}
    boneToModify {boneName=base_socks_05_r}
    constraintSetup [
     {targetBone {boneName=foot_r}}
     {targetBone {boneName=calf_twist_01_r}}
     {targetBone {boneName=calf_twist_02_r}}
     {targetBone {boneName=calf_r}}
    ]
    constraintWeights [1,1,1,1]
   }
   n6 = "Constraint - base_socks_06_r (PPPP)"!(
    ComponentPose=@n5
    Alpha=1
    ConstraintWeights_0=0.5024
    ConstraintWeights_1:float=0.3896
    ConstraintWeights_2:float=0.0996
    ConstraintWeights_3:float=0.0085
   ) -> (Pose) {
    alphaBoolBlend {blendOption=Linear}
    boneToModify {boneName=base_socks_06_r}
    constraintSetup [
     {targetBone {boneName=calf_twist_01_r}}
     {targetBone {boneName=calf_twist_02_r}}
     {targetBone {boneName=calf_r}}
     {targetBone {boneName=foot_r}}
    ]
    constraintWeights [1,1,1,1]
   }
   n7 = "Constraint - base_socks_07_r (PP)"!(ComponentPose=@n6, Alpha=1, ConstraintWeights_0=0.8882, ConstraintWeights_1:float=0.1118) -> (Pose) {
    alphaBoolBlend {blendOption=Linear}
    boneToModify {boneName=base_socks_07_r}
    constraintSetup [{targetBone {boneName=foot_r}}, {targetBone {boneName=calf_twist_01_r}}]
    constraintWeights [1,1]
   }
   n8 = "Constraint - base_socks_08_r (PPPPP)"!(
    ComponentPose=@n7
    Alpha=1
    ConstraintWeights_0=0.5932
    ConstraintWeights_1:float=0.2996
    ConstraintWeights_2:float=0.0566
    ConstraintWeights_3:float=0.0314
    ConstraintWeights_4:float=0.0191
   ) -> (Pose) {
    alphaBoolBlend {blendOption=Linear}
    boneToModify {boneName=base_socks_08_r}
    constraintSetup [
     {targetBone {boneName=foot_r}}
     {targetBone {boneName=calf_twist_01_r}}
     {targetBone {boneName=calf_twist_02_r}}
     {targetBone {boneName=ball_r}}
     {targetBone {boneName=calf_r}}
    ]
    constraintWeights [1,1,1,1,1]
   }
   n9 = "Constraint - base_socks_09_r (PPPPP)"!(
    ComponentPose=@n8
    Alpha=1
    ConstraintWeights_0=0.5017
    ConstraintWeights_1:float=0.3772
    ConstraintWeights_2:float=0.0629
    ConstraintWeights_3:float=0.0358
    ConstraintWeights_4:float=0.0224
   ) -> (Pose) {
    alphaBoolBlend {blendOption=Linear}
    boneToModify {boneName=base_socks_09_r}
    constraintSetup [
     {targetBone {boneName=foot_r}}
     {targetBone {boneName=calf_twist_01_r}}
     {targetBone {boneName=calf_twist_02_r}}
     {targetBone {boneName=ball_r}}
     {targetBone {boneName=calf_r}}
    ]
    constraintWeights [1,1,1,1,1]
   }
   n10 = "Constraint - base_socks_010_r (PPPP)"!(
    ComponentPose=@n9
    Alpha=1
    ConstraintWeights_0=0.6145
    ConstraintWeights_1:float=0.2394
    ConstraintWeights_2:float=0.0759
    ConstraintWeights_3:float=0.0702
   ) -> (Pose) {
    alphaBoolBlend {blendOption=Linear}
    boneToModify {boneName=base_socks_010_r}
    constraintSetup [
     {targetBone {boneName=calf_twist_01_r}}
     {targetBone {boneName=calf_twist_02_r}}
     {targetBone {boneName=foot_r}}
     {targetBone {boneName=calf_r}}
    ]
    constraintWeights [1,1,1,1]
   }
   n11 = "Constraint - base_socks_011_r (PPPP)"!(
    ComponentPose=@n10
    Alpha=1
    ConstraintWeights_0=0.5357
    ConstraintWeights_1:float=0.3626
    ConstraintWeights_2:float=0.0944
    ConstraintWeights_3:float=0.0073
   ) -> (Pose) {
    alphaBoolBlend {blendOption=Linear}
    boneToModify {boneName=base_socks_011_r}
    constraintSetup [
     {targetBone {boneName=calf_twist_01_r}}
     {targetBone {boneName=calf_twist_02_r}}
     {targetBone {boneName=calf_r}}
     {targetBone {boneName=foot_r}}
    ]
    constraintWeights [1,1,1,1]
   }
   n12 = "Constraint - base_socks_012_r (PPP)"!(ComponentPose=@n11, Alpha=1, ConstraintWeights_0=0.7935, ConstraintWeights_1:float=0.1069, ConstraintWeights_2:float=0.0996) -> (Pose) {
    alphaBoolBlend {blendOption=Linear}
    boneToModify {boneName=base_socks_012_r}
    constraintSetup [{targetBone {boneName=calf_twist_02_r}}, {targetBone {boneName=calf_twist_01_r}}, {targetBone {boneName=calf_r}}]
    constraintWeights [1,1,1]
   }
   n13 = "Constraint - base_socks_013_r (PPP)"!(ComponentPose=@n12, Alpha=1, ConstraintWeights_0=0.7621, ConstraintWeights_1:float=0.1398, ConstraintWeights_2:float=0.0981) -> (Pose) {
    alphaBoolBlend {blendOption=Linear}
    boneToModify {boneName=base_socks_013_r}
    constraintSetup [{targetBone {boneName=calf_twist_02_r}}, {targetBone {boneName=calf_twist_01_r}}, {targetBone {boneName=calf_r}}]
    constraintWeights [1,1,1]
   }
   n14 = Anim Dynamics - Bone: base_socks_01_r(ComponentPose=@n13, Alpha=1, ExternalForce="0,0,0") {
    angularDampingOverride=1
    angularSpringConstant=20000
    bAngularSpring=true
    bOverrideAngularDamping=true
    bUseGravityOverride=true
    gravityScale=0.35
    simulationSpace=RootRelative
    alphaBoolBlend {blendOption=Linear}
    boundBone {boneName=base_socks_01_r}
    componentAppliedLinearAccClamp=<3000 3000 3000>
    componentLinearAccScale=<0.3 0.3 0.3>
    componentLinearVelScale=<0.3 0.3 0.3>
    gravityOverride {z=-980}
    physicsBodyDefinitions [{
      boundBone {boneName=base_socks_01_r}
      boxExtents=<20 20 20>
      constraintSetup {angularLimitsMax=<50 20 15>, angularLimitsMin=<-50 -20 -15>, angularTarget=<1 0 0>}
      localJointOffset=<4 0 0>
     }]
   }
   n15 = Anim Dynamics - Bone: base_socks_02_r(ComponentPose=@n14, Alpha=1, ExternalForce="0,0,0") {
    angularDampingOverride=1
    angularSpringConstant=20000
    bAngularSpring=true
    bOverrideAngularDamping=true
    bUseGravityOverride=true
    gravityScale=0.35
    simulationSpace=RootRelative
    alphaBoolBlend {blendOption=Linear}
    boundBone {boneName=base_socks_02_r}
    componentAppliedLinearAccClamp=<3000 3000 3000>
    componentLinearAccScale=<0.3 0.3 0.3>
    componentLinearVelScale=<0.3 0.3 0.3>
    gravityOverride {z=-980}
    physicsBodyDefinitions [{
      boundBone {boneName=base_socks_02_r}
      boxExtents=<20 20 20>
      constraintSetup {angularLimitsMax=<50 20 20>, angularLimitsMin=<-50 -20 -20>, angularTarget=<1 0 0>}
      localJointOffset=<4 0 0>
     }]
   }
   n16 = Anim Dynamics - Bone: base_socks_03_r(ComponentPose=@n15, Alpha=1, ExternalForce="0,0,0") {
    angularDampingOverride=1
    angularSpringConstant=20000
    bAngularSpring=true
    bOverrideAngularDamping=true
    bUseGravityOverride=true
    gravityScale=0.35
    simulationSpace=RootRelative
    alphaBoolBlend {blendOption=Linear}
    boundBone {boneName=base_socks_03_r}
    componentAppliedLinearAccClamp=<3000 3000 3000>
    componentLinearAccScale=<0.3 0.3 0.3>
    componentLinearVelScale=<0.3 0.3 0.3>
    gravityOverride {z=-980}
    physicsBodyDefinitions [{
      boundBone {boneName=base_socks_03_r}
      boxExtents=<20 20 20>
      constraintSetup {angularLimitsMax=<50 20 20>, angularLimitsMin=<-50 -20 -20>, angularTarget=<1 0 0>}
      localJointOffset=<4 0 0>
     }]
   }
   n17 = Anim Dynamics - Bone: base_socks_04_r(ComponentPose=@n16, Alpha=1, ExternalForce="0,0,0") {
    angularDampingOverride=1
    angularSpringConstant=20000
    bAngularSpring=true
    bOverrideAngularDamping=true
    bUseGravityOverride=true
    gravityScale=0.35
    simulationSpace=RootRelative
    alphaBoolBlend {blendOption=Linear}
    boundBone {boneName=base_socks_04_r}
    componentAppliedLinearAccClamp=<3000 3000 3000>
    componentLinearAccScale=<0.3 0.3 0.3>
    componentLinearVelScale=<0.3 0.3 0.3>
    gravityOverride {z=-980}
    physicsBodyDefinitions [{
      boundBone {boneName=base_socks_04_r}
      boxExtents=<20 20 20>
      constraintSetup {angularLimitsMax=<50 20 20>, angularLimitsMin=<-50 -20 -20>, angularTarget=<1 0 0>}
      localJointOffset=<4 0 0>
     }]
   }
   n18 = Anim Dynamics - Bone: base_socks_05_r(ComponentPose=@n17, Alpha=1, ExternalForce="0,0,0") {
    angularDampingOverride=1
    angularSpringConstant=20000
    bAngularSpring=true
    bOverrideAngularDamping=true
    bUseGravityOverride=true
    gravityScale=0.35
    simulationSpace=RootRelative
    alphaBoolBlend {blendOption=Linear}
    boundBone {boneName=base_socks_05_r}
    componentAppliedLinearAccClamp=<3000 3000 3000>
    componentLinearAccScale=<0.3 0.3 0.3>
    componentLinearVelScale=<0.3 0.3 0.3>
    gravityOverride {z=-980}
    physicsBodyDefinitions [{
      boundBone {boneName=base_socks_05_r}
      boxExtents=<20 20 20>
      constraintSetup {angularLimitsMax=<50 20 20>, angularLimitsMin=<-50 -20 -20>, angularTarget=<1 0 0>}
      localJointOffset=<4 0 0>
     }]
   }
   n19 = Anim Dynamics - Bone: base_socks_06_r(ComponentPose=@n18, Alpha=1, ExternalForce="0,0,0") {
    angularDampingOverride=1
    angularSpringConstant=20000
    bAngularSpring=true
    bOverrideAngularDamping=true
    bUseGravityOverride=true
    gravityScale=0.35
    simulationSpace=RootRelative
    alphaBoolBlend {blendOption=Linear}
    boundBone {boneName=base_socks_06_r}
    componentAppliedLinearAccClamp=<3000 3000 3000>
    componentLinearAccScale=<0.3 0.3 0.3>
    componentLinearVelScale=<0.3 0.3 0.3>
    gravityOverride {z=-980}
    physicsBodyDefinitions [{
      boundBone {boneName=base_socks_06_r}
      boxExtents=<20 20 20>
      constraintSetup {angularLimitsMax=<50 20 15>, angularLimitsMin=<-50 -20 -15>, angularTarget=<1 0 0>}
      localJointOffset=<4 0 0>
     }]
   }
   n20 = Anim Dynamics - Bone: base_socks_07_r(ComponentPose=@n19, Alpha=1, ExternalForce="0,0,0") {
    angularDampingOverride=1
    angularSpringConstant=20000
    bAngularSpring=true
    bOverrideAngularDamping=true
    bUseGravityOverride=true
    gravityScale=0.35
    simulationSpace=RootRelative
    alphaBoolBlend {blendOption=Linear}
    boundBone {boneName=base_socks_07_r}
    componentAppliedLinearAccClamp=<3000 3000 3000>
    componentLinearAccScale=<0.3 0.3 0.3>
    componentLinearVelScale=<0.3 0.3 0.3>
    gravityOverride {z=-980}
    physicsBodyDefinitions [{
      boundBone {boneName=base_socks_07_r}
      boxExtents=<20 20 20>
      constraintSetup {angularLimitsMax=<50 20 20>, angularLimitsMin=<-50 -20 -20>, angularTarget=<1 0 0>}
      localJointOffset=<4 0 0>
     }]
   }
   n21 = Anim Dynamics - Bone: base_socks_08_r(ComponentPose=@n20, Alpha=1, ExternalForce="0,0,0") {
    angularDampingOverride=1
    angularSpringConstant=20000
    bAngularSpring=true
    bOverrideAngularDamping=true
    bUseGravityOverride=true
    gravityScale=0.35
    simulationSpace=RootRelative
    alphaBoolBlend {blendOption=Linear}
    boundBone {boneName=base_socks_08_r}
    componentAppliedLinearAccClamp=<3000 3000 3000>
    componentLinearAccScale=<0.3 0.3 0.3>
    componentLinearVelScale=<0.3 0.3 0.3>
    gravityOverride {z=-980}
    physicsBodyDefinitions [{
      boundBone {boneName=base_socks_08_r}
      boxExtents=<20 20 20>
      constraintSetup {angularLimitsMax=<50 20 15>, angularLimitsMin=<-50 -20 -15>, angularTarget=<1 0 0>}
      localJointOffset=<4 0 0>
     }]
   }
   n22 = Anim Dynamics - Bone: base_socks_09_r(ComponentPose=@n21, Alpha=1, ExternalForce="0,0,0") {
    angularDampingOverride=1
    angularSpringConstant=20000
    bAngularSpring=true
    bOverrideAngularDamping=true
    bUseGravityOverride=true
    gravityScale=0.35
    simulationSpace=RootRelative
    alphaBoolBlend {blendOption=Linear}
    boundBone {boneName=base_socks_09_r}
    componentAppliedLinearAccClamp=<3000 3000 3000>
    componentLinearAccScale=<0.3 0.3 0.3>
    componentLinearVelScale=<0.3 0.3 0.3>
    gravityOverride {z=-980}
    physicsBodyDefinitions [{
      boundBone {boneName=base_socks_09_r}
      boxExtents=<20 20 20>
      constraintSetup {angularLimitsMax=<50 20 20>, angularLimitsMin=<-50 -20 -20>, angularTarget=<1 0 0>}
      localJointOffset=<4 0 0>
     }]
   }
   n23 = Anim Dynamics - Bone: base_socks_010_r(ComponentPose=@n22, Alpha=1, ExternalForce="0,0,0") {
    angularDampingOverride=1
    angularSpringConstant=20000
    bAngularSpring=true
    bOverrideAngularDamping=true
    bUseGravityOverride=true
    gravityScale=0.35
    simulationSpace=RootRelative
    alphaBoolBlend {blendOption=Linear}
    boundBone {boneName=base_socks_010_r}
    componentAppliedLinearAccClamp=<3000 3000 3000>
    componentLinearAccScale=<0.3 0.3 0.3>
    componentLinearVelScale=<0.3 0.3 0.3>
    gravityOverride {z=-980}
    physicsBodyDefinitions [{
      boundBone {boneName=base_socks_010_r}
      boxExtents=<20 20 20>
      constraintSetup {angularLimitsMax=<50 20 15>, angularLimitsMin=<-50 -20 -15>, angularTarget=<1 0 0>}
      localJointOffset=<4 0 0>
     }]
   }
   n24 = Anim Dynamics - Bone: base_socks_011_r(ComponentPose=@n23, Alpha=1, ExternalForce="0,0,0") {
    angularDampingOverride=1
    angularSpringConstant=20000
    bAngularSpring=true
    bOverrideAngularDamping=true
    bUseGravityOverride=true
    gravityScale=0.35
    simulationSpace=RootRelative
    alphaBoolBlend {blendOption=Linear}
    boundBone {boneName=base_socks_011_r}
    componentAppliedLinearAccClamp=<3000 3000 3000>
    componentLinearAccScale=<0.3 0.3 0.3>
    componentLinearVelScale=<0.3 0.3 0.3>
    gravityOverride {z=-980}
    physicsBodyDefinitions [{
      boundBone {boneName=base_socks_011_r}
      boxExtents=<20 20 20>
      constraintSetup {angularLimitsMax=<50 20 20>, angularLimitsMin=<-50 -20 -20>, angularTarget=<1 0 0>}
      localJointOffset=<4 0 0>
     }]
   }
   n25 = Anim Dynamics - Bone: base_socks_012_r(ComponentPose=@n24, Alpha=1, ExternalForce="0,0,0") {
    angularDampingOverride=1
    angularSpringConstant=20000
    bAngularSpring=true
    bOverrideAngularDamping=true
    bUseGravityOverride=true
    gravityScale=0.35
    simulationSpace=RootRelative
    alphaBoolBlend {blendOption=Linear}
    boundBone {boneName=base_socks_012_r}
    componentAppliedLinearAccClamp=<3000 3000 3000>
    componentLinearAccScale=<0.3 0.3 0.3>
    componentLinearVelScale=<0.3 0.3 0.3>
    gravityOverride {z=-980}
    physicsBodyDefinitions [{
      boundBone {boneName=base_socks_012_r}
      boxExtents=<20 20 20>
      constraintSetup {angularLimitsMax=<50 20 20>, angularLimitsMin=<-50 -20 -20>, angularTarget=<1 0 0>}
      localJointOffset=<4 0 0>
     }]
   }
   n26 = Anim Dynamics - Bone: base_socks_013_r(ComponentPose=@n25, Alpha=1, ExternalForce="0,0,0") {
    angularDampingOverride=1
    angularSpringConstant=20000
    bAngularSpring=true
    bOverrideAngularDamping=true
    bUseGravityOverride=true
    gravityScale=0.35
    simulationSpace=RootRelative
    alphaBoolBlend {blendOption=Linear}
    boundBone {boneName=base_socks_013_r}
    componentAppliedLinearAccClamp=<3000 3000 3000>
    componentLinearAccScale=<0.3 0.3 0.3>
    componentLinearVelScale=<0.3 0.3 0.3>
    gravityOverride {z=-980}
    physicsBodyDefinitions [{
      boundBone {boneName=base_socks_013_r}
      boxExtents=<20 20 20>
      constraintSetup {angularLimitsMax=<50 20 15>, angularLimitsMin=<-50 -20 -15>, angularTarget=<1 0 0>}
      localJointOffset=<4 0 0>
     }]
   }
   Output Pose(Result=Component To Local(ComponentPose=@n26)) {layerGroup=None}
   n27 = Control Rig!(Source=Copy Pose From Mesh() {bUseAttachedParent=true}, Alpha, IsDBNO:bool=Get bIsDBNO()) -> (Pose) {
    alphaBoolBlend {blendOption=Linear}
    controlRigAssetReference {blueprintRigClass=&/F_LRG_PowerFluff_CtrlRig.F_LRG_PowerFluff_CtrlRig_C}
   }
   n28 = "Constraint - base_carrot_acc_01_r (PPPP)"!(
    ComponentPose=Local To Component(LocalPose=@n27)
    Alpha=1
    ConstraintWeights_0=0.9653
    ConstraintWeights_1:float=0.0148
    ConstraintWeights_2:float=0.0106
    ConstraintWeights_3:float=0.0093
   ) -> (Pose) {
    alphaBoolBlend {blendOption=Linear}
    boneToModify {boneName=base_carrot_acc_01_r}
    constraintSetup [
     {targetBone {boneName=lowerarm_twist_01_r}}
     {targetBone {boneName=lowerarm_r}}
     {targetBone {boneName=hand_r}}
     {targetBone {boneName=lowerarm_twist_02_r}}
    ]
    constraintWeights [1,1,1,1]
   }
   n29 = Anim Dynamics - Bone: base_carrot_acc_01_r!(SimulationSpace:AnimPhysSimSpaceType Enum, ComponentPose=@n28, Alpha=1, GravityOverride:Vector="0,0,-980", ExternalForce="0,0,0") -> (Pose) {
    angularDampingOverride=1
    angularSpringConstant=200
    bAngularSpring=true
    bOverrideAngularDamping=true
    bUseGravityOverride=true
    gravityScale=0.35
    simulationSpace=RootRelative
    alphaBoolBlend {blendOption=Linear}
    boundBone {boneName=base_carrot_acc_01_r}
    componentAppliedLinearAccClamp=<3000 3000 3000>
    componentLinearAccScale=<1 1 1>
    componentLinearVelScale=<1 1 1>
    gravityOverride {z=-980}
    physicsBodyDefinitions [{
      boundBone {boneName=base_carrot_acc_01_r}
      boxExtents=<30 30 30>
      constraintSetup {angularLimitsMax=<20 20 20>, angularLimitsMin=<-20 -20 -20>, angularTarget=<1 0 0>}
      localJointOffset=<5 0 0>
     }]
    relativeSpaceBone {boneName=hand_r}
   }
   Save cached pose 'carrot_cache'(Pose=Component To Local(ComponentPose=@n29)) {cacheName=carrot_cache}
   n30 = Anim Dynamics - Bone: dyn_rabit_back_r!(
    SimulationSpace:AnimPhysSimSpaceType Enum
    ComponentPose=Local To Component(LocalPose=Use cached pose 'carrot_cache'())
    Alpha=1
    GravityOverride:Vector="0,0,-980"
    ExternalForce="0,0,0"
   ) -> (Pose) {
    angularDampingOverride=1
    angularSpringConstant=8000
    bAngularSpring=true
    bOverrideAngularDamping=true
    bUseGravityOverride=true
    gravityScale=0.35
    simulationSpace=RootRelative
    alphaBoolBlend {blendOption=Linear}
    boundBone {boneName=dyn_rabit_back_r}
    componentAppliedLinearAccClamp=<3000 3000 3000>
    componentLinearAccScale=<0.5 0.5 0.5>
    componentLinearVelScale=<0.5 0.5 0.5>
    gravityOverride {z=-980}
    physicsBodyDefinitions [{
      boundBone {boneName=dyn_rabit_back_r}
      boxExtents=<20 20 20>
      constraintSetup {angularLimitsMax=<20 0 5>, angularLimitsMin=<-20 -30 -5>, angularTarget=<1 0 0>}
      localJointOffset=<4 0 0>
     }]
    relativeSpaceBone {boneName=pelvis}
   }
   n31 = Anim Dynamics - Bone: dyn_rabit_front_r!(SimulationSpace:AnimPhysSimSpaceType Enum, ComponentPose=@n30, Alpha=1, GravityOverride:Vector="0,0,-980", ExternalForce="0,0,0") -> (Pose) {
    angularDampingOverride=1
    angularSpringConstant=8000
    bAngularSpring=true
    bOverrideAngularDamping=true
    bUseGravityOverride=true
    gravityScale=0.35
    simulationSpace=RootRelative
    alphaBoolBlend {blendOption=Linear}
    boundBone {boneName=dyn_rabit_front_r}
    componentAppliedLinearAccClamp=<3000 3000 3000>
    componentLinearAccScale=<0.5 0.5 0.5>
    componentLinearVelScale=<0.5 0.5 0.5>
    gravityOverride {z=-980}
    physicsBodyDefinitions [{
      boundBone {boneName=dyn_rabit_front_r}
      boxExtents=<20 20 20>
      constraintSetup {angularLimitsMax=<20 0 50>, angularLimitsMin=<-20 -30 -5>, angularTarget=<1 0 0>}
      localJointOffset=<4 0 0>
     }]
    relativeSpaceBone {boneName=pelvis}
   }
   n32 = Anim Dynamics - Bone: dyn_grenade_front_r!(SimulationSpace:AnimPhysSimSpaceType Enum, ComponentPose=@n31, Alpha=1, GravityOverride:Vector="0,0,-980", ExternalForce="0,0,0") -> (Pose) {
    angularDampingOverride=1
    angularSpringConstant=8000
    bAngularSpring=true
    bOverrideAngularDamping=true
    bUseGravityOverride=true
    gravityScale=0.35
    simulationSpace=RootRelative
    alphaBoolBlend {blendOption=Linear}
    boundBone {boneName=dyn_grenade_front_r}
    componentAppliedLinearAccClamp=<3000 3000 3000>
    componentLinearAccScale=<0.5 0.5 0.5>
    componentLinearVelScale=<0.5 0.5 0.5>
    gravityOverride {z=-980}
    physicsBodyDefinitions [{
      boundBone {boneName=dyn_grenade_front_r}
      boxExtents=<20 20 20>
      constraintSetup {angularLimitsMax=<20 0 5>, angularLimitsMin=<-20 -25 -5>, angularTarget=<1 0 0>}
      localJointOffset=<4 0 0>
     }]
    relativeSpaceBone {boneName=pelvis}
   }
   n33 = Anim Dynamics - Bone: dyn_grenade_center_r!(SimulationSpace:AnimPhysSimSpaceType Enum, ComponentPose=@n32, Alpha=1, GravityOverride:Vector="0,0,-980", ExternalForce="0,0,0") -> (Pose) {
    angularDampingOverride=1
    angularSpringConstant=8000
    bAngularSpring=true
    bOverrideAngularDamping=true
    bUseGravityOverride=true
    gravityScale=0.35
    simulationSpace=RootRelative
    alphaBoolBlend {blendOption=Linear}
    boundBone {boneName=dyn_grenade_center_r}
    componentAppliedLinearAccClamp=<3000 3000 3000>
    componentLinearAccScale=<0.5 0.5 0.5>
    componentLinearVelScale=<0.5 0.5 0.5>
    gravityOverride {z=-980}
    physicsBodyDefinitions [{
      boundBone {boneName=dyn_grenade_center_r}
      boxExtents=<20 20 20>
      constraintSetup {angularLimitsMax=<20 0 5>, angularLimitsMin=<-20 -30 -5>, angularTarget=<1 0 0>}
      localJointOffset=<4 0 0>
     }]
    relativeSpaceBone {boneName=pelvis}
   }
   n34 = Anim Dynamics - Bone: dyn_grenade_back_r!(SimulationSpace:AnimPhysSimSpaceType Enum, ComponentPose=@n33, Alpha=1, GravityOverride:Vector="0,0,-980", ExternalForce="0,0,0") -> (Pose) {
    angularDampingOverride=1
    angularSpringConstant=8000
    bAngularSpring=true
    bOverrideAngularDamping=true
    bUseGravityOverride=true
    gravityScale=0.35
    simulationSpace=RootRelative
    alphaBoolBlend {blendOption=Linear}
    boundBone {boneName=dyn_grenade_back_r}
    componentAppliedLinearAccClamp=<3000 3000 3000>
    componentLinearAccScale=<0.5 0.5 0.5>
    componentLinearVelScale=<0.5 0.5 0.5>
    gravityOverride {z=-980}
    physicsBodyDefinitions [{
      boundBone {boneName=dyn_grenade_back_r}
      boxExtents=<20 20 20>
      constraintSetup {angularLimitsMax=<20 0 5>, angularLimitsMin=<-20 -22 -5>, angularTarget=<1 0 0>}
      localJointOffset=<4 0 0>
     }]
    relativeSpaceBone {boneName=pelvis}
   }
   Save cached pose 'grenades belt cache'(Pose=Component To Local(ComponentPose=@n34)) {cacheName=grenades belt cache}
   n35 = "Constraint - base_grenade_fr_l (PPPPPPP)"!(
    ComponentPose=Local To Component(LocalPose=Use cached pose 'grenades belt cache'())
    Alpha=1
    ConstraintWeights_0=0.9105
    ConstraintWeights_1:float=0.083
    ConstraintWeights_2:float=0.0041
    ConstraintWeights_3:float=0.001
    ConstraintWeights_4:float=0.0008
    ConstraintWeights_5:float=0.0004
    ConstraintWeights_6:float=0.0002
   ) -> (Pose) {
    alphaBoolBlend {blendOption=Linear}
    boneToModify {boneName=base_grenade_fr_l}
    constraintSetup [
     {targetBone {boneName=thigh_l}}
     {targetBone {boneName=thigh_twist_01_l}}
     {targetBone {boneName=pelvis}}
     {targetBone {boneName=thigh_twist_01_r}}
     {targetBone {boneName=thigh_r}}
     {targetBone {boneName=calf_l}}
     {targetBone {boneName=deform_knee_out_l}}
    ]
    constraintWeights [1,1,1,1,1,1,1]
   }
   n36 = "Constraint - base_grenade_cn_l (PPPPPPP)"!(
    ComponentPose=@n35
    Alpha=1
    ConstraintWeights_0=0.9022
    ConstraintWeights_1:float=0.0924
    ConstraintWeights_2:float=0.004
    ConstraintWeights_3:float=0.0006
    ConstraintWeights_4:float=0.0003
    ConstraintWeights_5:float=0.0003
    ConstraintWeights_6:float=0.0001
   ) -> (Pose) {
    alphaBoolBlend {blendOption=Linear}
    boneToModify {boneName=base_grenade_cn_l}
    constraintSetup [
     {targetBone {boneName=thigh_l}}
     {targetBone {boneName=thigh_twist_01_l}}
     {targetBone {boneName=pelvis}}
     {targetBone {boneName=calf_l}}
     {targetBone {boneName=thigh_twist_01_r}}
     {targetBone {boneName=thigh_r}}
     {targetBone {boneName=deform_knee_out_l}}
    ]
    constraintWeights [1,1,1,1,1,1,1]
   }
   n37 = "Constraint - base_grenade_br_l (PPPPP)"!(
    ComponentPose=@n36
    Alpha=1
    ConstraintWeights_0=0.894
    ConstraintWeights_1:float=0.1024
    ConstraintWeights_2:float=0.0022
    ConstraintWeights_3:float=0.0009
    ConstraintWeights_4:float=0.0004
   ) -> (Pose) {
    alphaBoolBlend {blendOption=Linear}
    boneToModify {boneName=base_grenade_br_l}
    constraintSetup [
     {targetBone {boneName=thigh_l}}
     {targetBone {boneName=thigh_twist_01_l}}
     {targetBone {boneName=pelvis}}
     {targetBone {boneName=calf_l}}
     {targetBone {boneName=deform_knee_in_l}}
    ]
    constraintWeights [1,1,1,1,1]
   }
   n38 = Anim Dynamics - Bone: base_grenade_fr_l!(SimulationSpace:AnimPhysSimSpaceType Enum, ComponentPose=@n37, Alpha=1, GravityOverride:Vector="0,0,-980", ExternalForce="0,0,0") -> (Pose) {
    angularDampingOverride=1
    angularSpringConstant=12000
    bAngularSpring=true
    bOverrideAngularDamping=true
    bUseGravityOverride=true
    gravityScale=0.35
    simulationSpace=RootRelative
    alphaBoolBlend {blendOption=Linear}
    boundBone {boneName=base_grenade_fr_l}
    componentAppliedLinearAccClamp=<3000 3000 3000>
    componentLinearAccScale=<0.5 0.5 0.5>
    componentLinearVelScale=<0.5 0.5 0.5>
    gravityOverride {z=-980}
    physicsBodyDefinitions [{
      boundBone {boneName=base_grenade_fr_l}
      boxExtents=<20 20 20>
      constraintSetup {angularLimitsMax=<20 5 30>, angularLimitsMin=<-20 -30 -40>, angularTarget=<1 0 0>}
      localJointOffset=<4 0 0>
     }]
   }
   n39 = Anim Dynamics - Bone: base_grenade_cn_l!(SimulationSpace:AnimPhysSimSpaceType Enum, ComponentPose=@n38, Alpha=1, GravityOverride:Vector="0,0,-980", ExternalForce="0,0,0") -> (Pose) {
    angularDampingOverride=1
    angularSpringConstant=12000
    bAngularSpring=true
    bOverrideAngularDamping=true
    bUseGravityOverride=true
    gravityScale=0.35
    simulationSpace=RootRelative
    alphaBoolBlend {blendOption=Linear}
    boundBone {boneName=base_grenade_cn_l}
    componentAppliedLinearAccClamp=<3000 3000 3000>
    componentLinearAccScale=<0.5 0.5 0.5>
    componentLinearVelScale=<0.5 0.5 0.5>
    gravityOverride {z=-980}
    physicsBodyDefinitions [{
      boundBone {boneName=base_grenade_cn_l}
      boxExtents=<20 20 20>
      constraintSetup {angularLimitsMax=<10 5 30>, angularLimitsMin=<-10 -28 -25>, angularTarget=<1 0 0>}
      localJointOffset=<4 0 0>
     }]
   }
   n40 = Anim Dynamics - Bone: base_grenade_br_l!(SimulationSpace:AnimPhysSimSpaceType Enum, ComponentPose=@n39, Alpha=1, GravityOverride:Vector="0,0,-980", ExternalForce="0,0,0") -> (Pose) {
    angularDampingOverride=1
    angularSpringConstant=12000
    bAngularSpring=true
    bOverrideAngularDamping=true
    bUseGravityOverride=true
    gravityScale=0.35
    simulationSpace=RootRelative
    alphaBoolBlend {blendOption=Linear}
    boundBone {boneName=base_grenade_br_l}
    componentAppliedLinearAccClamp=<3000 3000 3000>
    componentLinearAccScale=<0.5 0.5 0.5>
    componentLinearVelScale=<0.5 0.5 0.5>
    gravityOverride {z=-980}
    physicsBodyDefinitions [{
      boundBone {boneName=base_grenade_br_l}
      boxExtents=<20 20 20>
      constraintSetup {angularLimitsMax=<20 5 30>, angularLimitsMin=<-20 -27 -25>, angularTarget=<1 0 0>}
      localJointOffset=<4 0 0>
     }]
   }
   Save cached pose 'grenades_thing_L cache'(Pose=Component To Local(ComponentPose=@n40)) {cacheName=grenades_thing_L cache}
   n41 = "Constraint - base_socks_01_l (PPPP)"!(
    ComponentPose=Local To Component(LocalPose=Use cached pose 'grenades_thing_L cache'())
    Alpha=1
    ConstraintWeights_0=0.5858
    ConstraintWeights_1:float=0.3898
    ConstraintWeights_2:float=0.0167
    ConstraintWeights_3:float=0.0077
   ) -> (Pose) {
    alphaBoolBlend {blendOption=Linear}
    boneToModify {boneName=base_socks_01_l}
    constraintSetup [
     {targetBone {boneName=foot_l}}
     {targetBone {boneName=calf_twist_01_l}}
     {targetBone {boneName=calf_twist_02_l}}
     {targetBone {boneName=calf_l}}
    ]
    constraintWeights [1,1,1,1]
   }
   n42 = "Constraint - base_socks_02_l (PPPP)"!(
    ComponentPose=@n41
    Alpha=1
    ConstraintWeights_0=0.5244
    ConstraintWeights_1:float=0.3056
    ConstraintWeights_2:float=0.0853
    ConstraintWeights_3:float=0.0848
   ) -> (Pose) {
    alphaBoolBlend {blendOption=Linear}
    boneToModify {boneName=base_socks_02_l}
    constraintSetup [
     {targetBone {boneName=calf_twist_01_l}}
     {targetBone {boneName=calf_twist_02_l}}
     {targetBone {boneName=calf_l}}
     {targetBone {boneName=foot_l}}
    ]
    constraintWeights [1,1,1,1]
   }
   n43 = "Constraint - base_socks_03_l (PPP)"!(ComponentPose=@n42, Alpha=1, ConstraintWeights_0=0.7187, ConstraintWeights_1:float=0.1747, ConstraintWeights_2:float=0.1066) -> (Pose) {
    alphaBoolBlend {blendOption=Linear}
    boneToModify {boneName=base_socks_03_l}
    constraintSetup [{targetBone {boneName=calf_twist_02_l}}, {targetBone {boneName=calf_twist_01_l}}, {targetBone {boneName=calf_l}}]
    constraintWeights [1,1,1]
   }
   n44 = "Constraint - base_socks_04_l (PPPP)"!(
    ComponentPose=@n43
    Alpha=1
    ConstraintWeights_0=0.838
    ConstraintWeights_1:float=0.1421
    ConstraintWeights_2:float=0.0138
    ConstraintWeights_3:float=0.0061
   ) -> (Pose) {
    alphaBoolBlend {blendOption=Linear}
    boneToModify {boneName=base_socks_04_l}
    constraintSetup [
     {targetBone {boneName=foot_l}}
     {targetBone {boneName=calf_twist_01_l}}
     {targetBone {boneName=calf_twist_02_l}}
     {targetBone {boneName=calf_l}}
    ]
    constraintWeights [1,1,1,1]
   }
   n45 = "Constraint - base_socks_05_l (PPPP)"!(
    ComponentPose=@n44
    Alpha=1
    ConstraintWeights_0=0.8235
    ConstraintWeights_1:float=0.1608
    ConstraintWeights_2:float=0.0118
    ConstraintWeights_3:float=0.004
   ) -> (Pose) {
    alphaBoolBlend {blendOption=Linear}
    boneToModify {boneName=base_socks_05_l}
    constraintSetup [
     {targetBone {boneName=foot_l}}
     {targetBone {boneName=calf_twist_01_l}}
     {targetBone {boneName=calf_twist_02_l}}
     {targetBone {boneName=calf_l}}
    ]
    constraintWeights [1,1,1,1]
   }
   n46 = "Constraint - base_socks_06_l (PPPP)"!(
    ComponentPose=@n45
    Alpha=1
    ConstraintWeights_0=0.5023
    ConstraintWeights_1:float=0.3896
    ConstraintWeights_2:float=0.0996
    ConstraintWeights_3:float=0.0085
   ) -> (Pose) {
    alphaBoolBlend {blendOption=Linear}
    boneToModify {boneName=base_socks_06_l}
    constraintSetup [
     {targetBone {boneName=calf_twist_01_l}}
     {targetBone {boneName=calf_twist_02_l}}
     {targetBone {boneName=calf_l}}
     {targetBone {boneName=foot_l}}
    ]
    constraintWeights [1,1,1,1]
   }
   n47 = "Constraint - base_socks_07_l (PP)"!(ComponentPose=@n46, Alpha=1, ConstraintWeights_0=0.8883, ConstraintWeights_1:float=0.1117) -> (Pose) {
    alphaBoolBlend {blendOption=Linear}
    boneToModify {boneName=base_socks_07_l}
    constraintSetup [{targetBone {boneName=foot_l}}, {targetBone {boneName=calf_twist_01_l}}]
    constraintWeights [1,1]
   }
   n48 = "Constraint - base_socks_08_l (PPPPP)"!(
    ComponentPose=@n47
    Alpha=1
    ConstraintWeights_0=0.6585
    ConstraintWeights_1:float=0.2239
    ConstraintWeights_2:float=0.0559
    ConstraintWeights_3:float=0.0466
    ConstraintWeights_4:float=0.0151
   ) -> (Pose) {
    alphaBoolBlend {blendOption=Linear}
    boneToModify {boneName=base_socks_08_l}
    constraintSetup [
     {targetBone {boneName=foot_l}}
     {targetBone {boneName=calf_twist_01_l}}
     {targetBone {boneName=ball_l}}
     {targetBone {boneName=calf_twist_02_l}}
     {targetBone {boneName=calf_l}}
    ]
    constraintWeights [1,1,1,1,1]
   }
   n49 = "Constraint - base_socks_09_l (PPPPP)"!(
    ComponentPose=@n48
    Alpha=1
    ConstraintWeights_0=0.5017
    ConstraintWeights_1:float=0.3771
    ConstraintWeights_2:float=0.0629
    ConstraintWeights_3:float=0.0358
    ConstraintWeights_4:float=0.0224
   ) -> (Pose) {
    alphaBoolBlend {blendOption=Linear}
    boneToModify {boneName=base_socks_09_l}
    constraintSetup [
     {targetBone {boneName=foot_l}}
     {targetBone {boneName=calf_twist_01_l}}
     {targetBone {boneName=calf_twist_02_l}}
     {targetBone {boneName=ball_l}}
     {targetBone {boneName=calf_l}}
    ]
    constraintWeights [1,1,1,1,1]
   }
   n50 = "Constraint - base_socks_010_l (PPPP)"!(
    ComponentPose=@n49
    Alpha=1
    ConstraintWeights_0=0.6145
    ConstraintWeights_1:float=0.2394
    ConstraintWeights_2:float=0.0759
    ConstraintWeights_3:float=0.0702
   ) -> (Pose) {
    alphaBoolBlend {blendOption=Linear}
    boneToModify {boneName=base_socks_010_l}
    constraintSetup [
     {targetBone {boneName=calf_twist_01_l}}
     {targetBone {boneName=calf_twist_02_l}}
     {targetBone {boneName=foot_l}}
     {targetBone {boneName=calf_l}}
    ]
    constraintWeights [1,1,1,1]
   }
   n51 = "Constraint - base_socks_011_l (PPPP)"!(
    ComponentPose=@n50
    Alpha=1
    ConstraintWeights_0=0.5162
    ConstraintWeights_1:float=0.3829
    ConstraintWeights_2:float=0.0934
    ConstraintWeights_3:float=0.0075
   ) -> (Pose) {
    alphaBoolBlend {blendOption=Linear}
    boneToModify {boneName=base_socks_011_l}
    constraintSetup [
     {targetBone {boneName=calf_twist_01_l}}
     {targetBone {boneName=calf_twist_02_l}}
     {targetBone {boneName=calf_l}}
     {targetBone {boneName=foot_l}}
    ]
    constraintWeights [1,1,1,1]
   }
   n52 = "Constraint - base_socks_012_l (PPP)"!(ComponentPose=@n51, Alpha=1, ConstraintWeights_0=0.7935, ConstraintWeights_1:float=0.1068, ConstraintWeights_2:float=0.0996) -> (Pose) {
    alphaBoolBlend {blendOption=Linear}
    boneToModify {boneName=base_socks_012_l}
    constraintSetup [{targetBone {boneName=calf_twist_02_l}}, {targetBone {boneName=calf_twist_01_l}}, {targetBone {boneName=calf_l}}]
    constraintWeights [1,1,1]
   }
   n53 = "Constraint - base_socks_013_l (PPP)"!(ComponentPose=@n52, Alpha=1, ConstraintWeights_0=0.7621, ConstraintWeights_1:float=0.1398, ConstraintWeights_2:float=0.0981) -> (Pose) {
    alphaBoolBlend {blendOption=Linear}
    boneToModify {boneName=base_socks_013_l}
    constraintSetup [{targetBone {boneName=calf_twist_02_l}}, {targetBone {boneName=calf_twist_01_l}}, {targetBone {boneName=calf_l}}]
    constraintWeights [1,1,1]
   }
   n54 = Anim Dynamics - Bone: base_socks_01_l(ComponentPose=@n53, Alpha=1, ExternalForce="0,0,0") {
    angularDampingOverride=1
    angularSpringConstant=20000
    bAngularSpring=true
    bOverrideAngularDamping=true
    bUseGravityOverride=true
    gravityScale=0.35
    simulationSpace=RootRelative
    alphaBoolBlend {blendOption=Linear}
    boundBone {boneName=base_socks_01_l}
    componentAppliedLinearAccClamp=<3000 3000 3000>
    componentLinearAccScale=<0.3 0.3 0.3>
    componentLinearVelScale=<0.3 0.3 0.3>
    gravityOverride {z=-980}
    physicsBodyDefinitions [{
      boundBone {boneName=base_socks_01_l}
      boxExtents=<20 20 20>
      constraintSetup {angularLimitsMax=<50 20 20>, angularLimitsMin=<-50 -20 -20>, angularTarget=<1 0 0>}
      localJointOffset=<4 0 0>
     }]
   }
   n55 = Anim Dynamics - Bone: base_socks_02_l(ComponentPose=@n54, Alpha=1, ExternalForce="0,0,0") {
    angularDampingOverride=1
    angularSpringConstant=20000
    bAngularSpring=true
    bOverrideAngularDamping=true
    bUseGravityOverride=true
    gravityScale=0.35
    simulationSpace=RootRelative
    alphaBoolBlend {blendOption=Linear}
    boundBone {boneName=base_socks_02_l}
    componentAppliedLinearAccClamp=<3000 3000 3000>
    componentLinearAccScale=<0.3 0.3 0.3>
    componentLinearVelScale=<0.3 0.3 0.3>
    gravityOverride {z=-980}
    physicsBodyDefinitions [{
      boundBone {boneName=base_socks_02_l}
      boxExtents=<20 20 20>
      constraintSetup {angularLimitsMax=<50 20 0>, angularLimitsMin=<-50 -20 0>, angularTarget=<1 0 0>}
      localJointOffset=<4 0 0>
     }]
   }
   n56 = Anim Dynamics - Bone: base_socks_03_l(ComponentPose=@n55, Alpha=1, ExternalForce="0,0,0") {
    angularDampingOverride=1
    angularSpringConstant=20000
    bAngularSpring=true
    bOverrideAngularDamping=true
    bUseGravityOverride=true
    gravityScale=0.35
    simulationSpace=RootRelative
    alphaBoolBlend {blendOption=Linear}
    boundBone {boneName=base_socks_03_l}
    componentAppliedLinearAccClamp=<3000 3000 3000>
    componentLinearAccScale=<0.3 0.3 0.3>
    componentLinearVelScale=<0.3 0.3 0.3>
    gravityOverride {z=-980}
    physicsBodyDefinitions [{
      boundBone {boneName=base_socks_03_l}
      boxExtents=<20 20 20>
      constraintSetup {angularLimitsMax=<50 20 0>, angularLimitsMin=<-50 -20 0>, angularTarget=<1 0 0>}
      localJointOffset=<4 0 0>
     }]
   }
   n57 = Anim Dynamics - Bone: base_socks_04_l(ComponentPose=@n56, Alpha=1, ExternalForce="0,0,0") {
    angularDampingOverride=1
    angularSpringConstant=20000
    bAngularSpring=true
    bOverrideAngularDamping=true
    bUseGravityOverride=true
    gravityScale=0.35
    simulationSpace=RootRelative
    alphaBoolBlend {blendOption=Linear}
    boundBone {boneName=base_socks_04_l}
    componentAppliedLinearAccClamp=<3000 3000 3000>
    componentLinearAccScale=<0.3 0.3 0.3>
    componentLinearVelScale=<0.3 0.3 0.3>
    gravityOverride {z=-980}
    physicsBodyDefinitions [{
      boundBone {boneName=base_socks_04_l}
      boxExtents=<20 20 20>
      constraintSetup {angularLimitsMax=<50 20 0>, angularLimitsMin=<-50 -20 0>, angularTarget=<1 0 0>}
      localJointOffset=<4 0 0>
     }]
   }
   n58 = Anim Dynamics - Bone: base_socks_05_l(ComponentPose=@n57, Alpha=1, ExternalForce="0,0,0") {
    angularDampingOverride=1
    angularSpringConstant=20000
    bAngularSpring=true
    bOverrideAngularDamping=true
    bUseGravityOverride=true
    gravityScale=0.35
    simulationSpace=RootRelative
    alphaBoolBlend {blendOption=Linear}
    boundBone {boneName=base_socks_05_l}
    componentAppliedLinearAccClamp=<3000 3000 3000>
    componentLinearAccScale=<0.3 0.3 0.3>
    componentLinearVelScale=<0.3 0.3 0.3>
    gravityOverride {z=-980}
    physicsBodyDefinitions [{
      boundBone {boneName=base_socks_05_l}
      boxExtents=<20 20 20>
      constraintSetup {angularLimitsMax=<50 20 20>, angularLimitsMin=<-50 -20 -20>, angularTarget=<1 0 0>}
      localJointOffset=<4 0 0>
     }]
   }
   n59 = Anim Dynamics - Bone: base_socks_06_l(ComponentPose=@n58, Alpha=1, ExternalForce="0,0,0") {
    angularDampingOverride=1
    angularSpringConstant=20000
    bAngularSpring=true
    bOverrideAngularDamping=true
    bUseGravityOverride=true
    gravityScale=0.35
    simulationSpace=RootRelative
    alphaBoolBlend {blendOption=Linear}
    boundBone {boneName=base_socks_06_l}
    componentAppliedLinearAccClamp=<3000 3000 3000>
    componentLinearAccScale=<0.3 0.3 0.3>
    componentLinearVelScale=<0.3 0.3 0.3>
    gravityOverride {z=-980}
    physicsBodyDefinitions [{
      boundBone {boneName=base_socks_06_l}
      boxExtents=<20 20 20>
      constraintSetup {angularLimitsMax=<50 20 20>, angularLimitsMin=<-50 -20 -20>, angularTarget=<1 0 0>}
      localJointOffset=<4 0 0>
     }]
   }
   n60 = Anim Dynamics - Bone: base_socks_07_l(ComponentPose=@n59, Alpha=1, ExternalForce="0,0,0") {
    angularDampingOverride=1
    angularSpringConstant=20000
    bAngularSpring=true
    bOverrideAngularDamping=true
    bUseGravityOverride=true
    gravityScale=0.35
    simulationSpace=RootRelative
    alphaBoolBlend {blendOption=Linear}
    boundBone {boneName=base_socks_07_l}
    componentAppliedLinearAccClamp=<3000 3000 3000>
    componentLinearAccScale=<0.3 0.3 0.3>
    componentLinearVelScale=<0.3 0.3 0.3>
    gravityOverride {z=-980}
    physicsBodyDefinitions [{
      boundBone {boneName=base_socks_07_l}
      boxExtents=<20 20 20>
      constraintSetup {angularLimitsMax=<50 20 0>, angularLimitsMin=<-50 -20 0>, angularTarget=<1 0 0>}
      localJointOffset=<4 0 0>
     }]
   }
   n61 = Anim Dynamics - Bone: base_socks_08_l(ComponentPose=@n60, Alpha=1, ExternalForce="0,0,0") {
    angularDampingOverride=1
    angularSpringConstant=20000
    bAngularSpring=true
    bOverrideAngularDamping=true
    bUseGravityOverride=true
    gravityScale=0.35
    simulationSpace=RootRelative
    alphaBoolBlend {blendOption=Linear}
    boundBone {boneName=base_socks_08_l}
    componentAppliedLinearAccClamp=<3000 3000 3000>
    componentLinearAccScale=<0.3 0.3 0.3>
    componentLinearVelScale=<0.3 0.3 0.3>
    gravityOverride {z=-980}
    physicsBodyDefinitions [{
      boundBone {boneName=base_socks_08_l}
      boxExtents=<20 20 20>
      constraintSetup {angularLimitsMax=<50 20 0>, angularLimitsMin=<-50 -20 0>, angularTarget=<1 0 0>}
      localJointOffset=<4 0 0>
     }]
   }
   n62 = Anim Dynamics - Bone: base_socks_09_l(ComponentPose=@n61, Alpha=1, ExternalForce="0,0,0") {
    angularDampingOverride=1
    angularSpringConstant=20000
    bAngularSpring=true
    bOverrideAngularDamping=true
    bUseGravityOverride=true
    gravityScale=0.35
    simulationSpace=RootRelative
    alphaBoolBlend {blendOption=Linear}
    boundBone {boneName=base_socks_09_l}
    componentAppliedLinearAccClamp=<3000 3000 3000>
    componentLinearAccScale=<0.3 0.3 0.3>
    componentLinearVelScale=<0.3 0.3 0.3>
    gravityOverride {z=-980}
    physicsBodyDefinitions [{
      boundBone {boneName=base_socks_09_l}
      boxExtents=<20 20 20>
      constraintSetup {angularLimitsMax=<50 20 20>, angularLimitsMin=<-50 -20 -20>, angularTarget=<1 0 0>}
      localJointOffset=<4 0 0>
     }]
   }
   n63 = Anim Dynamics - Bone: base_socks_010_l(ComponentPose=@n62, Alpha=1, ExternalForce="0,0,0") {
    angularDampingOverride=1
    angularSpringConstant=20000
    bAngularSpring=true
    bOverrideAngularDamping=true
    bUseGravityOverride=true
    gravityScale=0.35
    simulationSpace=RootRelative
    alphaBoolBlend {blendOption=Linear}
    boundBone {boneName=base_socks_010_l}
    componentAppliedLinearAccClamp=<3000 3000 3000>
    componentLinearAccScale=<0.3 0.3 0.3>
    componentLinearVelScale=<0.3 0.3 0.3>
    gravityOverride {z=-980}
    physicsBodyDefinitions [{
      boundBone {boneName=base_socks_010_l}
      boxExtents=<20 20 20>
      constraintSetup {angularLimitsMax=<50 20 20>, angularLimitsMin=<-50 -20 -20>, angularTarget=<1 0 0>}
      localJointOffset=<4 0 0>
     }]
   }
   n64 = Anim Dynamics - Bone: base_socks_011_l(ComponentPose=@n63, Alpha=1, ExternalForce="0,0,0") {
    angularDampingOverride=1
    angularSpringConstant=20000
    bAngularSpring=true
    bOverrideAngularDamping=true
    bUseGravityOverride=true
    gravityScale=0.35
    simulationSpace=RootRelative
    alphaBoolBlend {blendOption=Linear}
    boundBone {boneName=base_socks_011_l}
    componentAppliedLinearAccClamp=<3000 3000 3000>
    componentLinearAccScale=<0.3 0.3 0.3>
    componentLinearVelScale=<0.3 0.3 0.3>
    gravityOverride {z=-980}
    physicsBodyDefinitions [{
      boundBone {boneName=base_socks_011_l}
      boxExtents=<20 20 20>
      constraintSetup {angularLimitsMax=<50 20 0>, angularLimitsMin=<-50 -20 0>, angularTarget=<1 0 0>}
      localJointOffset=<4 0 0>
     }]
   }
   n65 = Anim Dynamics - Bone: base_socks_012_l(ComponentPose=@n64, Alpha=1, ExternalForce="0,0,0") {
    angularDampingOverride=1
    angularSpringConstant=20000
    bAngularSpring=true
    bOverrideAngularDamping=true
    bUseGravityOverride=true
    gravityScale=0.35
    simulationSpace=RootRelative
    alphaBoolBlend {blendOption=Linear}
    boundBone {boneName=base_socks_012_l}
    componentAppliedLinearAccClamp=<3000 3000 3000>
    componentLinearAccScale=<0.3 0.3 0.3>
    componentLinearVelScale=<0.3 0.3 0.3>
    gravityOverride {z=-980}
    physicsBodyDefinitions [{
      boundBone {boneName=base_socks_012_l}
      boxExtents=<20 20 20>
      constraintSetup {angularLimitsMax=<50 20 0>, angularLimitsMin=<-50 -20 0>, angularTarget=<1 0 0>}
      localJointOffset=<4 0 0>
     }]
   }
   n66 = Anim Dynamics - Bone: base_socks_013_l(ComponentPose=@n65, Alpha=1, ExternalForce="0,0,0") {
    angularDampingOverride=1
    angularSpringConstant=20000
    bAngularSpring=true
    bOverrideAngularDamping=true
    bUseGravityOverride=true
    gravityScale=0.35
    simulationSpace=RootRelative
    alphaBoolBlend {blendOption=Linear}
    boundBone {boneName=base_socks_013_l}
    componentAppliedLinearAccClamp=<3000 3000 3000>
    componentLinearAccScale=<0.3 0.3 0.3>
    componentLinearVelScale=<0.3 0.3 0.3>
    gravityOverride {z=-980}
    physicsBodyDefinitions [{
      boundBone {boneName=base_socks_013_l}
      boxExtents=<20 20 20>
      constraintSetup {angularLimitsMax=<50 20 20>, angularLimitsMin=<-50 -20 -20>, angularTarget=<1 0 0>}
      localJointOffset=<4 0 0>
     }]
   }
   Save cached pose 'socket L cache'(Pose=Component To Local(ComponentPose=@n66)) {cacheName=socket L cache}
  ]}
 EventGraph {nodes [
   Event BlueprintUpdateAnimation()
   > n1 = Gravity Override State Function DEPRECATED(
    CharacterPartAnimInstance=Self-Reference()
    GameTimeInSeconds=GetGameTimeInSeconds()
    GravityOverrideParams=Get Gravity Override Params()
    RBAN Params=Get RBAN Params()
   )
   > Set GravityOverride(GravityOverride=@n1)
   > Set Sim Space Settings(Sim Space Settings:Sim Space Settings Structure=Break RBANSetting Struct!(RBANSettingStruct:"RBANSetting Struct Structure (by ref)"=@n1.RBAN SettingsOut) -> (
     ComponentLinearAccScale_28_#1:Vector=~
     ComponentLinearVelScale_31_#2:Vector=~
     SimSpaceSettings_34_#3:Sim Space Settings Structure
    ) {is=title K2Node_BreakStruct}.SimSpaceSettings_34_#3) -> (Output_Get:Sim Space Settings Structure)
   n2 = Get bIsInFrontEnd()
   > Set Chain Alpha(Chain Alpha:bool=NOT Boolean(A=@n2)) -> (Output_Get:bool)
   > n3 = Branch(Condition=@n2)
   > Set alpha frontend(alpha frontend:double=0.3) -> (Output_Get:double)
   > Set simspace var(simspace var:Byte=2) -> (Output_Get:Byte=0)
   Set alpha frontend(execute=@n3.else, alpha frontend=1)
   > Set simspace var(simspace var=4) -> (Output_Get=0)
  ]}
]
variables [
 alpha frontend {type=number}
 Chain Alpha {type=boolean}
 Gravity Override Params {type=GravityOverrideParamsStruct}
 GravityOverride {type=Vector}
 RBAN Params {
  type=RBANParamsStruct
  default {
   dBNO {simSpaceSettings {
     maxAngularAcceleration=1000
     maxAngularVelocity=3
     maxLinearAcceleration=300
     maxLinearVelocity=700
     velocityScaleZ=0
     externalLinearDragV=<0 0.5 0>
    }}
   dBNO_Carried {simSpaceSettings {
     maxAngularAcceleration=1000
     maxAngularVelocity=3
     maxLinearAcceleration=300
     maxLinearVelocity=700
     velocityScaleZ=0
     externalLinearDragV=<0 0.5 0>
    }}
   emoteOrMelee {simSpaceSettings {
     maxAngularAcceleration=1000
     maxAngularVelocity=3
     maxLinearAcceleration=300
     maxLinearVelocity=700
     velocityScaleZ=0
     externalLinearDragV=<0 0.5 0>
    }}
   front_End {simSpaceSettings {
     maxAngularAcceleration=1000
     maxAngularVelocity=3
     maxLinearAcceleration=300
     maxLinearVelocity=700
     velocityScaleZ=0
     externalLinearDragV=<0 0.5 0>
    }}
   onGround_Crouch_Moving {simSpaceSettings {
     maxAngularAcceleration=1000
     maxAngularVelocity=3
     maxLinearAcceleration=300
     maxLinearVelocity=700
     velocityScaleZ=0
     externalLinearDragV=<0 0.5 0>
    }}
   onGround_Falling {simSpaceSettings {
     maxAngularAcceleration=1000
     maxAngularVelocity=3
     maxLinearAcceleration=300
     maxLinearVelocity=700
     velocityScaleZ=0
     externalLinearDragV=<0 0.5 0>
    }}
   onGround_Moving {simSpaceSettings {
     maxAngularAcceleration=1000
     maxAngularVelocity=3
     maxLinearAcceleration=300
     maxLinearVelocity=700
     velocityScaleZ=0
     externalLinearDragV=<0 0.5 0>
    }}
   onGround_Standing {simSpaceSettings {
     maxAngularAcceleration=1000
     maxAngularVelocity=3
     maxLinearAcceleration=300
     maxLinearVelocity=700
     velocityScaleZ=0
     externalLinearDragV=<0 0.5 0>
    }}
   rail_Grinding {simSpaceSettings {
     maxAngularAcceleration=1000
     maxAngularVelocity=3
     maxLinearAcceleration=300
     maxLinearVelocity=700
     velocityScaleZ=0
     worldAlpha=0.1
     externalLinearDragV=<0 0.5 0>
    }}
   riding_Moving {simSpaceSettings {
     maxAngularAcceleration=1000
     maxAngularVelocity=3
     maxLinearAcceleration=300
     maxLinearVelocity=700
     velocityScaleZ=0
     externalLinearDragV=<0 0.5 0>
    }}
   riding_Moving_Targeting {simSpaceSettings {
     maxAngularAcceleration=1000
     maxAngularVelocity=3
     maxLinearAcceleration=300
     maxLinearVelocity=700
     velocityScaleZ=0
     externalLinearDragV=<0 0.5 0>
    }}
   riding_Standing {simSpaceSettings {
     maxAngularAcceleration=1000
     maxAngularVelocity=3
     maxLinearAcceleration=300
     maxLinearVelocity=700
     velocityScaleZ=0
     externalLinearDragV=<0 0.5 0>
    }}
   skydive_DivingDown {simSpaceSettings {
     maxAngularAcceleration=1000
     maxAngularVelocity=3
     maxLinearAcceleration=300
     maxLinearVelocity=700
     velocityScaleZ=0
     externalLinearDragV=<0 0.5 0>
    }}
   skydive_FloatingDown {
    componentLinearAccScale=<0 0.5 0>
    componentLinearVelScale=<0 0.5 0>
    simSpaceSettings {
     maxAngularAcceleration=1000
     maxAngularVelocity=3
     maxLinearAcceleration=300
     maxLinearVelocity=700
     velocityScaleZ=0
     externalLinearDragV=<0 0.5 0>
    }
   }
   skydive_GoingUp {simSpaceSettings {
     maxAngularAcceleration=1000
     maxAngularVelocity=3
     maxLinearAcceleration=300
     maxLinearVelocity=700
     velocityScaleZ=0
     externalLinearDragV=<0 0.5 0>
    }}
   skydive_Parachuting {
    componentLinearAccScale=<0 0.1 0.3>
    componentLinearVelScale=<0 0.1 0.3>
    simSpaceSettings {
     maxAngularAcceleration=1000
     maxAngularVelocity=3
     maxLinearAcceleration=300
     maxLinearVelocity=700
     velocityScaleZ=0
     externalLinearDragV=<0 0.5 0>
    }
   }
   skydive_Surfing {simSpaceSettings {
     maxAngularAcceleration=1000
     maxAngularVelocity=3
     maxLinearAcceleration=300
     maxLinearVelocity=700
     velocityScaleZ=0
     externalLinearDragV=<0 0.5 0>
    }}
   swimming {simSpaceSettings {
     maxAngularAcceleration=1000
     maxAngularVelocity=3
     maxLinearAcceleration=300
     maxLinearVelocity=700
     velocityScaleZ=0
     externalLinearDragV=<0 0.5 0>
    }}
   tactical_Sprint {simSpaceSettings {
     maxAngularAcceleration=1000
     maxAngularVelocity=3
     maxLinearAcceleration=300
     maxLinearVelocity=700
     velocityScaleZ=0
     externalLinearDragV=<0 0.5 0>
    }}
   ziplining {simSpaceSettings {
     maxAngularAcceleration=1000
     maxAngularVelocity=3
     maxLinearAcceleration=300
     maxLinearVelocity=700
     velocityScaleZ=0
     externalLinearDragV=<0 0.5 0>
    }}
  }
 }
 Sim Space Settings {
  type=SimSpaceSettings
  default {
   dampingAlpha=1
   maxAngularAcceleration=10000
   maxAngularVelocity=10000
   maxLinearAcceleration=10000
   maxLinearVelocity=10000
   velocityScaleZ=1
   worldAlpha=0
   externalLinearDragV=<0 0 0>
  }
 }
 simspace var {type=integer}
]
