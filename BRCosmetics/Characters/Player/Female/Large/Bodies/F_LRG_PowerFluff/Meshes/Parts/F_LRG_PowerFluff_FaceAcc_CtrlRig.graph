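# Control Rig for the F_LRG_PowerFluff face accessory that corrects hair braid and clavicle/upper-arm helper bones.
# A front-end flag selects either fixed pose offsets or gameplay correctives, then parent and aim constraints steer the braids toward torso-driven targets.
# Gameplay runs a shared DeformRig function, then blends additive bone offsets from head swing/twist (DeltaRotation), arm angles and state alphas like aim, crouch, gliding and sprint.
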ControlRigBlueprint /BRCosmetics/Characters/Player/Female/Large/Bodies/F_LRG_PowerFluff/Meshes/Parts/F_LRG_PowerFluff_FaceAcc_CtrlRig
graphs [
 New Function_ContainedGraph {nodes [
   n1 = RigVMFunctionEntryNode()
   RigVMFunction_MathBoolToInteger = &RigVMFunction_MathBoolToInteger(Value=@n1.onlyTwist)
   RigVMFunction_MathQuaternionSwingTwist = &RigVMFunction_MathQuaternionSwingTwist(
    twistAxis=@n1.twistAxis
    Input=&RigUnit_GetRelativeTransformForItem(
     bChildInitial=False
     bParentInitial=False
     child=(Type=Bone,Name="None")
     parent=(Type=Bone,Name="None")
     child.Name=@n1.child
     parent.Name=@n1.parent
     parent.Type=@n1.parentType
    ).RelativeTransform.Rotation
   )
   RigVMFunction_MathQuaternionToEuler = &RigVMFunction_MathQuaternionToEuler(
    Value=RigVMDispatchNode(
     Values=((),())
     Index=@RigVMFunction_MathBoolToInteger.Result
     Values.0=@RigVMFunction_MathQuaternionSwingTwist.Swing
     Values.1=@RigVMFunction_MathQuaternionSwingTwist.Twist
    ) {template="DISPATCH_RigVMDispatch_SelectInt32(in Index,in Values,out Result)"}.Result
    RotationOrder=@n1.Rotation Order
   )
   > DISPATCH_RigVMDispatch_Print = RigVMDispatchNode(ScreenColor=(R=1,G=1,B=1,A=1), ScreenDuration=1.5, Enabled=@n1.debugRotation, Value=@RigVMFunction_MathQuaternionToEuler.Result) {template="DISPATCH_RigVMDispatch_Print(in Prefix,in Value,in Enabled,in ScreenDuration,in ScreenColor)"}
   Clamp = &RigVMFunction_MathIntClamp(Value=@n1.primAxis)
   DISPATCH_RigVMDispatch_SelectInt32 = RigVMDispatchNode(
    Values=(0,0,0)
    Index=@Clamp.Result
    Values.0=@RigVMFunction_MathQuaternionToEuler.Result.X
    Values.1=@RigVMFunction_MathQuaternionToEuler.Result.Y
    Values.2=@RigVMFunction_MathQuaternionToEuler.Result.Z
   ) {template="DISPATCH_RigVMDispatch_SelectInt32(in Index,in Values,out Result)"}
   Remap = &RigVMFunction_MathDoubleRemap(
    TargetMaximum=1
    TargetMinimum=0
    bClamp=False
    Value=@DISPATCH_RigVMDispatch_SelectInt32.Result
    SourceMaximum=@n1.max
    SourceMinimum=@n1.min
   )
   Multiply = &RigVMFunction_MathDoubleMul(A=&RigVMFunction_MathDoubleClamp(Value=@Remap.Result).Result, B=@n1.premult)
   > RigVMFunctionReturnNode(outputWeight=@Multiply.Result)
  ]}
 New Function_ContainedGraph {nodes [
   RigVMFunctionEntryNode()
   > &RigUnit_ModifyTransforms(
    ItemToModify="((Item=(Type=Bone,Name=\"base_hair_braid_a_00_l\"),Transform=(Rotation=(X=0,Y=0,Z=0,W=1),Translation=(X=0,Y=0,Z=0),Scale3D=(X=1,Y=1,Z=1))),(Item=(Type=Bone,Name=\"base_hair_braid_b_00_l\"),Transform=(Rotation=(X=0,Y=0,Z=0,W=1),Translation=(X=0,Y=0,Z=0),Scale3D ...(truncated)"
    Weight=1
    WeightMinimum=0
   )
   > &RigUnit_ModifyTransforms(
    ItemToModify="((Item=(Type=Bone,Name=\"base_hair_braid_a_00_r\"),Transform=(Rotation=(X=-0.0075,Y=0.0861,Z=0.0439,W=-0.9953),Translation=(X=0,Y=0,Z=0),Scale3D=(X=1,Y=1,Z=1))),(Item=(Type=Bone,Name=\"base_hair_braid_b_00_r\"),Transform=(Rotation=(X=-0.0113,Y=0.0421,Z=0,W=-0. ...(truncated)"
    Weight=1
    WeightMinimum=0
   )
   > &RigUnit_ModifyTransforms(
    ItemToModify=((Item=(Type=Bone,Name="helper_clavicle_r"),Transform=(Rotation=(X=0,Y=-0.1305,Z=0,W=0.9914),Translation=(X=-0.3132,Y=-1.2149,Z=3),Scale3D=(X=1,Y=1,Z=1))))
    Weight=1
    WeightMinimum=0
   )
   > RigVMFunctionReturnNode()
  ]}
 RigVMFunctionLibrary {nodes [
   DeltaRotation = RigVMCollapseNode(ExecuteContext=(), Rotation Order=ZYX, premult=1, twistAxis=(X=1,Y=0,Z=0))
   FrontEnd = RigVMCollapseNode(ExecuteContext=())
  ]}
 RigVMFunction_Sequence_1_ContainedGraph {nodes [
   n1 = RigVMFunctionEntryNode()
   RigVMFunction_Sequence = &RigVMFunction_Sequence(ExecuteContext=@n1.ExecuteContext)
   n2 = &RigVMFunction_Sequence(ExecuteContext=@RigVMFunction_Sequence.B)
   n3 = &RigVMFunction_Sequence(ExecuteContext=@n2.B)
   n4 = &RigVMFunction_Sequence(ExecuteContext=@n3.B)
   n5 = &RigVMFunction_Sequence(ExecuteContext=@n4.B)
   n6 = &RigVMFunction_Sequence(ExecuteContext=@n5.B)
   n7 = &RigVMFunction_Sequence(ExecuteContext=@n6.B)
   n8 = &RigVMFunction_Sequence(ExecuteContext=@n7.B)
   n9 = &RigVMFunction_Sequence(ExecuteContext=@n8.B)
   RigVMFunctionReturnNode(A=@RigVMFunction_Sequence.A, B=@n2.A, C=@n3.A, D=@n4.A, E=@n5.A, F=@n6.A, G=@n7.A, H=@n8.A, I=@n9.A, J=@n9.B)
  ]}
 RigVMModel {nodes [
   RigUnit_BeginExecution = &RigUnit_BeginExecution()
   RigVMFunction_ControlFlowBranch = &RigVMFunction_ControlFlowBranch(ExecuteContext=@RigUnit_BeginExecution.ExecutePin, Condition=RigVMVariableNode() {variable=isFrontEnd}.Value)
   n1 = &RigVMFunction_Sequence(ExecuteContext=@RigVMFunction_ControlFlowBranch.Completed)
   &RigUnit_ParentConstraint(
    AdvancedSettings=(RotationOrderForFilter=XZY)
    Child=(Type=Bone,Name="aim_hair_braid_a_04_r")
    Parents=((Item=(Type=Bone,Name="clavicle_r"),Weight=0.0029),(Item=(Type=Bone,Name="spine_05"),Weight=0.1066),(Item=(Type=Bone,Name="spine_04"),Weight=0.8905))
    Weight=1
    ExecutePin=@n1.A
   )
   > &RigUnit_AimConstraintLocalSpaceOffset(
    Child=(Type=Bone,Name="dyn_hair_braid_a_01_r")
    Filter=(bX=True,bY=True,bZ=True)
    Parents=((Item=(Type=Bone,Name="aim_hair_braid_a_04_transform_r"),Weight=1))
    Weight=1
    WorldUp=(Target=(X=0,Y=0,Z=1),Kind=Direction,Space=(Type=Bone,Name="aim_hair_braid_a_04_transform_r"))
   )
   > &RigUnit_ParentConstraint(
    AdvancedSettings=(RotationOrderForFilter=XZY)
    Child=(Type=Bone,Name="aim_hair_braid_b_04_r")
    Parents=((Item=(Type=Bone,Name="upperarm_twist_01_r"),Weight=0.0205),(Item=(Type=Bone,Name="clavicle_r"),Weight=0.2846),(Item=(Type=Bone,Name="spine_05"),Weight=0.3062),(Item=(Type=Bone,Name="spine_04"),Weight=0.3887))
    Weight=1
   )
   > &RigUnit_AimConstraintLocalSpaceOffset(
    Child=(Type=Bone,Name="dyn_hair_braid_b_01_r")
    Filter=(bX=True,bY=True,bZ=True)
    Parents=((Item=(Type=Bone,Name="aim_hair_braid_b_04_transform_r"),Weight=1))
    Weight=1
    WorldUp=(Target=(X=0,Y=0,Z=1),Kind=Direction,Space=(Type=Bone,Name="aim_hair_braid_b_04_transform_r"))
   )
   &RigUnit_ParentConstraint(
    AdvancedSettings=(RotationOrderForFilter=XZY)
    Child=(Type=Bone,Name="aim_hair_braid_a_04_l")
    Parents=((Item=(Type=Bone,Name="clavicle_l"),Weight=0.019),(Item=(Type=Bone,Name="spine_05"),Weight=0.345),(Item=(Type=Bone,Name="spine_04"),Weight=0.6361))
    Weight=1
    ExecutePin=@n1.B
   )
   > &RigUnit_AimConstraintLocalSpaceOffset(
    Child=(Type=Bone,Name="dyn_hair_braid_a_01_l")
    Filter=(bX=True,bY=True,bZ=True)
    Parents=((Item=(Type=Bone,Name="aim_hair_braid_a_04_transform_l"),Weight=1))
    Weight=1
    WorldUp=(Target=(X=0,Y=0,Z=1),Kind=Direction,Space=(Type=None,Name="None"))
   )
   > &RigUnit_ParentConstraint(
    AdvancedSettings=(RotationOrderForFilter=XZY)
    Child=(Type=Bone,Name="aim_hair_braid_b_03_l")
    Parents=((Item=(Type=Bone,Name="upperarm_twist_01_l"),Weight=0.0463),(Item=(Type=Bone,Name="spine_05"),Weight=0.2894),(Item=(Type=Bone,Name="clavicle_l"),Weight=0.3147),(Item=(Type=Bone,Name="spine_04"),Weight=0.3496))
    Weight=1
   )
   > &RigUnit_AimConstraintLocalSpaceOffset(
    Child=(Type=Bone,Name="dyn_hair_braid_b_01_l")
    Filter=(bX=True,bY=True,bZ=True)
    Parents=((Item=(Type=Bone,Name="aim_hair_braid_b_03_tranform_l"),Weight=1))
    Weight=1
    WorldUp=(Target=(X=0,Y=0,Z=1),Kind=Direction,Space=(Type=Bone,Name="aim_hair_braid_a_04_transform_l"))
   )
   RigVMFunction_Sequence_1 = RigVMAggregateNode(ExecuteContext=@RigVMFunction_ControlFlowBranch.False)
   FN_DeformRig_FNC_1 = RigVMFunctionReferenceNode(
    DeformRig_ Curve="(deformRig_l_upperarm_up_alpha_72_#1=False,deformRig_r_upperarm_up_alpha_73_#2=False,deformRig_l_upperarm_fr_alpha_74_#3=False,deformRig_r_upperarm_fr_alpha_75_CB51B8 ...(truncated)"
    DeformRig_Bool="(upperarm_1_#4=True,elbow_3_#5=True,wrist_7_#6=True,groin_8_#7=True,glute_17_#8=True,knee_9_40C408324CA261 ...(truncated)"
    DeformRig_Debug="(upperarm_debug=False,wrist_debug=False,lat_debug=False,groinGlute_debug=False,upperarm_debug_18_#4=False,wrist_debug_20_#6=False,groinGlute_debug_27_#7=False,lat_deb ...(truncated)"
    DeformRig_Value="(upperarm_up_18_#9=(Translation=(X=0,Y=0,Z=8),Scale3D=(X=0.7,Y=1,Z=1)),upperarm_fwd_fr_19_#10=(Translation=(X=0,Y=10,Z=0),Scale3D=(X=0.5,Y=1,Z=1)),upperarm_fwd_bk_20_9973AAD84EE0A1E349BC9E85067D4A3 ...(truncated)"
    ExecuteContext=@RigVMFunction_Sequence_1.B
   )
   DeltaRotation_3 = RigVMFunctionReferenceNode(
    Rotation Order=ZYX
    child=head
    debugRotation=false
    max=-80
    min=0
    parent=head_null
    parentType=Null
    premult=1
    primAxis=0
    twistAxis=(X=0,Y=1,Z=0)
    ExecuteContext=@RigVMFunction_Sequence_1.D
   )
   n2 = &RigVMFunction_MathDoubleRemap(
    SourceMaximum=0.05
    SourceMinimum=0
    TargetMaximum=1
    TargetMinimum=0
    Value=&RigVMFunction_MathDoubleMul(B=@DeltaRotation_3.outputWeight, A=RigVMVariableNode() {variable=delta_head_up}.Value).Result
   )
   &RigUnit_ModifyTransforms(
    ItemToModify="((Item=(Type=Bone,Name=\"base_hair_braid_a_00_l\"),Transform=(Rotation=(X=0,Y=-0.1305,Z=0,W=0.9914),Translation=(X=0,Y=0,Z=0),Scale3D=(X=1,Y=1,Z=1))),(Item=(Type=Bone,Name=\"base_hair_braid_b_00_l\"),Transform=(Rotation=(X=0,Y=0.0436,Z=0,W=-0.999),Translation= ...(truncated)"
    WeightMinimum=0
    Weight=@n2.Result
    ExecutePin=@RigVMFunction_Sequence_1.E
   )
   > &RigUnit_ModifyTransforms(
    ItemToModify="((Item=(Type=Bone,Name=\"None\"),Transform=(Rotation=(X=0,Y=0,Z=0,W=1),Translation=(X=0,Y=0,Z=0),Scale3D=(X=1,Y=1,Z=1))),(Item=(Type=Bone,Name=\"base_hair_braid_a_00_r\"),Transform=(Rotation=(X=0,Y=0.1305,Z=0,W=0.9914),Translation=(X=0,Y=0,Z=0),Scale3D=(X=1,Y= ...(truncated)"
    WeightMinimum=0
    Weight=@n2.Result
   )
   n3 = &RigVMFunction_AlphaInterp(InterpSpeedDecreasing=0.5, InterpSpeedIncreasing=10, Scale=1, bClampResult=False, Value=RigVMVariableNode() {variable=RBAN_OFF}.Value)
   > n4 = &RigUnit_ModifyTransforms(
    ItemToModify="((Item=(Type=Bone,Name=\"None\"),Transform=(Rotation=(X=0,Y=0,Z=0,W=1),Translation=(X=0,Y=0,Z=0),Scale3D=(X=1,Y=1,Z=1))),(Item=(Type=Bone,Name=\"base_hair_braid_a_00_r\"),Transform=(Rotation=(X=0,Y=-0.1305,Z=0,W=0.9914),Translation=(X=0,Y=0,Z=0),Scale3D=(X=1,Y ...(truncated)"
    WeightMinimum=0
    Weight=&RigVMFunction_MathDoubleMul(B=@n3.Result, A=@n2.Result).Result
   )
   DeltaRotation_4 = RigVMFunctionReferenceNode(
    Rotation Order=ZYX
    child=head
    debugRotation=false
    max=80
    min=0
    parent=head_null
    parentType=Null
    premult=1
    primAxis=0
    twistAxis=(X=0,Y=1,Z=0)
    ExecuteContext=@n12.ExecutePin
   )
   n5 = &RigVMFunction_MathDoubleRemap(
    SourceMaximum=0.3
    SourceMinimum=0
    TargetMaximum=1
    TargetMinimum=0
    Value=&RigVMFunction_MathDoubleMul(B=@DeltaRotation_4.outputWeight, A=RigVMVariableNode() {variable=delta_head_up}.Value).Result
   )
   &RigUnit_ModifyTransforms(
    ItemToModify="((Item=(Type=Bone,Name=\"None\"),Transform=(Rotation=(X=0,Y=0,Z=0,W=1),Translation=(X=0,Y=0,Z=0),Scale3D=(X=1,Y=1,Z=1))),(Item=(Type=Bone,Name=\"base_hair_braid_a_00_r\"),Transform=(Rotation=(X=0.0038,Y=-0.0871,Z=0.0435,W=0.9952),Translation=(X=0,Y=0,Z=0),Scal ...(truncated)"
    WeightMinimum=0
    Weight=@n5.Result
    ExecutePin=@n4.ExecutePin
   )
   n6 = &RigVMFunction_MathDoubleRemap(
    SourceMaximum=90
    SourceMinimum=50
    TargetMaximum=1
    TargetMinimum=0
    Value=&RigVMFunction_MathQuaternionToEuler(
     RotationOrder=YZX
     Value=&RigUnit_GetRelativeTransformForItem(Child=(Type=Bone,Name="upperarm_l"), Parent=(Type=Bone,Name="clavicle_l"), bChildInitial=False, bParentInitial=False).RelativeTransform.Rotation
    ).Result.X
   )
   &RigUnit_ModifyTransforms(
    ItemToModify="((Item=(Type=Bone,Name=\"helper_clavicle_l\"),Transform=(Rotation=(X=0,Y=0,Z=0,W=1),Translation=(X=0,Y=2.3106,Z=-2.6486),Scale3D=(X=1,Y=1,Z=1))),(Item=(Type=Bone,Name=\"helper_upperarm_l\"),Transform=(Rotation=(X=0,Y=0,Z=0,W=1),Translation=(X=0.5041,Y=-0.4741, ...(truncated)"
    WeightMinimum=0
    Weight=@n6.Result
    ExecutePin=@RigVMFunction_Sequence_1.F
   )
   > &RigUnit_ModifyTransforms(
    ItemToModify="((Item=(Type=Bone,Name=\"helper_clavicle_l\"),Transform=(Rotation=(X=0,Y=0,Z=0,W=1),Translation=(X=-0.3616,Y=3.5687,Z=-1.2033),Scale3D=(X=1,Y=1,Z=1))),(Item=(Type=Bone,Name=\"helper_upperarm_l\"),Transform=(Rotation=(X=0,Y=0,Z=0,W=-1),Translation=(X=0,Y=0,Z=0) ...(truncated)"
    WeightMinimum=0
    Weight=&RigVMFunction_MathDoubleMul(A=@n6.Result, B=RigVMVariableNode() {variable=gliding_alpha}.Value).Result
   )
   n7 = &RigVMFunction_MathDoubleRemap(
    SourceMaximum=90
    SourceMinimum=20
    TargetMaximum=1
    TargetMinimum=0
    Value=&RigVMFunction_MathQuaternionToEuler(
     RotationOrder=YZX
     Value=&RigUnit_GetRelativeTransformForItem(Child=(Type=Bone,Name="upperarm_r"), Parent=(Type=Bone,Name="clavicle_r"), bChildInitial=False, bParentInitial=False).RelativeTransform.Rotation
    ).Result.X
   )
   > &RigUnit_ModifyTransforms(
    ItemToModify="((Item=(Type=Bone,Name=\"helper_clavicle_r\"),Transform=(Rotation=(X=0,Y=0,Z=0,W=-1),Translation=(X=-7.9176,Y=4.6125,Z=0.0158),Scale3D=(X=1,Y=1,Z=1))),(Item=(Type=Bone,Name=\"helper_upperarm_r\"),Transform=(Rotation=(X=0,Y=0,Z=0,W=-1),Translation=(X=0,Y=0,Z=-3 ...(truncated)"
    WeightMinimum=0
    Weight=@n7.Result
   )
   > &RigUnit_ModifyTransforms(
    ItemToModify="((Item=(Type=Bone,Name=\"helper_clavicle_r\"),Transform=(Rotation=(X=0,Y=0,Z=0,W=1),Translation=(X=0,Y=-3.5671,Z=5.3494),Scale3D=(X=1,Y=1,Z=1))),(Item=(Type=Bone,Name=\"helper_upperarm_r\"),Transform=(Rotation=(X=0,Y=0,Z=0,W=1),Translation=(X=0,Y=0,Z=6.5332),S ...(truncated)"
    WeightMinimum=0
    Weight=&RigVMFunction_MathDoubleMul(B=@n7.Result, A=RigVMVariableNode() {variable=gliding_alpha}.Value).Result
   )
   n8 = &RigVMFunction_AlphaInterp(InterpSpeedDecreasing=5, InterpSpeedIncreasing=10, Scale=1, bClampResult=False, Value=RigVMVariableNode() {variable=aim_alpha}.Value)
   n9 = &RigVMFunction_MathDoubleMul(A=@n8.Result, B=@n7.Result)
   > &RigUnit_ModifyTransforms(
    ItemToModify="((Item=(Type=Bone,Name=\"helper_clavicle_r\"),Transform=(Rotation=(X=0,Y=0,Z=0,W=1),Translation=(X=16.8809,Y=-8.7438,Z=0),Scale3D=(X=1,Y=1,Z=1))),(Item=(Type=Bone,Name=\"helper_upperarm_r\"),Transform=(Rotation=(X=0,Y=0,Z=0,W=1),Translation=(X=0,Y=0,Z=0),Scale ...(truncated)"
    WeightMinimum=0
    Weight=@n9.Result
   )
   AlphaInterp = &RigVMFunction_AlphaInterp(InterpSpeedDecreasing=5, InterpSpeedIncreasing=10, Scale=1, bClampResult=False, Value=RigVMVariableNode() {variable=RBAN_OFF}.Value)
   > &RigUnit_ModifyTransforms(
    ItemToModify="((Item=(Type=Bone,Name=\"helper_clavicle_r\"),Transform=(Rotation=(X=0,Y=0,Z=0,W=1),Translation=(X=0,Y=0,Z=0),Scale3D=(X=1,Y=1,Z=1))),(Item=(Type=Bone,Name=\"helper_upperarm_r\"),Transform=(Rotation=(X=0,Y=0,Z=0,W=1),Translation=(X=0,Y=0,Z=0),Scale3D=(X=1,Y=1, ...(truncated)"
    WeightMinimum=0
    Weight=&RigVMFunction_MathDoubleMul(B=@AlphaInterp.Result, A=@n9.Result).Result
   )
   > &RigUnit_ModifyTransforms(
    ItemToModify="((Item=(Type=Bone,Name=\"helper_clavicle_r\"),Transform=(Rotation=(X=0,Y=0,Z=0,W=1),Translation=(X=0,Y=-0.3379,Z=0),Scale3D=(X=1,Y=1,Z=1))),(Item=(Type=Bone,Name=\"helper_upperarm_r\"),Transform=(Rotation=(X=0,Y=0,Z=0,W=1),Translation=(X=0,Y=0,Z=0),Scale3D=(X= ...(truncated)"
    WeightMinimum=0
    Weight=&RigVMFunction_MathDoubleMul(B=@n9.Result, A=RigVMVariableNode() {variable=crouch_alpha}.Value).Result
   )
   DeltaRotation_8 = RigVMFunctionReferenceNode(
    Rotation Order=ZYX
    child=head
    max=-60
    parent=head_null
    parentType=Null
    premult=1
    primAxis=2
    twistAxis=(X=1,Y=0,Z=0)
    ExecuteContext=@RigVMFunction_Sequence_1.C
   )
   > RigVMVariableNode(Value=@DeltaRotation_8.outputWeight) {variable=delta_head_up}
   > &RigUnit_ModifyTransforms(
    ItemToModify="((Item=(Type=Bone,Name=\"base_hair_braid_a_00_l\"),Transform=(Rotation=(X=0,Y=-0.1736,Z=0,W=-0.9848),Translation=(X=0,Y=0,Z=0),Scale3D=(X=1,Y=1,Z=1))),(Item=(Type=Bone,Name=\"base_hair_braid_b_00_l\"),Transform=(Rotation=(X=-0.0302,Y=-0.171,Z=0.171,W=-0.9698), ...(truncated)"
    WeightMinimum=0
    Weight=@DeltaRotation_8.outputWeight
   )
   > &RigUnit_ModifyTransforms(
    ItemToModify="((Item=(Type=Bone,Name=\"base_hair_braid_a_00_r\"),Transform=(Rotation=(X=0,Y=-0.2164,Z=0,W=-0.9763),Translation=(X=0,Y=0,Z=0),Scale3D=(X=1,Y=1,Z=1))),(Item=(Type=Bone,Name=\"base_hair_braid_b_00_r\"),Transform=(Rotation=(X=0,Y=-0.0872,Z=0,W=-0.9962),Translati ...(truncated)"
    WeightMinimum=0
    Weight=@DeltaRotation_8.outputWeight
   )
   > DeltaRotation_7 = RigVMFunctionReferenceNode(
    Rotation Order=ZYX
    child=head
    debugRotation=false
    max=60
    parent=head_null
    parentType=Null
    premult=1
    primAxis=2
    twistAxis=(X=1,Y=0,Z=0)
   )
   > RigVMVariableNode(Value=@DeltaRotation_7.outputWeight) {variable=delta_head_down}
   &RigUnit_ModifyTransforms(
    ItemToModify="((Item=(Type=Bone,Name=\"helper_clavicle_r\"),Transform=(Rotation=(X=0,Y=0,Z=0,W=1),Translation=(X=0,Y=0,Z=0),Scale3D=(X=1,Y=1,Z=1))),(Item=(Type=Bone,Name=\"helper_upperarm_r\"),Transform=(Rotation=(X=0,Y=0,Z=0,W=1),Translation=(X=0,Y=0,Z=0),Scale3D=(X=1,Y=1, ...(truncated)"
    WeightMinimum=0
    Weight=RigVMVariableNode() {variable=tactical_srpinting_alpha}.Value
    ExecutePin=@RigVMFunction_Sequence_1.G
   )
   n10 = RigVMVariableNode() {variable=delta_head_down}
   > &RigUnit_ModifyTransforms(
    ItemToModify="((Item=(Type=Bone,Name=\"helper_clavicle_r\"),Transform=(Rotation=(X=0,Y=0,Z=0,W=1),Translation=(X=0,Y=0,Z=0),Scale3D=(X=1,Y=1,Z=1))),(Item=(Type=Bone,Name=\"helper_upperarm_r\"),Transform=(Rotation=(X=0,Y=0,Z=0,W=1),Translation=(X=0,Y=0,Z=0),Scale3D=(X=1,Y=1, ...(truncated)"
    WeightMinimum=0
    Weight=@n10.Value
   )
   > &RigUnit_ModifyTransforms(
    ItemToModify="((Item=(Type=Bone,Name=\"helper_clavicle_l\"),Transform=(Rotation=(X=0,Y=0,Z=0,W=1),Translation=(X=0,Y=0,Z=0),Scale3D=(X=1,Y=1,Z=1))),(Item=(Type=Bone,Name=\"helper_upperarm_l\"),Transform=(Rotation=(X=0,Y=0,Z=0,W=1),Translation=(X=0,Y=0,Z=0),Scale3D=(X=1,Y=1, ...(truncated)"
    WeightMinimum=0
    Weight=@n10.Value
   )
   FrontEnd = RigVMFunctionReferenceNode(ExecuteContext=@RigVMFunction_ControlFlowBranch.True)
   &RigUnit_ModifyTransforms(
    ItemToModify=((Item=(Type=Bone,Name="None"),Transform=(Translation=(X=0,Y=0,Z=0),Scale3D=(X=1,Y=1,Z=1))),(Item=(Type=Bone,Name="None"),Transform=(Rotation=(X=0,Y=0,Z=0,W=1),Translation=(X=0,Y=0,Z=0),Scale3D=(X=1,Y=1,Z=1))))
    WeightMinimum=0
    Weight=@DeltaRotation_4.outputWeight
    ExecutePin=@n11.ExecutePin
   )
   &RigUnit_ModifyTransforms(
    ExecutePin=()
    ItemToModify=((Item=(Type=Bone,Name="None"),Transform=(Translation=(X=0,Y=0,Z=0),Scale3D=(X=1,Y=1,Z=1))))
    WeightMinimum=0
    Weight=@n5.Result
   )
   &RigUnit_ModifyTransforms(
    ExecutePin=()
    ItemToModify="((Item=(Type=Bone,Name=\"helper_clavicle_l\"),Transform=(Rotation=(X=0,Y=0,Z=0,W=1),Translation=(X=0,Y=-2.8764,Z=0),Scale3D=(X=1,Y=1,Z=1))),(Item=(Type=Bone,Name=\"helper_upperarm_l\"),Transform=(Rotation=(X=0,Y=0,Z=0,W=1),Translation=(X=-1.9757,Y=-20.7187,Z=- ...(truncated)"
    Weight=1
    WeightMinimum=0
   )
   n11 = &RigUnit_ModifyTransforms(
    ItemToModify="((Item=(Type=Bone,Name=\"base_hair_braid_a_00_r\"),Transform=(Rotation=(X=-0.0376,Y=-0.1695,Z=0.2961,W=0.9392),Translation=(X=0,Y=0,Z=0),Scale3D=(X=1,Y=1,Z=1))),(Item=(Type=Bone,Name=\"base_hair_braid_b_00_r\"),Transform=(Rotation=(X=0.1294,Y=0.017,Z=0.1294,W= ...(truncated)"
    WeightMinimum=0
    ExecutePin=@DeltaRotation_4.ExecuteContext
    Weight=@DeltaRotation_4.outputWeight
   )
   n12 = &RigUnit_ModifyTransforms(
    ItemToModify="((Item=(Type=Bone,Name=\"base_hair_braid_a_00_r\"),Transform=(Rotation=(X=0,Y=-0.2588,Z=0,W=-0.9659),Translation=(X=0,Y=0,Z=0),Scale3D=(X=1,Y=1,Z=1))),(Item=(Type=Bone,Name=\"base_hair_braid_b_00_r\"),Transform=(Rotation=(X=0,Y=0.3827,Z=0,W=0.9239),Translation ...(truncated)"
    WeightMinimum=0
    Weight=@DeltaRotation_3.outputWeight
    ExecutePin=@n13.ExecutePin
   )
   n13 = &RigUnit_ModifyTransforms(
    ItemToModify="((Item=(Type=Bone,Name=\"base_hair_braid_b_00_l\"),Transform=(Rotation=(X=0,Y=-0.1305,Z=0,W=0.9914),Translation=(X=0,Y=0,Z=0),Scale3D=(X=1,Y=1,Z=1))),(Item=(Type=Bone,Name=\"aim_hair_braid_a_04_transform_l\"),Transform=(Rotation=(X=0,Y=0,Z=0,W=1),Translation=( ...(truncated)"
    WeightMinimum=0
    Weight=@DeltaRotation_3.outputWeight
    ExecutePin=@DeltaRotation_3.ExecuteContext
   )
  ]}
]
hierarchy [
 root {loc=<0 0 0>, rot=<0 0 0 1>}
 pelvis {parent=root, loc=<0 -0.648 94.7699>, rot=<0.0127 0.707 -0.0127 -0.707>}
 spine_01 {parent=pelvis, loc=<5.0975 -0.6388 0>, rot=<0 0 0.034 -0.9994>}
 spine_02 {parent=spine_01, loc=<7.252 -0.035 0>, rot=<0 0 0.0238 -0.9997>}
 spine_03 {parent=spine_02, loc=<6.0106 0 0>, rot=<0 0 0.0149 -0.9999>}
 spine_04 {parent=spine_03, loc=<8.1023 0 0>, rot=<0 0 -0.0132 -0.9999>}
 spine_05 {parent=spine_04, loc=<10.0142 0 0>, rot=<0 0 -0.0015 -1>}
 clavicle_l {parent=spine_05, loc=<4.1651 -0.3023 -5.0942>, rot=<0.0465 -0.7289 -0.009 -0.6829>}
 upperarm_l {parent=clavicle_l, loc=<11.2117 -1.1486 -0.0332>, rot=<0.0179 -0.2982 -0.028 -0.9539>}
 upperarm_twist_01_l {parent=upperarm_l, loc=<7.2135 -0.1134 0>, rot=<0 0 0 1>}
 helper_upperarm_l {parent=upperarm_l, loc=<0 0 0>, rot=<0 0 0 1>}
 helper_clavicle_l {parent=clavicle_l, loc=<0 0 0>, rot=<0 0 0 1>}
 neck_01 {parent=spine_05, loc=<11.3775 -0.6396 0>, rot=<0 0 -0.2003 -0.9797>}
 neck_02 {parent=neck_01, loc=<4.5627 0 0>, rot=<0 0 0.0504 -0.9987>}
 head {parent=neck_02, loc=<5.3914 0 0>, rot=<0 0 0.115 -0.9934>}
 base_hair {parent=head, loc=<-0.0066 0.795 0>, rot=<0 0 0 1>}
 base_bun_hair_r {parent=base_hair, loc=<10.3372 -2.8745 8.1681>, rot=<0.0029 -0.7071 -0.0029 -0.7071>}
 dyn_bun_hair_r {parent=base_bun_hair_r, loc=<0 0 0>, rot=<0 0 0 1>}
 base_bun_hair_l {parent=base_hair, loc=<10.3372 -2.8745 -8.1681>, rot=<0.0029 -0.7071 -0.0029 -0.7071>}
 dyn_bun_hair_l {parent=base_bun_hair_l, loc=<0 0 0>, rot=<0 0 0 1>}
 base_earing_r {parent=base_hair, loc=<1.0941 -0.6352 8.3184>, rot=<0.0041 -1 0 0>}
 dyn_earing_r {parent=base_earing_r, loc=<0 0 0>, rot=<0 0 0 1>}
 base_earing_l {parent=base_hair, loc=<1.0301 -0.3236 -8.3568>, rot=<0.0041 -1 0 0>}
 dyn_earing_l {parent=base_earing_l, loc=<0 0 0>, rot=<0 0 0 1>}
 base_ear_01_r {parent=base_hair, loc=<15.0174 3.9035 4.317>, rot=<-0.6748 0.0996 -0.1266 -0.7202>}
 dyn_ear_01_r {parent=base_ear_01_r, loc=<0 0 0>, rot=<0 0 0 1>}
 dyn_ear_02_r {parent=dyn_ear_01_r, loc=<8.1219 0 0>, rot=<0 -0.2315 0 -0.9728>}
 dyn_ear_03_r {parent=dyn_ear_02_r, loc=<8.643 0 0.5151>, rot=<0 0 0 1>}
 base_ear_01_l {parent=base_hair, loc=<14.8364 3.1695 -4.9069>, rot=<-0.6831 -0.2095 0.0529 -0.6976>}
 dyn_ear_01_l {parent=base_ear_01_l, loc=<0 0 0>, rot=<0 0 0 1>}
 dyn_ear_02_l {parent=dyn_ear_01_l, loc=<8.1553 0 0>, rot=<0 -0.2415 0 -0.9704>}
 dyn_ear_03_l {parent=dyn_ear_02_l, loc=<7.7133 0 0>, rot=<0 0 0 1>}
 base_hair_braid_b_00_r {parent=base_hair, loc=<3.3848 -8.5584 8.8224>, rot=<0.2131 0.6742 0.6942 0.1347>}
 dyn_hair_braid_b_00_r {parent=base_hair_braid_b_00_r, loc=<0 0 0>, rot=<0 0 0 1>}
 dyn_hair_braid_b_01_r {parent=dyn_hair_braid_b_00_r, loc=<6.3068 0.0678 -1.0482>, rot=<0.0128 0.0969 -0.017 -0.9951>}
 dyn_hair_braid_b_02_r {parent=dyn_hair_braid_b_01_r, loc=<6.3838 0.0063 -0.8062>, rot=<-0.0033 0.0236 0.0024 -0.9997>}
 dyn_hair_braid_b_03_r {parent=dyn_hair_braid_b_02_r, loc=<6.6336 0.025 0.4259>, rot=<0.0086 0.0902 0.0037 -0.9959>}
 dyn_hair_braid_b_04_r {parent=dyn_hair_braid_b_03_r, loc=<7.1483 0 -0.5821>, rot=<0 0 0 1>}
 base_hair_b_acc_01_r {parent=dyn_hair_braid_b_01_r, loc=<0.861 1.0773 -0.8588>, rot=<0.3554 -0.1387 -0.5178 -0.7658>}
 dyn_hair_b_acc_01_r {parent=base_hair_b_acc_01_r, loc=<0 0 0>, rot=<0 0 0 1>}
 base_hair_b_acc_02_r {parent=dyn_hair_braid_b_01_r, loc=<1.1783 0.7493 -0.9919>, rot=<0.4056 -0.1962 -0.3539 -0.8196>}
 dyn_hair_b_acc_02_r {parent=base_hair_b_acc_02_r, loc=<0 0 0>, rot=<0 0 0 1>}
 base_hair_braid_a_00_r {parent=base_hair, loc=<3.0924 -9.1865 6.8707>, rot=<0.2217 0.6714 0.6875 0.1655>}
 dyn_hair_braid_a_00_r {parent=base_hair_braid_a_00_r, loc=<0 0 0>, rot=<0 0 0 1>}
 dyn_hair_braid_a_01_r {parent=dyn_hair_braid_a_00_r, loc=<10.3976 -0.0041 -0.998>, rot=<0.0034 0.1254 0.0265 -0.9917>}
 dyn_hair_braid_a_02_r {parent=dyn_hair_braid_a_01_r, loc=<7.1888 -0.0015 -0.011>, rot=<-0.0008 0.0298 0.0022 -0.9996>}
 dyn_hair_braid_a_03_r {parent=dyn_hair_braid_a_02_r, loc=<7.4976 -0.0003 0.0098>, rot=<0.0007 0.0145 0.0005 -0.9999>}
 dyn_hair_braid_a_04_r {parent=dyn_hair_braid_a_03_r, loc=<6.8065 0 0>, rot=<0 0 0 1>}
 base_hair_braid_b_00_l {parent=base_hair, loc=<4.9929 -8.551 -8.1144>, rot=<-0.1786 -0.6842 -0.6471 -0.285>}
 dyn_hair_braid_b_00_l {parent=base_hair_braid_b_00_l, loc=<0 0 0>, rot=<0 0 0 1>}
 dyn_hair_braid_b_01_l {parent=dyn_hair_braid_b_00_l, loc=<6.138 0.0934 -2.418>, rot=<0 -0.0564 0 -0.9984>}
 dyn_hair_braid_b_02_l {parent=dyn_hair_braid_b_01_l, loc=<8.8464 0.0024 -0.0863>, rot=<-0.0034 0.1067 0.0092 -0.9942>}
 dyn_hair_braid_b_03_l {parent=dyn_hair_braid_b_02_l, loc=<9.4439 -0.0741 -0.1771>, rot=<-0.0038 -0.0796 -0.0472 -0.9957>}
 base_hair_a_acc_01_l {parent=dyn_hair_braid_b_01_l, loc=<-1.1942 -1.2397 -0.7351>, rot=<0.7774 0.1529 -0.4075 0.4541>}
 dyn_hair_a_acc_01_l {parent=base_hair_a_acc_01_l, loc=<0 0 0>, rot=<0 0 0 1>}
 base_hair_a_acc_02_l {parent=dyn_hair_braid_b_01_l, loc=<-0.9074 -0.9094 -0.7942>, rot=<-0.582 0.199 0.2108 -0.7597>}
 dyn_hair_a_acc_02_l {parent=base_hair_a_acc_02_l, loc=<0 0 0>, rot=<0 0 0 1>}
 base_hair_braid_a_00_l {parent=base_hair, loc=<4.597 -9.6391 -6.3318>, rot=<-0.1603 -0.6887 -0.6728 -0.2176>}
 dyn_hair_braid_a_00_l {parent=base_hair_braid_a_00_l, loc=<0 0 0>, rot=<0 0 0 1>}
 dyn_hair_braid_a_01_l {parent=dyn_hair_braid_a_00_l, loc=<6.6533 0.1763 -1.4922>, rot=<-0.0087 0.0805 0.0229 -0.9965>}
 dyn_hair_braid_a_02_l {parent=dyn_hair_braid_a_01_l, loc=<7.4987 0.1321 0.0136>, rot=<-0.0036 0.0625 -0.0451 -0.997>}
 dyn_hair_braid_a_03_l {parent=dyn_hair_braid_a_02_l, loc=<7.8227 -0.1688 -0.0161>, rot=<-0.0006 0.0229 -0.0129 -0.9997>}
 dyn_hair_braid_a_04_l {parent=dyn_hair_braid_a_03_l, loc=<6.6231 -0.2352 0>, rot=<0 0 0 1>}
 FACIAL_C_FacialRoot {parent=head, loc=<2.1458 0.049 -0.0002>, rot=<0.4879 -0.5118 -0.4879 -0.5118>}
 FACIAL_C_Forehead {parent=FACIAL_C_FacialRoot, loc=<-0.0007 -3.7441 9.9582>, rot=<-0.0548 0 0 -0.9985>}
 FACIAL_L_ForeheadIn {parent=FACIAL_C_FacialRoot, loc=<1.9065 -3.7082 9.8485>, rot=<-0.0721 -0.0696 0.0035 -0.995>}
 FACIAL_R_ForeheadIn {parent=FACIAL_C_FacialRoot, loc=<-1.9067 -3.7094 9.8442>, rot=<-0.0721 0.0697 -0.0035 -0.995>}
 FACIAL_L_ForeheadMid {parent=FACIAL_C_FacialRoot, loc=<3.5649 -4.0951 9.255>, rot=<-0.0822 -0.1754 -0.0226 -0.9808>}
 FACIAL_R_ForeheadMid {parent=FACIAL_C_FacialRoot, loc=<-3.5642 -4.0943 9.2509>, rot=<-0.0822 0.1754 0.0226 -0.9808>}
 FACIAL_L_ForeheadOut {parent=FACIAL_C_FacialRoot, loc=<5.152 -4.1404 8.3365>, rot=<-0.1323 -0.2547 -0.0821 -0.9544>}
 FACIAL_R_ForeheadOut {parent=FACIAL_C_FacialRoot, loc=<-5.1493 -4.1367 8.3341>, rot=<-0.1323 0.2547 0.0821 -0.9544>}
 FACIAL_L_EyesackUpper {parent=FACIAL_C_FacialRoot, loc=<3.6156 -2.8169 9.8815>, rot=<0.0375 -0.0835 -0.0655 -0.9936>}
 FACIAL_R_EyesackUpper {parent=FACIAL_C_FacialRoot, loc=<-3.6175 -2.8181 9.8796>, rot=<0.0375 0.0835 0.0655 -0.9936>}
 FACIAL_L_EyelidUpperFurrow {parent=FACIAL_C_FacialRoot, loc=<3.5053 -2.5172 9.7425>, rot=<-0.0413 -0.1499 -0.0814 -0.9845>}
 FACIAL_R_EyelidUpperFurrow {parent=FACIAL_C_FacialRoot, loc=<-3.5109 -2.5159 9.7386>, rot=<-0.0413 0.1499 0.0814 -0.9845>}
 FACIAL_L_EyelidUpperA {parent=FACIAL_C_FacialRoot, loc=<3.5497 -1.4158 7.6689>, rot=<-0.1591 0 0 -0.9873>}
 FACIAL_R_EyelidUpperA {parent=FACIAL_C_FacialRoot, loc=<-3.5489 -1.4138 7.6707>, rot=<-0.1591 0 0 -0.9873>}
 FACIAL_L_Eye {parent=FACIAL_C_FacialRoot, loc=<3.5468 -1.3816 7.7044>, rot=<-0.0114 0 0 -0.9999>}
 FACIAL_R_Eye {parent=FACIAL_C_FacialRoot, loc=<-3.547 -1.3797 7.706>, rot=<-0.0114 0 0 -0.9999>}
 FACIAL_L_EyelidLowerA {parent=FACIAL_C_FacialRoot, loc=<3.5497 -1.4158 7.6689>, rot=<0.273 0 0 -0.962>}
 FACIAL_R_EyelidLowerA {parent=FACIAL_C_FacialRoot, loc=<-3.5489 -1.4138 7.6707>, rot=<0.273 0 0 -0.962>}
 FACIAL_L_EyeCornerInner {parent=FACIAL_C_FacialRoot, loc=<2.2385 -0.8633 9.061>, rot=<-0.0114 0 0 -0.9999>}
 FACIAL_R_EyeCornerInner {parent=FACIAL_C_FacialRoot, loc=<-2.2197 -0.8553 9.0604>, rot=<-0.0114 0 0 -0.9999>}
 FACIAL_L_EyeCornerOuter {parent=FACIAL_C_FacialRoot, loc=<5.29 -1.6185 8.4146>, rot=<-0.0106 -0.3826 -0.0044 -0.9239>}
 FACIAL_R_EyeCornerOuter {parent=FACIAL_C_FacialRoot, loc=<-5.3043 -1.6208 8.4156>, rot=<-0.0106 0.3826 0.0044 -0.9239>}
 FACIAL_L_EyesackLower {parent=FACIAL_C_FacialRoot, loc=<4.0822 0.3647 8.0949>, rot=<-0.0336 -0.1394 -0.0239 -0.9894>}
 FACIAL_R_EyesackLower {parent=FACIAL_C_FacialRoot, loc=<-4.083 0.3769 8.0988>, rot=<-0.0336 0.1395 0.0239 -0.9894>}
 FACIAL_L_CheekInner {parent=FACIAL_C_FacialRoot, loc=<2.5309 0.8356 8.7303>, rot=<-0.1343 -0.0857 -0.1981 -0.9671>}
 FACIAL_R_CheekInner {parent=FACIAL_C_FacialRoot, loc=<-2.5271 0.8344 8.7296>, rot=<-0.1343 0.0857 0.1981 -0.9671>}
 FACIAL_L_CheekOuter {parent=FACIAL_C_FacialRoot, loc=<4.865 0.7831 6.9164>, rot=<0.0447 -0.378 0.096 -0.9197>}
 FACIAL_R_CheekOuter {parent=FACIAL_C_FacialRoot, loc=<-4.8608 0.7919 6.9158>, rot=<0.0447 0.3781 -0.096 -0.9197>}
 FACIAL_L_NasolabialBulge {parent=FACIAL_C_FacialRoot, loc=<3.7268 4.075 6.7324>, rot=<0.0474 -0.3776 0.0195 -0.9246>}
 FACIAL_R_NasolabialBulge {parent=FACIAL_C_FacialRoot, loc=<-3.7269 4.0769 6.7286>, rot=<0.0474 0.3776 -0.0195 -0.9246>}
 FACIAL_L_NasolabialFurrow {parent=FACIAL_C_FacialRoot, loc=<3.5404 2.9011 9.6529>, rot=<-0.0438 -0.1217 -0.365 -0.922>}
 FACIAL_R_NasolabialFurrow {parent=FACIAL_C_FacialRoot, loc=<-3.5408 2.8985 9.6506>, rot=<-0.0438 0.1217 0.365 -0.922>}
 FACIAL_L_CheekLower {parent=FACIAL_C_FacialRoot, loc=<3.9099 4.0775 6.1257>, rot=<0.0243 -0.4275 0.0529 -0.9021>}
 FACIAL_R_CheekLower {parent=FACIAL_C_FacialRoot, loc=<-3.909 4.0791 6.1213>, rot=<0.0243 0.4275 -0.0529 -0.9021>}
 FACIAL_C_Nose {parent=FACIAL_C_FacialRoot, loc=<0.0005 0.0192 10.2015>, rot=<-0.1417 0 0 -0.9899>}
 FACIAL_C_NoseLower {parent=FACIAL_C_Nose, loc=<-0.0014 3.1911 -0.3616>, rot=<0.1305 0 0 -0.9914>}
 FACIAL_L_Nostril {parent=FACIAL_C_Nose, loc=<0.9879 2.3206 -0.1856>, rot=<0.0173 0.1305 0.0023 -0.9913>}
 FACIAL_R_Nostril {parent=FACIAL_C_Nose, loc=<-0.983 2.3143 -0.194>, rot=<0.0173 -0.1305 -0.0023 -0.9913>}
 FACIAL_C_NoseTip {parent=FACIAL_C_Nose, loc=<-0.0003 2.3998 0.4142>, rot=<0.0872 0 0 -0.9962>}
 FACIAL_C_MouthUpper {parent=FACIAL_C_FacialRoot, loc=<-0.002 4.261 7.9554>, rot=<-0.0114 0 0 -0.9999>}
 FACIAL_C_LipUpper {parent=FACIAL_C_MouthUpper, loc=<-0.002 0.0253 2.4061>, rot=<-0.0872 0 0 -0.9962>}
 FACIAL_L_LipUpper {parent=FACIAL_C_MouthUpper, loc=<1.0775 0.0135 2.2516>, rot=<-0.0778 -0.1601 -0.0379 -0.9833>}
 FACIAL_R_LipUpper {parent=FACIAL_C_MouthUpper, loc=<-1.084 0.0161 2.2462>, rot=<-0.0778 0.1601 0.0379 -0.9833>}
 FACIAL_L_LipUpperOuter {parent=FACIAL_C_MouthUpper, loc=<1.9333 0.1646 1.7292>, rot=<-0.0984 -0.2293 -0.0983 -0.9634>}
 FACIAL_R_LipUpperOuter {parent=FACIAL_C_MouthUpper, loc=<-1.94 0.1661 1.718>, rot=<-0.0984 0.2293 0.0983 -0.9634>}
 FACIAL_L_LipCorner {parent=FACIAL_C_MouthUpper, loc=<2.5834 0.5579 0.8047>, rot=<0 -0.2078 0 -0.9782>}
 FACIAL_R_LipCorner {parent=FACIAL_C_MouthUpper, loc=<-2.5903 0.5619 0.7806>, rot=<0 0.2078 0 -0.9782>}
 FACIAL_L_Masseter {parent=FACIAL_C_FacialRoot, loc=<7.5237 1.4752 4.3965>, rot=<0.0581 0.1128 0.0092 -0.9919>}
 FACIAL_R_Masseter {parent=FACIAL_C_FacialRoot, loc=<-7.5267 1.4802 4.3913>, rot=<0.058 -0.1128 -0.0092 -0.9919>}
 FACIAL_C_UnderChin {parent=FACIAL_C_FacialRoot, loc=<-0.0001 9.0488 4.4684>, rot=<0.4899 0 0 -0.8718>}
 FACIAL_L_UnderChin {parent=FACIAL_C_FacialRoot, loc=<3.3027 7.5616 3.0463>, rot=<0.3929 -0.3558 0.0783 -0.8443>}
 FACIAL_R_UnderChin {parent=FACIAL_C_FacialRoot, loc=<-3.3057 7.5633 3.0456>, rot=<0.3929 0.3558 -0.0783 -0.8443>}
 FACIAL_C_TeethUpper {parent=FACIAL_C_FacialRoot, loc=<-0.0044 4.8208 9.3118>, rot=<-0.0114 0 0 -0.9999>}
 FACIAL_C_LowerLipRotation {parent=FACIAL_C_FacialRoot, loc=<0.003 1.5095 2.74>, rot=<0.3313 0 0 -0.9435>}
 FACIAL_C_MouthLower {parent=FACIAL_C_LowerLipRotation, loc=<-0.0048 -1.1132 5.7907>, rot=<-0.342 0 0 -0.9397>}
 FACIAL_C_LipLower {parent=FACIAL_C_MouthLower, loc=<-0.0009 1.4062 2.045>, rot=<0.0872 0 0 -0.9962>}
 FACIAL_L_LipLower {parent=FACIAL_C_MouthLower, loc=<0.9741 1.3783 1.8189>, rot=<0.0927 -0.1294 0.0272 -0.9869>}
 FACIAL_R_LipLower {parent=FACIAL_C_MouthLower, loc=<-0.9769 1.3845 1.8135>, rot=<0.0927 0.1294 -0.0272 -0.9869>}
 FACIAL_L_LipLowerOuter {parent=FACIAL_C_MouthLower, loc=<1.7918 1.1241 1.3073>, rot=<0.0913 -0.2321 0.0318 -0.9679>}
 FACIAL_R_LipLowerOuter {parent=FACIAL_C_MouthLower, loc=<-1.7949 1.1308 1.3011>, rot=<0.0913 0.2321 -0.0318 -0.9679>}
 FACIAL_C_TeethLower {parent=FACIAL_C_LowerLipRotation, loc=<-0.0068 -1.5042 7.1171>, rot=<-0.342 0 0 -0.9397>}
 FACIAL_C_Tongue1 {parent=FACIAL_C_TeethLower, loc=<0.0065 -0.0158 -5.0456>, rot=<-0.2588 0 0 -0.9659>}
 FACIAL_C_Tongue2 {parent=FACIAL_C_Tongue1, loc=<-0.0019 0.4913 1.4595>, rot=<0.2588 0 0 -0.9659>}
 FACIAL_C_Tongue3 {parent=FACIAL_C_Tongue2, loc=<-0.0015 0.2579 1.3347>, rot=<0.1305 0 0 -0.9914>}
 FACIAL_C_Tongue4 {parent=FACIAL_C_Tongue3, loc=<-0.0007 -0.1368 1.1094>, rot=<0.0436 0 0 -0.999>}
 FACIAL_C_TongueUpper3 {parent=FACIAL_C_Tongue3, loc=<0.0005 -0.368 -0.1939>, rot=<0 0 0 1>}
 FACIAL_C_TongueLower3 {parent=FACIAL_C_Tongue3, loc=<0.0007 0.9173 -0.0188>, rot=<0 0 0 1>}
 FACIAL_L_TongueSide3 {parent=FACIAL_C_Tongue3, loc=<1.6698 0.0091 0.1275>, rot=<0 0 0 1>}
 FACIAL_R_TongueSide3 {parent=FACIAL_C_Tongue3, loc=<-1.6714 0.015 0.1234>, rot=<0 0 0 1>}
 FACIAL_C_TongueUpper2 {parent=FACIAL_C_Tongue2, loc=<0.0004 -0.6427 -0.2337>, rot=<0 0 0 1>}
 FACIAL_L_TongueSide2 {parent=FACIAL_C_Tongue2, loc=<1.9846 -0.0097 0.1307>, rot=<0 0 0 1>}
 FACIAL_R_TongueSide2 {parent=FACIAL_C_Tongue2, loc=<-1.9844 -0.0068 0.1288>, rot=<0 0 0 1>}
 FACIAL_C_Jaw {parent=FACIAL_C_FacialRoot, loc=<0.003 1.5095 2.74>, rot=<0.3313 0 0 -0.9435>}
 FACIAL_C_Jawline {parent=FACIAL_C_Jaw, loc=<-0.0032 3.2642 7.7548>, rot=<0.2586 0 0 -0.966>}
 FACIAL_L_ChinSide {parent=FACIAL_C_Jaw, loc=<2.9379 0.8976 7.502>, rot=<-0.2208 -0.3738 0.0499 -0.8995>}
 FACIAL_R_ChinSide {parent=FACIAL_C_Jaw, loc=<-2.9443 0.8979 7.5036>, rot=<-0.2208 0.3738 -0.0499 -0.8995>}
 FACIAL_C_Chin {parent=FACIAL_C_Jaw, loc=<-0.0038 1.4366 7.9832>, rot=<-0.2591 0 0 -0.9659>}
 FACIAL_L_Ear {parent=FACIAL_C_FacialRoot, loc=<8.2588 -0.789 0.0306>, rot=<-0.0377 -0.0004 -0.0262 -0.9989>}
 FACIAL_R_Ear {parent=FACIAL_C_FacialRoot, loc=<-8.2591 -0.7838 0.0186>, rot=<-0.0377 0.0004 0.0262 -0.9989>}
 FACIAL_L_Jawline {parent=FACIAL_C_FacialRoot, loc=<5.0795 5.4537 4.8819>, rot=<0.1983 0.1584 -0.0529 -0.9658>}
 FACIAL_R_Jawline {parent=FACIAL_C_FacialRoot, loc=<-5.0846 5.4563 4.8797>, rot=<0.1983 -0.1583 0.0529 -0.9658>}
 FACIAL_C_NoseBridge {parent=FACIAL_C_FacialRoot, loc=<-0.0009 -2.1062 10.9594>, rot=<0.0324 0 0 -0.9995>}
 FACIAL_L_EyelidUpperB {parent=FACIAL_C_FacialRoot, loc=<3.5497 -1.4158 7.6689>, rot=<-0.2361 0 0 -0.9717>}
 FACIAL_L_EyelidLowerB {parent=FACIAL_C_FacialRoot, loc=<3.5497 -1.4158 7.6689>, rot=<0.3065 0 0 -0.9519>}
 FACIAL_R_EyelidLowerB {parent=FACIAL_C_FacialRoot, loc=<-3.5488 -1.4138 7.6707>, rot=<0.3147 0 0 -0.9492>}
 FACIAL_R_EyelidUpperB {parent=FACIAL_C_FacialRoot, loc=<-3.5489 -1.4138 7.6707>, rot=<-0.2361 0 0 -0.9717>}
 head_null {parent=neck_02, type=Null, loc=<5.3914 0 0>, rot=<0 0 0.115 -0.9934>}
 clavicle_r {parent=spine_05, loc=<4.1651 -0.3023 5.0942>, rot=<0.7289 0.0465 -0.6829 0.009>}
 upperarm_r {parent=clavicle_r, loc=<-11.2117 1.1486 0.0332>, rot=<0.0179 -0.2982 -0.028 -0.9539>}
 upperarm_twist_01_r {parent=upperarm_r, loc=<-7.2135 0.1134 0>, rot=<0 0 0 -1>}
 helper_upperarm_r {parent=upperarm_r, loc=<0 0 0>, rot=<0 0 0 1>}
 helper_clavicle_r {parent=clavicle_r, loc=<0 0 0>, rot=<0 0 0 1>}
 aim_hair_braid_b_03_l {loc=<12.1345 2.6585 136.5648>, rot=<0.331 -0.3149 -0.5997 -0.657>}
 aim_hair_braid_b_03_tranform_l {parent=aim_hair_braid_b_03_l, loc=<0 0 0>, rot=<0 0 0 1>}
 aim_hair_braid_a_04_l {loc=<8.3603 6.6216 135.0323>, rot=<-0.2411 0.2356 0.6616 0.6698>}
 aim_hair_braid_a_04_transform_l {parent=aim_hair_braid_a_04_l, loc=<0 0 0>, rot=<0 0 0 1>}
 aim_hair_braid_a_04_r {loc=<-8.255 10.1183 131.8123>, rot=<0.2258 -0.2372 -0.6744 -0.6617>}
 aim_hair_braid_a_04_transform_r {parent=aim_hair_braid_a_04_r, loc=<0 0 0>, rot=<0 0 0 -1>}
 aim_hair_braid_b_04_r {loc=<-12.4878 5.4242 134.8195>, rot=<0.2002 -0.244 -0.7118 -0.6275>}
 aim_hair_braid_b_04_transform_r {parent=aim_hair_braid_b_04_r, loc=<0 0 0>, rot=<0 0 0 -1>}
]
variables [
 RBAN_OFF {type=double, default=0}
 aim_alpha {type=double, default=0}
 crouch_alpha {type=double, default=0}
 delta_head_down {type=double, default=0}
 delta_head_up {type=double, default=0}
 gliding_alpha {type=double, default=0}
 isFrontEnd {type=bool, default=False}
 moving2D_alpha {type=double, default=0}
 tactical_srpinting_alpha {type=double, default=0}
]
