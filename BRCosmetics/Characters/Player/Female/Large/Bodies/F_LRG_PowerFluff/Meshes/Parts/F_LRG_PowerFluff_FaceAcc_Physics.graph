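# Physics asset for the PowerFluff large female body's FaceAcc mesh part, defining capsule bodies and joints for four hair braid chains (a and b, left and right).
# Braid roots lock angularly to neck_01, while downstream links allow limited swing with little or no twist, generally loosening toward the tips.
# Capsules on the spine, neck, clavicle helpers and upper-arm helpers, joined by angle-locked constraints, form an upper-body proxy around the braids.
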PhysicsAsset /BRCosmetics/Characters/Player/Female/Large/Bodies/F_LRG_PowerFluff/Meshes/Parts/F_LRG_PowerFluff_FaceAcc_Physics
bodies {
 dyn_hair_braid_a_00_l [{
   shapeType=Capsule
   length=2.8726
   radius=1.5295
   shapeName=dyn_hair_braid_a_00_l_capsule
   center=<4.466 0.1846 -1.3521>
   rotation=<rot 0 -87.5352 -79.5405>
  }]
 dyn_hair_braid_a_00_r [{
   shapeType=Capsule
   length=4.8842
   radius=1.559
   shapeName=dyn_hair_braid_a_00_r_capsule
   center=<5.7964 -0.3115 -1.4696>
   rotation=<rot 0 -85.9 -85.769>
  }]
 dyn_hair_braid_a_01_l [{shapeType=Capsule, length=5, radius=1, shapeName=dyn_hair_braid_a_01_l_capsule, center=<3.7837 0 0>, rotation=<rot 0 -89.8571 -89.498>}]
 dyn_hair_braid_a_01_r [{shapeType=Capsule, length=5.1628, radius=1, shapeName=dyn_hair_braid_a_01_r_capsule, center=<3.4515 0 -0.1563>, rotation=<rot 0 90 -89.9998>}]
 dyn_hair_braid_a_02_l [{shapeType=Capsule, length=4.5, radius=1, shapeName=dyn_hair_braid_a_02_l_capsule, center=<3.5649 0 0.2945>, rotation=<rot -90 90 -90>}]
 dyn_hair_braid_a_02_r [{shapeType=Capsule, length=3.5, radius=1.5, shapeName=dyn_hair_braid_a_02_r_capsule, center=<3.5866 0 0.3312>, rotation=<rot -90 180 -180>}]
 dyn_hair_braid_a_03_l [{shapeType=Capsule, length=4, radius=1, shapeName=dyn_hair_braid_a_03_l_capsule, center=<3.1249 0 0.5054>, rotation=<rot -90 0.7981 -0.7981>}]
 dyn_hair_braid_a_03_r [{
   shapeType=Capsule
   length=5.5
   radius=1
   shapeName=dyn_hair_braid_a_03_r_capsule
   center=<3.9117 0 0.1228>
   rotation=<rot -90 -6.7214 6.7214>
  }]
 dyn_hair_braid_a_04_l [{shapeType=Capsule, length=1.5, radius=1.5, shapeName=dyn_hair_braid_a_04_l_capsule, center=<2.0523 0 0.5363>, rotation=<rot -90 90 -90>}]
 dyn_hair_braid_a_04_r [{shapeType=Capsule, length=3.5, radius=1.2, shapeName=dyn_hair_braid_a_04_r_capsule, center=<2.0814 0 0.0862>, rotation=<rot -90 -180 180>}]
 dyn_hair_braid_b_00_l [{
   shapeType=Capsule
   length=0.8764
   radius=1
   shapeName=dyn_hair_braid_b_00_l_capsule
   center=<3.5697 -0.4342 -2.0249>
   rotation=<rot 0 18.2687 -50.5134>
  }]
 dyn_hair_braid_b_00_r [{
   shapeType=Capsule
   length=2.0609
   radius=1.4476
   shapeName=dyn_hair_braid_b_00_r_capsule
   center=<3.769 0.001 -0.5017>
   rotation=<rot 0 -96.3873 -86.3988>
  }]
 dyn_hair_braid_b_01_l [{shapeType=Capsule, length=5.5, radius=1, shapeName=dyn_hair_braid_b_01_l_capsule, center=<4.3184 0 0>, rotation=<rot 0 94.0915 -89.3835>}]
 dyn_hair_braid_b_01_r [{shapeType=Capsule, length=5, radius=1, shapeName=dyn_hair_braid_b_01_r_capsule, center=<3 0 0>, rotation=<rot 0 -90 89.9998>}]
 dyn_hair_braid_b_02_l [{
   shapeType=Capsule
   length=3
   radius=1.3
   shapeName=dyn_hair_braid_b_02_l_capsule
   center=<4.2047 -0.1019 0.4128>
   rotation=<rot -90 179.9604 -179.9605>
  }]
 dyn_hair_braid_b_02_r [{
   shapeType=Capsule
   length=3
   radius=1
   shapeName=dyn_hair_braid_b_02_r_capsule
   center=<3.3997 0.1392 0.3991>
   rotation=<rot -90 -8.2517 8.2517>
  }]
 dyn_hair_braid_b_03_l [{shapeType=Capsule, length=3.5, radius=1, shapeName=dyn_hair_braid_b_03_l_capsule, center=<1.5701 -0.6646 1.707>, rotation=<rot -73.9163 0 0>}]
 dyn_hair_braid_b_03_r [{
   shapeType=Capsule
   length=6.5
   radius=1.2
   shapeName=dyn_hair_braid_b_03_r_capsule
   center=<3.537 0 0.6225>
   rotation=<rot -90 -179.9852 179.9852>
  }]
 dyn_hair_braid_b_04_r [{shapeType=Capsule, length=2, radius=1.5, shapeName=dyn_hair_braid_b_04_r_capsule, center=<3.2185 0 0.7561>, rotation=<rot -90 -90 90>}]
 helper_clavicle_l [{
   shapeType=Capsule
   length=13.5
   radius=5.505
   shapeName=helper_clavicle_l_capsule
   center=<5.3618 -2.0236 1.7413>
   rotation=<rot 74.9999 -0.0042 -0.0044>
  }]
 helper_clavicle_r [{
   shapeType=Capsule
   length=13.5
   radius=5.505
   shapeName=helper_clavicle_l_capsule
   center=<-5.3618 2.0236 -1.7413>
   rotation=<rot 74.9999 -0.0042 -0.0044>
  }]
 helper_upperarm_l [{shapeType=Capsule, length=18, radius=7.5, shapeName=helper_upperarm_l_capsule, center=<10.9095 -1 -2.5274>, rotation=<rot 90 0 0>}]
 helper_upperarm_r [{shapeType=Capsule, length=18, radius=7.5, shapeName=helper_upperarm_l_capsule, center=<-10.9095 1 2.5274>, rotation=<rot 90 0 0>}]
 neck_01 [{shapeType=Capsule, length=19.5, radius=6.5, shapeName=head_capsule, center=<-6.9406 1.6729 0>, rotation=<rot 90 0 0.0002>}]
 spine_04 [{shapeType=Capsule, length=20.5, radius=13, shapeName=spine_05_capsule, center=<-0.6046 3.1388 0.6644>}]
 spine_05 [
  {shapeType=Capsule, length=26.5, radius=10.5, shapeName=spine_05_capsule, center=<-3.6034 0.0061 0>}
  {shapeType=Capsule, length=22.5, radius=15, shapeName=spine_05_capsule, center=<0 -13.956 0>}
 ]
}
constraints [
 {
  bone1Name=spine_05
  bone2Name=spine_04
  swing1LimitDegrees=45
  swing1Motion=Locked
  swing2LimitDegrees=45
  swing2Motion=Locked
  twistLimitDegrees=45
  twistMotion=Locked
 }
 {bone1Name=dyn_hair_braid_b_04_r, bone2Name=dyn_hair_braid_b_03_r, swing1LimitDegrees=15, swing2LimitDegrees=15, twistLimitDegrees=0}
 {bone1Name=dyn_hair_braid_b_03_r, bone2Name=dyn_hair_braid_b_02_r, swing1LimitDegrees=15, swing2LimitDegrees=15, twistLimitDegrees=0}
 {bone1Name=dyn_hair_braid_b_02_r, bone2Name=dyn_hair_braid_b_01_r, swing1LimitDegrees=15, swing2LimitDegrees=15, twistLimitDegrees=0}
 {bone1Name=dyn_hair_braid_b_01_r, bone2Name=dyn_hair_braid_b_00_r, swing1LimitDegrees=10, swing2LimitDegrees=15, twistLimitDegrees=0}
 {bone1Name=dyn_hair_braid_b_03_l, bone2Name=dyn_hair_braid_b_02_l, swing1LimitDegrees=25, swing2LimitDegrees=25, twistLimitDegrees=3}
 {bone1Name=dyn_hair_braid_b_02_l, bone2Name=dyn_hair_braid_b_01_l, swing1LimitDegrees=15, swing2LimitDegrees=15, twistLimitDegrees=0}
 {bone1Name=dyn_hair_braid_b_01_l, bone2Name=dyn_hair_braid_b_00_l, swing1LimitDegrees=10, swing2LimitDegrees=15, twistLimitDegrees=0}
 {bone1Name=dyn_hair_braid_a_04_r, bone2Name=dyn_hair_braid_a_03_r, swing1LimitDegrees=25, swing2LimitDegrees=25, twistLimitDegrees=0}
 {bone1Name=dyn_hair_braid_a_03_r, bone2Name=dyn_hair_braid_a_02_r, swing1LimitDegrees=20, swing2LimitDegrees=20, twistLimitDegrees=0}
 {bone1Name=dyn_hair_braid_a_02_r, bone2Name=dyn_hair_braid_a_01_r, swing1LimitDegrees=15, swing2LimitDegrees=15, twistLimitDegrees=0}
 {bone1Name=dyn_hair_braid_a_01_r, bone2Name=dyn_hair_braid_a_00_r, swing1LimitDegrees=10, swing2LimitDegrees=15, twistLimitDegrees=0}
 {bone1Name=dyn_hair_braid_a_04_l, bone2Name=dyn_hair_braid_a_03_l, swing1LimitDegrees=30, swing2LimitDegrees=30, twistLimitDegrees=3}
 {bone1Name=dyn_hair_braid_a_03_l, bone2Name=dyn_hair_braid_a_02_l, swing1LimitDegrees=30, swing2LimitDegrees=30, twistLimitDegrees=3}
 {bone1Name=dyn_hair_braid_a_02_l, bone2Name=dyn_hair_braid_a_01_l, swing1LimitDegrees=15, swing2LimitDegrees=15, twistLimitDegrees=3}
 {bone1Name=dyn_hair_braid_a_01_l, bone2Name=dyn_hair_braid_a_00_l, swing1LimitDegrees=10, swing2LimitDegrees=15, twistLimitDegrees=0}
 {
  bone1Name=neck_01
  bone2Name=spine_05
  swing1LimitDegrees=45
  swing1Motion=Locked
  swing2LimitDegrees=45
  swing2Motion=Locked
  twistLimitDegrees=45
  twistMotion=Locked
 }
 {
  bone1Name=dyn_hair_braid_b_00_l
  bone2Name=neck_01
  swing1LimitDegrees=45
  swing1Motion=Locked
  swing2LimitDegrees=45
  swing2Motion=Locked
  twistLimitDegrees=45
  twistMotion=Locked
 }
 {
  bone1Name=dyn_hair_braid_a_00_l
  bone2Name=neck_01
  swing1LimitDegrees=45
  swing1Motion=Locked
  swing2LimitDegrees=45
  swing2Motion=Locked
  twistLimitDegrees=45
  twistMotion=Locked
 }
 {
  bone1Name=dyn_hair_braid_a_00_r
  bone2Name=neck_01
  swing1LimitDegrees=45
  swing1Motion=Locked
  swing2LimitDegrees=45
  swing2Motion=Locked
  twistLimitDegrees=45
  twistMotion=Locked
 }
 {
  bone1Name=dyn_hair_braid_b_00_r
  bone2Name=neck_01
  swing1LimitDegrees=45
  swing1Motion=Locked
  swing2LimitDegrees=45
  swing2Motion=Locked
  twistLimitDegrees=45
  twistMotion=Locked
 }
 {
  bone1Name=helper_clavicle_l
  bone2Name=spine_05
  swing1LimitDegrees=45
  swing1Motion=Locked
  swing2LimitDegrees=45
  swing2Motion=Locked
  twistLimitDegrees=45
  twistMotion=Locked
 }
 {
  bone1Name=helper_upperarm_l
  bone2Name=spine_05
  swing1LimitDegrees=45
  swing1Motion=Locked
  swing2LimitDegrees=45
  swing2Motion=Locked
  twistLimitDegrees=45
  twistMotion=Locked
 }
 {
  bone1Name=helper_clavicle_r
  bone2Name=spine_05
  swing1LimitDegrees=45
  swing1Motion=Locked
  swing2LimitDegrees=45
  swing2Motion=Locked
  twistLimitDegrees=45
  twistMotion=Locked
 }
 {
  bone1Name=helper_upperarm_r
  bone2Name=spine_05
  swing1LimitDegrees=45
  swing1Motion=Locked
  swing2LimitDegrees=45
  swing2Motion=Locked
  twistLimitDegrees=45
  twistMotion=Locked
 }
]
