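# Post-process style Anim Blueprint for PowerFluff face accessories, adding secondary motion to earrings, hair accessories, ears and hair buns.
# The copied parent pose runs through a Control Rig, then cached stages of AnimDynamics, with weighted constraints placing hair accessories and a spine_05-based RigidBody node before the ears.
# The event graph sets Control Rig alphas from movement, aim, crouch, sprint, parachute and rigid-body state, and computes per-state gravity and sim-space settings.
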AnimBlueprint /BRCosmetics/Characters/Player/Female/Large/Bodies/F_LRG_PowerFluff/Meshes/Parts/F_LRG_PowerFluff_FaceAcc_AnimBP
target_skeleton=&/F_LRG_PowerFluff_FaceAcc_Skeleton
graphs [
 AnimGraph {nodes [
   Local To Component()
   n1 = Anim Dynamics - Bone: dyn_bun_hair_r!(
    ComponentPose=Local To Component(LocalPose=Use cached pose 'ear R cache'())
    Alpha=1
    GravityOverride:Vector="0,0,-980"
    ExternalForce="0,0,0"
   ) -> (Pose) {
    angularDampingOverride=1
    angularSpringConstant=10000
    bAngularSpring=true
    bOverrideAngularDamping=true
    bUseGravityOverride=true
    gravityScale=0.35
    simulationSpace=RootRelative
    alphaBoolBlend {blendOption=Linear}
    boundBone {boneName=dyn_bun_hair_r}
    componentAppliedLinearAccClamp=<10000 10000 10000>
    componentLinearAccScale=<0.3 0.3 0.3>
    componentLinearVelScale=<0.3 0.3 0.3>
    gravityOverride {z=-980}
    physicsBodyDefinitions [{
      boundBone {boneName=dyn_bun_hair_r}
      boxExtents=<30 30 30>
      constraintSetup {angularLimitsMax=<0 3 0>, angularLimitsMin=<0 -10 0>, angularTarget=<1 0 0>}
      localJointOffset=<1 0 0>
     }]
   }
   n2 = Anim Dynamics - Bone: dyn_bun_hair_l!(ComponentPose=@n1, Alpha=1, GravityOverride:Vector="0,0,-980", ExternalForce="0,0,0") -> (Pose) {
    angularDampingOverride=1
    angularSpringConstant=10000
    bAngularSpring=true
    bOverrideAngularDamping=true
    bUseGravityOverride=true
    gravityScale=0.35
    simulationSpace=RootRelative
    alphaBoolBlend {blendOption=Linear}
    boundBone {boneName=dyn_bun_hair_l}
    componentAppliedLinearAccClamp=<10000 10000 10000>
    componentLinearAccScale=<0.2 0.2 0.2>
    componentLinearVelScale=<0.2 0.2 0.2>
    gravityOverride {z=-980}
    physicsBodyDefinitions [{
      boundBone {boneName=dyn_bun_hair_l}
      boxExtents=<20 20 20>
      constraintSetup {angularLimitsMax=<0 3 0>, angularLimitsMin=<0 -7 0>, angularTarget=<1 0 0>}
      localJointOffset=<4 0 0>
     }]
   }
   Output Pose(Result=Component To Local(ComponentPose=@n2)) {layerGroup=None}
   n3 = Control Rig!(
    Source=Copy Pose From Mesh() {bUseAttachedParent=true}
    Alpha
    aim_alpha:double=Get aim_alpha() -> (aim_alpha:double)
    moving2D_alpha:double=Get moving2D_alpha() -> (moving2D_alpha:double)
    gliding_alpha:double=Get skydiving_alpha() -> (skydiving_alpha:double)
    crouch_alpha:double=Get crouch_alpha() -> (crouch_alpha:double)
    tactical_srpinting_alpha:double=Get tacticalsprint_alpha() -> (tacticalsprint_alpha:double)
    isFrontEnd:bool=Get bIsInFrontEnd()
    RBAN_OFF:double=Get RBAN_OFF() -> (RBAN_OFF:double)
   ) -> (Pose) {
    alphaBoolBlend {blendOption=Linear}
    controlRigAssetReference {blueprintRigClass=&/F_LRG_PowerFluff_FaceAcc_CtrlRig.F_LRG_PowerFluff_FaceAcc_CtrlRig_C}
   }
   n4 = Anim Dynamics - Bone: dyn_earing_l(ComponentPose=Local To Component(LocalPose=@n3), Alpha=1, ExternalForce="0,0,0") {
    angularDampingOverride=1
    angularSpringConstant=700
    bOverrideAngularDamping=true
    bUseGravityOverride=true
    gravityScale=0.35
    simulationSpace=RootRelative
    alphaBoolBlend {blendOption=Linear}
    alphaScaleBiasClamp {bInterpResult=true, interpSpeedDecreasing=2.5, interpSpeedIncreasing=8}
    boundBone {boneName=dyn_earing_l}
    componentAppliedLinearAccClamp=<3000 3000 3000>
    componentLinearAccScale=<0.3 0.3 0.3>
    componentLinearVelScale=<0.3 0.3 0.3>
    gravityOverride {z=-980}
    physicsBodyDefinitions [{
      boundBone {boneName=dyn_earing_l}
      boxExtents=<5 8 5>
      constraintSetup {angularLimitsMax=<0 10 30>, angularLimitsMin=<0 -15 -30>, angularTarget=<1 0 0>}
      localJointOffset=<1 0 0>
     }]
    relativeSpaceBone {boneName=root}
   }
   n5 = Anim Dynamics - Bone: dyn_earing_r(ComponentPose=@n4, Alpha=1, ExternalForce="0,0,0") {
    angularDampingOverride=1
    angularSpringConstant=700
    bOverrideAngularDamping=true
    bUseGravityOverride=true
    gravityScale=0.35
    simulationSpace=RootRelative
    alphaBoolBlend {blendOption=Linear}
    alphaScaleBiasClamp {bInterpResult=true, interpSpeedDecreasing=2.5, interpSpeedIncreasing=8}
    boundBone {boneName=dyn_earing_r}
    componentAppliedLinearAccClamp=<3000 3000 3000>
    componentLinearAccScale=<0.3 0.3 0.3>
    componentLinearVelScale=<0.3 0.3 0.3>
    gravityOverride {z=-980}
    physicsBodyDefinitions [{
      boundBone {boneName=dyn_earing_r}
      boxExtents=<5 8 5>
      constraintSetup {angularLimitsMax=<0 30 40>, angularLimitsMin=<0 -15 -40>, angularTarget=<1 0 0>}
      localJointOffset=<1 0 0>
     }]
    relativeSpaceBone {boneName=root}
   }
   Save cached pose 'earing cache'(Pose=Component To Local(ComponentPose=@n5)) {cacheName=earing cache}
   n6 = RigidBody!(
    ComponentPose=Local To Component(LocalPose=Use cached pose 'hair accesories cache'())
    Alpha
    OverrideWorldGravity:Vector="0,0,0"
    ExternalForce
    SimSpaceSettings:Sim Space Settings Structure=(WorldAlpha=0,VelocityScaleZ=1,DampingAlpha=1,MaxLinearVelocity=10000,MaxAngularVelocity=10000,MaxLinearAcceleration=10000,MaxAngularAcceleration=10000,ExternalLinearDragV=(X=0,Y=0,Z=0),ExternalLinearVelocity=(X=0,Y=0,Z=0),ExternalAngularVelocity=(X=0,Y=0,Z=0))
   ) -> (Pose) {
    bOverrideWorldGravity=true
    overridePhysicsAsset=&/F_LRG_PowerFluff_FaceAcc_Physics
    simulationSpace=BaseBoneSpace
    alphaBoolBlend {blendOption=Linear}
    baseBoneRef {boneName=spine_05}
   }
   n7 = Anim Dynamics - Bone: dyn_ear_03_l!(ComponentPose=@n6, Alpha=1, GravityOverride:Vector="0,0,-980", ExternalForce="0,0,0") -> (Pose) {
    angularDampingOverride=1
    bAngularSpring=true
    bOverrideAngularDamping=true
    bUseGravityOverride=true
    gravityScale=0.35
    simulationSpace=RootRelative
    alphaBoolBlend {blendOption=Linear}
    boundBone {boneName=dyn_ear_03_l}
    componentAppliedLinearAccClamp=<10000 10000 10000>
    componentLinearAccScale=<0.3 0.3 0.3>
    componentLinearVelScale=<1 1 1>
    gravityOverride {z=-980}
    physicsBodyDefinitions [{
      boundBone {boneName=dyn_ear_03_l}
      boxExtents=<30 30 30>
      constraintSetup {angularLimitsMax=<0 3 3>, angularLimitsMin=<0 -3 -3>, angularTarget=<1 0 0>}
      localJointOffset=<2 0 0>
     }]
   }
   n8 = Anim Dynamics - Bone: dyn_ear_02_l!(ComponentPose=@n7, Alpha=1, GravityOverride:Vector="0,0,-980", ExternalForce="0,0,0") -> (Pose) {
    angularDampingOverride=1
    bAngularSpring=true
    bOverrideAngularDamping=true
    bUseGravityOverride=true
    gravityScale=0.35
    simulationSpace=RootRelative
    alphaBoolBlend {blendOption=Linear}
    boundBone {boneName=dyn_ear_02_l}
    componentAppliedLinearAccClamp=<10000 10000 10000>
    componentLinearAccScale=<0.3 0.3 0.3>
    componentLinearVelScale=<1 1 1>
    gravityOverride {z=-980}
    physicsBodyDefinitions [{
      boundBone {boneName=dyn_ear_02_l}
      boxExtents=<30 30 30>
      constraintSetup {angularLimitsMax=<0 3 3>, angularLimitsMin=<0 -3 -3>, angularTarget=<1 0 0>}
      localJointOffset=<2 0 0>
     }]
   }
   n9 = Anim Dynamics - Bone: dyn_ear_01_l!(ComponentPose=@n8, Alpha=1, GravityOverride:Vector="0,0,-980", ExternalForce="0,0,0") -> (Pose) {
    angularDampingOverride=1
    bAngularSpring=true
    bOverrideAngularDamping=true
    bUseGravityOverride=true
    gravityScale=0.35
    simulationSpace=RootRelative
    alphaBoolBlend {blendOption=Linear}
    boundBone {boneName=dyn_ear_01_l}
    componentAppliedLinearAccClamp=<10000 10000 10000>
    componentLinearAccScale=<0.3 0.3 0.3>
    componentLinearVelScale=<1 1 1>
    gravityOverride {z=-980}
    physicsBodyDefinitions [{
      boundBone {boneName=dyn_ear_01_l}
      boxExtents=<30 30 30>
      constraintSetup {angularLimitsMax=<0 3 3>, angularLimitsMin=<0 -3 -3>, angularTarget=<1 0 0>}
      localJointOffset=<2 0 0>
     }]
   }
   Save cached pose 'ear L cache'(Pose=Component To Local(ComponentPose=@n9)) {cacheName=ear L cache}
   n10 = "Constraint - base_hair_b_acc_01_r (PPPPP)"!(
    ComponentPose=Local To Component(LocalPose=Use cached pose 'earing cache'())
    Alpha=1
    ConstraintWeights_0=0.7813
    ConstraintWeights_1:float=0.1782
    ConstraintWeights_2:float=0.0199
    ConstraintWeights_3:float=0.0194
    ConstraintWeights_4:float=0.0013
   ) -> (Pose) {
    alphaBoolBlend {blendOption=Linear}
    boneToModify {boneName=base_hair_b_acc_01_r}
    constraintSetup [
     {targetBone {boneName=dyn_hair_braid_b_01_r}}
     {targetBone {boneName=dyn_hair_b_acc_01_r}}
     {targetBone {boneName=dyn_hair_braid_b_00_r}}
     {targetBone {boneName=dyn_hair_b_acc_02_r}}
     {targetBone {boneName=dyn_hair_braid_b_02_r}}
    ]
    constraintWeights [1,1,1,1,1]
   }
   n11 = "Constraint - base_hair_a_acc_02_l (PPPPP)"!(
    ComponentPose=@n10
    Alpha=1
    ConstraintWeights_0=0.5855
    ConstraintWeights_1:float=0.3415
    ConstraintWeights_2:float=0.0369
    ConstraintWeights_3:float=0.0357
    ConstraintWeights_4:float=0.0004
   ) -> (Pose) {
    alphaBoolBlend {blendOption=Linear}
    boneToModify {boneName=base_hair_a_acc_02_l}
    constraintSetup [
     {targetBone {boneName=dyn_hair_a_acc_01_l}}
     {targetBone {boneName=dyn_hair_a_acc_02_l}}
     {targetBone {boneName=head}}
     {targetBone {boneName=dyn_hair_braid_b_00_l}}
     {targetBone {boneName=dyn_hair_braid_a_00_l}}
    ]
    constraintWeights [1,1,1,1,1]
   }
   n12 = "Constraint - base_hair_a_acc_01_l (PPPPP)"!(
    ComponentPose=@n11
    Alpha=1
    ConstraintWeights_0=0.2961
    ConstraintWeights_1:float=0.2823
    ConstraintWeights_2:float=0.2219
    ConstraintWeights_3:float=0.1963
    ConstraintWeights_4:float=0.0035
   ) -> (Pose) {
    alphaBoolBlend {blendOption=Linear}
    boneToModify {boneName=base_hair_a_acc_01_l}
    constraintSetup [
     {targetBone {boneName=dyn_hair_a_acc_01_l}}
     {targetBone {boneName=dyn_hair_braid_b_00_l}}
     {targetBone {boneName=head}}
     {targetBone {boneName=dyn_hair_a_acc_02_l}}
     {targetBone {boneName=dyn_hair_braid_a_00_l}}
    ]
    constraintWeights [1,1,1,1,1]
   }
   n13 = "Constraint - base_hair_b_acc_02_r (PPPPP)"!(
    ComponentPose=@n12
    Alpha=1
    ConstraintWeights_0=0.5167
    ConstraintWeights_1:float=0.306
    ConstraintWeights_2:float=0.1761
    ConstraintWeights_3:float=0.0007
    ConstraintWeights_4:float=0.0004
   ) -> (Pose) {
    alphaBoolBlend {blendOption=Linear}
    boneToModify {boneName=base_hair_b_acc_02_r}
    constraintSetup [
     {targetBone {boneName=dyn_hair_b_acc_01_r}}
     {targetBone {boneName=dyn_hair_braid_b_01_r}}
     {targetBone {boneName=dyn_hair_b_acc_02_r}}
     {targetBone {boneName=dyn_hair_braid_b_02_r}}
     {targetBone {boneName=dyn_hair_braid_b_00_r}}
    ]
    constraintWeights [1,1,1,1,1]
   }
   n14 = Anim Dynamics - Bone: base_hair_b_acc_02_r(ComponentPose=@n13, Alpha=1, ExternalForce="0,0,0") {
    angularDampingOverride=1
    angularSpringConstant=5000
    bAngularSpring=true
    bOverrideAngularDamping=true
    bUseGravityOverride=true
    gravityScale=0.35
    simulationSpace=RootRelative
    alphaBoolBlend {blendOption=Linear}
    boundBone {boneName=base_hair_b_acc_02_r}
    componentAppliedLinearAccClamp=<3000 3000 3000>
    componentLinearAccScale=<1 1 1>
    componentLinearVelScale=<1 1 1>
    gravityOverride {z=-980}
    physicsBodyDefinitions [{
      boundBone {boneName=base_hair_b_acc_02_r}
      boxExtents=<30 30 30>
      constraintSetup {angularLimitsMax=<15 15 15>, angularLimitsMin=<-15 -15 -15>, angularTarget=<1 0 0>}
      localJointOffset=<4 0 0>
     }]
   }
   n15 = Anim Dynamics - Bone: base_hair_b_acc_01_r(ComponentPose=@n14, Alpha=1, ExternalForce="0,0,0") {
    angularDampingOverride=1
    angularSpringConstant=5000
    bAngularSpring=true
    bOverrideAngularDamping=true
    bUseGravityOverride=true
    gravityScale=0.35
    simulationSpace=RootRelative
    alphaBoolBlend {blendOption=Linear}
    boundBone {boneName=base_hair_b_acc_01_r}
    componentAppliedLinearAccClamp=<3000 3000 3000>
    componentLinearAccScale=<1 1 1>
    componentLinearVelScale=<1 1 1>
    gravityOverride {z=-980}
    physicsBodyDefinitions [{
      boundBone {boneName=base_hair_b_acc_01_r}
      boxExtents=<30 30 30>
      constraintSetup {angularLimitsMax=<15 15 15>, angularLimitsMin=<-15 -15 -15>, angularTarget=<1 0 0>}
      localJointOffset=<4 0 0>
     }]
   }
   n16 = Anim Dynamics - Bone: base_hair_a_acc_01_l!(ComponentPose=@n15, Alpha=1, GravityOverride:Vector="0,0,-980", ExternalForce="0,0,0") -> (Pose) {
    angularDampingOverride=1
    angularSpringConstant=5000
    bAngularSpring=true
    bOverrideAngularDamping=true
    bUseGravityOverride=true
    gravityScale=0.35
    simulationSpace=RootRelative
    alphaBoolBlend {blendOption=Linear}
    boundBone {boneName=base_hair_a_acc_01_l}
    componentAppliedLinearAccClamp=<3000 3000 3000>
    componentLinearAccScale=<1 1 1>
    componentLinearVelScale=<1 1 1>
    gravityOverride {z=-980}
    physicsBodyDefinitions [{
      boundBone {boneName=base_hair_a_acc_01_l}
      boxExtents=<30 30 30>
      constraintSetup {angularLimitsMax=<15 15 15>, angularLimitsMin=<-15 -15 -15>, angularTarget=<1 0 0>}
      localJointOffset=<4 0 0>
     }]
   }
   n17 = Anim Dynamics - Bone: base_hair_a_acc_02_l!(ComponentPose=@n16, Alpha=1, GravityOverride:Vector="0,0,-980", ExternalForce="0,0,0") -> (Pose) {
    angularDampingOverride=1
    angularSpringConstant=5000
    bAngularSpring=true
    bOverrideAngularDamping=true
    bUseGravityOverride=true
    gravityScale=0.35
    simulationSpace=RootRelative
    alphaBoolBlend {blendOption=Linear}
    boundBone {boneName=base_hair_a_acc_02_l}
    componentAppliedLinearAccClamp=<3000 3000 3000>
    componentLinearAccScale=<1 1 1>
    componentLinearVelScale=<1 1 1>
    gravityOverride {z=-980}
    physicsBodyDefinitions [{
      boundBone {boneName=base_hair_a_acc_02_l}
      boxExtents=<30 30 30>
      constraintSetup {angularLimitsMax=<15 15 15>, angularLimitsMin=<-15 -15 -15>, angularTarget=<1 0 0>}
      localJointOffset=<4 0 0>
     }]
   }
   Save cached pose 'hair accesories cache'(Pose=Component To Local(ComponentPose=@n17)) {cacheName=hair accesories cache}
   n18 = Anim Dynamics - Bone: dyn_ear_03_r!(
    ComponentPose=Local To Component(LocalPose=Use cached pose 'ear L cache'())
    Alpha=1
    GravityOverride:Vector="0,0,-980"
    ExternalForce="0,0,0"
   ) -> (Pose) {
    angularDampingOverride=1
    bAngularSpring=true
    bOverrideAngularDamping=true
    bUseGravityOverride=true
    gravityScale=0.35
    simulationSpace=RootRelative
    alphaBoolBlend {blendOption=Linear}
    boundBone {boneName=dyn_ear_03_r}
    componentAppliedLinearAccClamp=<10000 10000 10000>
    componentLinearAccScale=<0.3 0.3 0.3>
    componentLinearVelScale=<1 1 1>
    gravityOverride {z=-980}
    physicsBodyDefinitions [{
      boundBone {boneName=dyn_ear_03_r}
      boxExtents=<35 35 35>
      constraintSetup {angularLimitsMax=<0 2 3>, angularLimitsMin=<0 -4 -2>, angularTarget=<1 0 0>}
      localJointOffset=<2 0 0>
     }]
   }
   n19 = Anim Dynamics - Bone: dyn_ear_02_r!(ComponentPose=@n18, Alpha=1, GravityOverride:Vector="0,0,-980", ExternalForce="0,0,0") -> (Pose) {
    angularDampingOverride=1
    bAngularSpring=true
    bOverrideAngularDamping=true
    bUseGravityOverride=true
    gravityScale=0.35
    simulationSpace=RootRelative
    alphaBoolBlend {blendOption=Linear}
    boundBone {boneName=dyn_ear_02_r}
    componentAppliedLinearAccClamp=<10000 10000 10000>
    componentLinearAccScale=<0.3 0.3 0.3>
    componentLinearVelScale=<1 1 1>
    gravityOverride {z=-980}
    physicsBodyDefinitions [{
      boundBone {boneName=dyn_ear_02_r}
      boxExtents=<35 35 35>
      constraintSetup {angularLimitsMax=<0 2 3>, angularLimitsMin=<0 -4 -2>, angularTarget=<1 0 0>}
      localJointOffset=<2 0 0>
     }]
   }
   n20 = Anim Dynamics - Bone: dyn_ear_01_r!(ComponentPose=@n19, Alpha=1, GravityOverride:Vector="0,0,-980", ExternalForce="0,0,0") -> (Pose) {
    angularDampingOverride=1
    bAngularSpring=true
    bOverrideAngularDamping=true
    bUseGravityOverride=true
    gravityScale=0.35
    simulationSpace=RootRelative
    alphaBoolBlend {blendOption=Linear}
    boundBone {boneName=dyn_ear_01_r}
    componentAppliedLinearAccClamp=<10000 10000 10000>
    componentLinearAccScale=<0.3 0.3 0.3>
    componentLinearVelScale=<1 1 1>
    gravityOverride {z=-980}
    physicsBodyDefinitions [{
      boundBone {boneName=dyn_ear_01_r}
      boxExtents=<35 35 35>
      constraintSetup {angularLimitsMax=<0 2 3>, angularLimitsMin=<0 -4 -2>, angularTarget=<1 0 0>}
      localJointOffset=<2 0 0>
     }]
   }
   Save cached pose 'ear R cache'(Pose=Component To Local(ComponentPose=@n20)) {cacheName=ear R cache}
   Get aim_alpha()
  ]}
 EventGraph {nodes [
   Event BlueprintUpdateAnimation()
   > n1 = Gravity Override State Function DEPRECATED(
    CharacterPartAnimInstance=Self-Reference()
    GameTimeInSeconds=GetGameTimeInSeconds()
    GravityOverrideParams=Get Gravity Override Params()
    RBAN Params=Get RBAN Params()
   )
   > Set GravityOverride(GravityOverride=@n1)
   > Set Sim Space Settings(Sim Space Settings:Sim Space Settings Structure=Break RBANSetting Struct!(RBANSettingStruct:"RBANSetting Struct Structure (by ref)"=@n1.RBAN SettingsOut) -> (
     ComponentLinearAccScale_28_#1:Vector=~
     ComponentLinearVelScale_31_#2:Vector=~
     SimSpaceSettings_34_#3:Sim Space Settings Structure
    ) {is=title K2Node_BreakStruct}.SimSpaceSettings_34_#3) -> (Output_Get:Sim Space Settings Structure)
   > n2 = Sequence!(execute) -> (then_0, then_1, then_2:Exec)
   > n3 = Branch(Condition=Get bIsMoving2D())
   > Set moving2D_alpha(moving2D_alpha:double=1) -> (Output_Get:double)
   Set moving2D_alpha(execute=@n3.else, moving2D_alpha=0)
   n4 = Branch(execute=@n2.then_1, Condition=Get bIsParachuteOpened())
   > Set skydiving_alpha(skydiving_alpha:double=1) -> (Output_Get:double)
   Set skydiving_alpha(execute=@n4.else, skydiving_alpha=0)
   n5 = Branch(execute=@n2.then_2, Condition=Get bRigidBodyEnabled())
   > Set RBAN_OFF(RBAN_OFF:double=0) -> (Output_Get:double)
   Set RBAN_OFF(execute=@n5.else, RBAN_OFF=1)
   Event OnBeginTargeting() {is=title K2Node_Event}
   > Set aim_alpha(aim_alpha:double=1) -> (Output_Get:double)
   Event OnEndTargeting() {is=title K2Node_Event}
   > Set aim_alpha(aim_alpha=0)
   Event OnBeginTacticalSprinting() {is=title K2Node_Event}
   > Set tacticalsprint_alpha(tacticalsprint_alpha:double=1) -> (Output_Get:double)
   Event OnEndTacticalSprinting() {is=title K2Node_Event}
   > Set tacticalsprint_alpha(tacticalsprint_alpha=0)
   Event OnBeginCrouching()
   > Set crouch_alpha(crouch_alpha:double=1) -> (Output_Get:double)
   Event OnEndCrouching()
   > Set crouch_alpha(crouch_alpha=0)
  ]}
]
variables [
 aim_alpha {type=number}
 crouch_alpha {type=number}
 Gravity Override Params {
  type=GravityOverrideParamsStruct
  default {skydive_FloatingDown {gravityOverride=<0 -700 -1000>, windAmplitude=<200 -500 100>}, skydive_GoingUp {gravityOverride=<0 -300 -1000>}}
 }
 GravityOverride {type=Vector}
 moving2D_alpha {type=number}
 RBAN Params {
  type=RBANParamsStruct
  default {
   dBNO {simSpaceSettings {
     maxAngularAcceleration=1000
     maxAngularVelocity=3
     maxLinearAcceleration=300
     maxLinearVelocity=700
     velocityScaleZ=0
     externalLinearDragV=<0 0.5 0>
    }}
   dBNO_Carried {simSpaceSettings {
     maxAngularAcceleration=1000
     maxAngularVelocity=3
     maxLinearAcceleration=300
     maxLinearVelocity=700
     velocityScaleZ=0
     externalLinearDragV=<0 0.5 0>
    }}
   emoteOrMelee {simSpaceSettings {
     maxAngularAcceleration=1000
     maxAngularVelocity=3
     maxLinearAcceleration=300
     maxLinearVelocity=700
     velocityScaleZ=0
     externalLinearDragV=<0 0.5 0>
    }}
   front_End {simSpaceSettings {
     maxAngularAcceleration=1000
     maxAngularVelocity=3
     maxLinearAcceleration=300
     maxLinearVelocity=700
     velocityScaleZ=0
     worldAlpha=0.3
     externalLinearDragV=<0 0.5 0>
    }}
   onGround_Crouch_Moving {simSpaceSettings {
     maxAngularAcceleration=1000
     maxAngularVelocity=3
     maxLinearAcceleration=300
     maxLinearVelocity=700
     velocityScaleZ=0
     externalLinearDragV=<0 0.5 0>
    }}
   onGround_Falling {simSpaceSettings {
     maxAngularAcceleration=1000
     maxAngularVelocity=3
     maxLinearAcceleration=300
     maxLinearVelocity=700
     velocityScaleZ=0
     externalLinearDragV=<0 0.5 0>
    }}
   onGround_Moving {simSpaceSettings {
     maxAngularAcceleration=1000
     maxAngularVelocity=3
     maxLinearAcceleration=300
     maxLinearVelocity=700
     velocityScaleZ=0
     externalLinearDragV=<0 0.5 0>
    }}
   onGround_Standing {simSpaceSettings {
     maxAngularAcceleration=1000
     maxAngularVelocity=3
     maxLinearAcceleration=300
     maxLinearVelocity=700
     velocityScaleZ=0
     externalLinearDragV=<0 0.5 0>
    }}
   rail_Grinding {simSpaceSettings {
     maxAngularAcceleration=1000
     maxAngularVelocity=3
     maxLinearAcceleration=300
     maxLinearVelocity=700
     velocityScaleZ=0
     worldAlpha=0.1
     externalLinearDragV=<0 0.5 0>
    }}
   riding_Moving {simSpaceSettings {
     maxAngularAcceleration=1000
     maxAngularVelocity=3
     maxLinearAcceleration=300
     maxLinearVelocity=700
     velocityScaleZ=0
     externalLinearDragV=<0 0.5 0>
    }}
   riding_Moving_Targeting {simSpaceSettings {
     maxAngularAcceleration=1000
     maxAngularVelocity=3
     maxLinearAcceleration=300
     maxLinearVelocity=700
     velocityScaleZ=0
     externalLinearDragV=<0 0.5 0>
    }}
   riding_Standing {simSpaceSettings {
     maxAngularAcceleration=1000
     maxAngularVelocity=3
     maxLinearAcceleration=300
     maxLinearVelocity=700
     velocityScaleZ=0
     externalLinearDragV=<0 0.5 0>
    }}
   skydive_DivingDown {simSpaceSettings {
     maxAngularAcceleration=1000
     maxAngularVelocity=3
     maxLinearAcceleration=300
     maxLinearVelocity=700
     velocityScaleZ=0
     externalLinearDragV=<0 0.5 0>
    }}
   skydive_FloatingDown {
    componentLinearAccScale=<0 0.5 0>
    componentLinearVelScale=<0 0.5 0>
    simSpaceSettings {
     maxAngularAcceleration=1000
     maxAngularVelocity=3
     maxLinearAcceleration=300
     maxLinearVelocity=700
     velocityScaleZ=0
     externalLinearDragV=<0 0.5 0>
    }
   }
   skydive_GoingUp {simSpaceSettings {
     maxAngularAcceleration=1000
     maxAngularVelocity=3
     maxLinearAcceleration=300
     maxLinearVelocity=700
     velocityScaleZ=0
     externalLinearDragV=<0 0.5 0>
    }}
   skydive_Parachuting {
    componentLinearAccScale=<0 0.1 0.3>
    componentLinearVelScale=<0 0.1 0.3>
    simSpaceSettings {
     maxAngularAcceleration=1000
     maxAngularVelocity=3
     maxLinearAcceleration=300
     maxLinearVelocity=700
     velocityScaleZ=0
     externalLinearDragV=<0 0.5 0>
    }
   }
   skydive_Surfing {simSpaceSettings {
     maxAngularAcceleration=1000
     maxAngularVelocity=3
     maxLinearAcceleration=300
     maxLinearVelocity=700
     velocityScaleZ=0
     externalLinearDragV=<0 0.5 0>
    }}
   swimming {simSpaceSettings {
     maxAngularAcceleration=1000
     maxAngularVelocity=3
     maxLinearAcceleration=300
     maxLinearVelocity=700
     velocityScaleZ=0
     externalLinearDragV=<0 0.5 0>
    }}
   tactical_Sprint {simSpaceSettings {
     maxAngularAcceleration=1000
     maxAngularVelocity=3
     maxLinearAcceleration=300
     maxLinearVelocity=700
     velocityScaleZ=0
     externalLinearDragV=<0 0.5 0>
    }}
   ziplining {simSpaceSettings {
     maxAngularAcceleration=1000
     maxAngularVelocity=3
     maxLinearAcceleration=300
     maxLinearVelocity=700
     velocityScaleZ=0
     externalLinearDragV=<0 0.5 0>
    }}
  }
 }
 RBAN_OFF {type=number}
 Sim Space Settings {
  type=SimSpaceSettings
  default {
   dampingAlpha=1
   maxAngularAcceleration=1000
   maxAngularVelocity=3
   maxLinearAcceleration=300
   maxLinearVelocity=700
   velocityScaleZ=0
   worldAlpha=1
   externalLinearDragV=<0 0.5 0>
  }
 }
 skydiving_alpha {type=number}
 tacticalsprint_alpha {type=number}
]
